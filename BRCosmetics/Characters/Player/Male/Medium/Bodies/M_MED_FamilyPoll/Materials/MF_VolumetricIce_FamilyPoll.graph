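# Material function that adds a volumetric ice look to the FamilyPoll player body cosmetic, applied over incoming material attributes.
# The refracted view vector drives parallax offsets for scrolling interior noise and emission, plus view-dependent thickness.
# A dielectric interior-volume function combines refracted environment light, absorption, scattering and sun/sky lighting into a fake-refraction opaque result.
# Channel masks, surface detail normal/roughness, hue shifts and time-of-day colors shape the final blend.
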MaterialFunction /BRCosmetics/Characters/Player/Male/Medium/Bodies/M_MED_FamilyPoll/Materials/MF_VolumetricIce_FamilyPoll
graph {expressions {
  MaterialExpressionAdd_0 {
   class=MaterialExpressionAdd
   inputs [
    A {from=MaterialExpressionMultiply_0}
    B {from=MaterialExpressionNamedRerouteUsage_19}
   ]
  }
  MaterialExpressionAdd_1 {
   class=MaterialExpressionAdd
   inputs [
    A {from=MaterialExpressionDesaturation_0}
    B {from=MaterialExpressionNamedRerouteUsage_151}
   ]
  }
  MaterialExpressionAdd_12 {
   class=MaterialExpressionAdd
   inputs [
    A {from=MaterialExpressionNamedRerouteUsage_95}
    B {from=MaterialExpressionAdd_6}
   ]
  }
  MaterialExpressionAdd_16 {
   class=MaterialExpressionAdd
   inputs [
    A {from=MaterialExpressionNamedRerouteUsage_20}
    B {from=MaterialExpressionMultiply_53}
   ]
  }
  MaterialExpressionAdd_17 {
   class=MaterialExpressionAdd
   inputs [
    A {from=MaterialExpressionDivide_11}
    B {from=MaterialExpressionAppendVector_0}
   ]
   properties {desc=motion}
  }
  MaterialExpressionAdd_18 {
   class=MaterialExpressionAdd
   inputs [
    A {from=MaterialExpressionDivide_8}
    B {from=MaterialExpressionAppendVector_1}
   ]
   properties {desc=motion}
  }
  MaterialExpressionAdd_19 {
   class=MaterialExpressionAdd
   inputs [
    A {from=MaterialExpressionMultiply_47}
    B {from=MaterialExpressionNamedRerouteUsage_108}
   ]
  }
  MaterialExpressionAdd_2 {
   class=MaterialExpressionAdd
   inputs [
    A {from=MaterialExpressionAdd_19}
    B {from=MaterialExpressionNamedRerouteUsage_84}
   ]
  }
  MaterialExpressionAdd_3 {
   class=MaterialExpressionAdd
   inputs [
    A {from=MaterialExpressionMultiply_29}
    B {from=MaterialExpressionNamedRerouteUsage_125}
   ]
  }
  MaterialExpressionAdd_6 {
   class=MaterialExpressionAdd
   inputs [
    A {from=MaterialExpressionConvert_0}
    B {from=""}
   ]
   properties {constB=0.5}
  }
  MaterialExpressionAdd_7 {
   class=MaterialExpressionAdd
   inputs [
    A {from=MaterialExpressionMultiply_35}
    B {from=MaterialExpressionNamedRerouteUsage_110}
   ]
  }
  MaterialExpressionAdd_8 {
   class=MaterialExpressionAdd
   inputs [
    A {from=MaterialExpressionMultiply_38}
    B {from=MaterialExpressionNamedRerouteUsage_88}
   ]
  }
  MaterialExpressionAppendVector_0 {
   class=MaterialExpressionAppendVector
   inputs [
    A {from=MaterialExpressionConstant_11, output=""}
    B {from=MaterialExpressionMultiply_10, output=""}
   ]
  }
  MaterialExpressionAppendVector_1 {
   class=MaterialExpressionAppendVector
   inputs [
    A {from=MaterialExpressionConstant_12, output=""}
    B {from=MaterialExpressionMultiply_3, output=""}
   ]
  }
  MaterialExpressionBlendMaterialAttributes_1 {
   class=MaterialExpressionBlendMaterialAttributes
   inputs [
    A {from=MaterialExpressionNamedRerouteUsage_27, output=""}
    B {from=MaterialExpressionNamedRerouteUsage_26, output=""}
    Alpha {from=MaterialExpressionNamedRerouteUsage_53, output=""}
   ]
   properties {}
  }
  MaterialExpressionCameraVectorWS_5 {class=MaterialExpressionCameraVectorWS, inputs [], properties {}}
  MaterialExpressionChannelMaskParameter_0 {
   class=MaterialExpressionChannelMaskParameter
   inputs [None {from=MaterialExpressionNamedRerouteUsage_47, output=""}]
   properties {group=Ice | 01_Globals, maskChannel=Green, parameterName=Ice Mask Channel, sortPriority=10, defaultValue=<rgba 0 1 0 0>}
  }
  MaterialExpressionChannelMaskParameter_1 {
   class=MaterialExpressionChannelMaskParameter
   inputs [None {from=MaterialExpressionNamedRerouteUsage_52, output=""}]
   properties {group=Ice | 01_Globals, maskChannel=Blue, parameterName=Ice Surface Mask Channel, sortPriority=110, defaultValue=<rgba 0 0 1 0>}
  }
  MaterialExpressionChannelMaskParameter_2 {
   class=MaterialExpressionChannelMaskParameter
   inputs [None {from=MaterialExpressionNamedRerouteUsage_48, output=""}]
   properties {group=Ice | 01_Globals, parameterName=Emission Cutout Channel, sortPriority=230}
  }
  MaterialExpressionChannelMaskParameter_3 {
   class=MaterialExpressionChannelMaskParameter
   inputs [None {from=MaterialExpressionNamedRerouteUsage_7, output=""}]
   properties {group=Ice | 01_Globals, maskChannel=Green, parameterName=Parallax Thickness Channel, sortPriority=410, defaultValue=<rgba 0 1 0 0>}
  }
  MaterialExpressionClamp_0 {
   class=MaterialExpressionClamp
   inputs [
    None {from=MaterialExpressionScalarParameter_0, output=""}
    Min {from="", output=""}
    Max {from="", output=""}
   ]
   properties {maxDefault=5, minDefault=1}
  }
  MaterialExpressionComponentMask_0 {class=MaterialExpressionComponentMask, inputs [None {from=MaterialExpressionNamedRerouteUsage_35}], properties {g=true, r=true}}
  MaterialExpressionComponentMask_1 {class=MaterialExpressionComponentMask, inputs [None {from=MaterialExpressionNamedRerouteUsage_35}], properties {b=true}}
  MaterialExpressionComponentMask_5 {
   class=MaterialExpressionComponentMask
   inputs [None {from=MaterialExpressionNamedRerouteUsage_90}]
   properties {b=true, desc="dot(VertexN, RefrDir)"}
  }
  MaterialExpressionConstant_10 {class=MaterialExpressionConstant, properties {r=0.2}}
  MaterialExpressionConstant_11 {class=MaterialExpressionConstant}
  MaterialExpressionConstant_12 {class=MaterialExpressionConstant}
  MaterialExpressionConstant_13 {class=MaterialExpressionConstant}
  MaterialExpressionConstant_14 {class=MaterialExpressionConstant, properties {r=10000}}
  MaterialExpressionConstant_15 {class=MaterialExpressionConstant}
  MaterialExpressionConstant_4 {class=MaterialExpressionConstant, properties {r=1}}
  MaterialExpressionConstant_7 {class=MaterialExpressionConstant, properties {r=1.2}}
  MaterialExpressionConstant_9 {class=MaterialExpressionConstant, properties {r=0.55}}
  MaterialExpressionConvert_0 {
   class=MaterialExpressionConvert
   inputs [None {from=MaterialExpressionRgbToHsv_0}]
   properties {
    nodeName=Break Float3
    convertInputs [{
      type=Vector3
      expressionInput {
       expression=&self:MaterialExpressionRgbToHsv_0
       inputName=None
       mask=0
       maskA=0
       maskB=0
       maskG=0
       maskR=0
       outputIndex=0
      }
     }]
    convertOutputs [{type=Scalar}, {type=Scalar}, {type=Scalar}]
   }
  }
  MaterialExpressionConvert_1 {
   class=MaterialExpressionConvert
   inputs [
    None {from=MaterialExpressionDotProduct_3}
    None {from=MaterialExpressionDotProduct_5}
    None {from=MaterialExpressionDotProduct_6}
   ]
   properties {
    nodeName=Make Vector3
    convertInputs [
     {
      type=Scalar
      expressionInput {
       expression=&self:MaterialExpressionDotProduct_3
       inputName=None
       mask=0
       maskA=0
       maskB=0
       maskG=0
       maskR=0
       outputIndex=0
      }
     }
     {
      type=Scalar
      expressionInput {
       expression=&self:MaterialExpressionDotProduct_5
       inputName=None
       mask=0
       maskA=0
       maskB=0
       maskG=0
       maskR=0
       outputIndex=0
      }
     }
     {
      type=Scalar
      expressionInput {
       expression=&self:MaterialExpressionDotProduct_6
       inputName=None
       mask=0
       maskA=0
       maskB=0
       maskG=0
       maskR=0
       outputIndex=0
      }
     }
    ]
    convertOutputs [{type=Vector3}]
   }
  }
  MaterialExpressionConvert_8 {
   class=MaterialExpressionConvert
   inputs [
    None {from=MaterialExpressionFrac_0}
    None {from=MaterialExpressionSaturate_12}
    None {from=MaterialExpressionDivide_4}
   ]
   properties {
    nodeName=Make Vector3
    convertInputs [
     {
      type=Scalar
      expressionInput {
       expression=&self:MaterialExpressionFrac_0
       inputName=None
       mask=0
       maskA=0
       maskB=0
       maskG=0
       maskR=0
       outputIndex=0
      }
     }
     {
      type=Scalar
      expressionInput {
       expression=&self:MaterialExpressionSaturate_12
       inputName=None
       mask=0
       maskA=0
       maskB=0
       maskG=0
       maskR=0
       outputIndex=0
      }
     }
     {
      type=Scalar
      expressionInput {
       expression=&self:MaterialExpressionDivide_4
       inputName=None
       mask=0
       maskA=0
       maskB=0
       maskG=0
       maskR=0
       outputIndex=0
      }
     }
    ]
    convertOutputs [{type=Vector3}]
   }
  }
  MaterialExpressionDesaturation_0 {
   class=MaterialExpressionDesaturation
   inputs [
    None {from=MaterialExpressionMaterialFunctionCall_4, output=Result}
    Fraction {from=MaterialExpressionNamedRerouteUsage_114, output=""}
   ]
   properties {}
  }
  MaterialExpressionDesaturation_1 {
   class=MaterialExpressionDesaturation
   inputs [
    None {from=MaterialExpressionMaterialFunctionCall_1, output=Result}
    Fraction {from=MaterialExpressionNamedRerouteUsage_89, output=""}
   ]
   properties {}
  }
  MaterialExpressionDesaturation_2 {
   class=MaterialExpressionDesaturation
   inputs [
    None {from=MaterialExpressionMultiply_24, output=""}
    Fraction {from=MaterialExpressionNamedRerouteUsage_136, output=""}
   ]
   properties {}
  }
  MaterialExpressionDivide_0 {
   class=MaterialExpressionDivide
   inputs [
    A {from=""}
    B {from=MaterialExpressionNamedRerouteUsage_92}
   ]
  }
  MaterialExpressionDivide_1 {
   class=MaterialExpressionDivide
   inputs [
    A {from=MaterialExpressionNamedRerouteUsage_4}
    B {from=MaterialExpressionMax_2}
   ]
  }
  MaterialExpressionDivide_11 {
   class=MaterialExpressionDivide
   inputs [
    A {from=MaterialExpressionNamedRerouteUsage_101}
    B {from=MaterialExpressionNamedRerouteUsage_83}
   ]
  }
  MaterialExpressionDivide_12 {
   class=MaterialExpressionDivide
   inputs [
    A {from=MaterialExpressionNamedRerouteUsage_112}
    B {from=MaterialExpressionNamedRerouteUsage_156}
   ]
  }
  MaterialExpressionDivide_15 {
   class=MaterialExpressionDivide
   inputs [
    A {from=MaterialExpressionSwitch_1}
    B {from=MaterialExpressionNamedRerouteUsage_33}
   ]
  }
  MaterialExpressionDivide_2 {
   class=MaterialExpressionDivide
   inputs [
    A {from=MaterialExpressionSwitch_0}
    B {from=MaterialExpressionNamedRerouteUsage_24}
   ]
  }
  MaterialExpressionDivide_4 {
   class=MaterialExpressionDivide
   inputs [
    A {from=""}
    B {from=MaterialExpressionMax_10}
   ]
  }
  MaterialExpressionDivide_6 {
   class=MaterialExpressionDivide
   inputs [
    A {from=MaterialExpressionComponentMask_0}
    B {from=MaterialExpressionMax_11}
   ]
  }
  MaterialExpressionDivide_8 {
   class=MaterialExpressionDivide
   inputs [
    A {from=MaterialExpressionAdd_7}
    B {from=MaterialExpressionNamedRerouteUsage_12}
   ]
  }
  MaterialExpressionDivide_9 {
   class=MaterialExpressionDivide
   inputs [
    A {from=MaterialExpressionNamedRerouteUsage_29}
    B {from=MaterialExpressionNamedRerouteUsage_34}
   ]
  }
  MaterialExpressionDotProduct_1 {
   class=MaterialExpressionDotProduct
   inputs [
    A {from=MaterialExpressionVertexNormalWS_0, output=""}
    B {from=MaterialExpressionNamedRerouteUsage_67, output=""}
   ]
   properties {}
  }
  MaterialExpressionDotProduct_3 {
   class=MaterialExpressionDotProduct
   inputs [
    A {from=MaterialExpressionNamedRerouteDeclaration_9, output=""}
    B {from=MaterialExpressionMaterialFunctionCall_21, output=X U}
   ]
   properties {}
  }
  MaterialExpressionDotProduct_5 {
   class=MaterialExpressionDotProduct
   inputs [
    A {from=MaterialExpressionNamedRerouteDeclaration_9, output=""}
    B {from=MaterialExpressionMaterialFunctionCall_21, output=Y V}
   ]
   properties {}
  }
  MaterialExpressionDotProduct_6 {
   class=MaterialExpressionDotProduct
   inputs [
    A {from=MaterialExpressionNamedRerouteDeclaration_9, output=""}
    B {from=MaterialExpressionMaterialFunctionCall_21, output=Z W}
   ]
   properties {}
  }
  MaterialExpressionDotProduct_9 {
   class=MaterialExpressionDotProduct
   inputs [
    A {from=MaterialExpressionNamedRerouteUsage_129, output=""}
    B {from=MaterialExpressionNamedRerouteUsage_104, output=""}
   ]
   properties {}
  }
  MaterialExpressionFrac_0 {class=MaterialExpressionFrac, inputs [None {from=MaterialExpressionAdd_12, output=""}], properties {}}
  MaterialExpressionFunctionInput_0 {
   class=MaterialExpressionFunctionInput
   inputs [Preview {from="", output=""}]
   properties {inputName=Material, inputType=FunctionInput_MaterialAttributes}
  }
  MaterialExpressionFunctionInput_1 {class=MaterialExpressionFunctionInput, inputs [Preview {from="", output=""}], properties {inputName=Tech Art Mask, sortPriority=10}}
  MaterialExpressionFunctionInput_2 {
   class=MaterialExpressionFunctionInput
   inputs [Preview {from="", output=""}]
   properties {bUsePreviewValueAsDefault=true, inputName=Additional Interior Emission, sortPriority=20}
  }
  MaterialExpressionFunctionOutput_1 {
   class=MaterialExpressionFunctionOutput
   inputs [None {from=MaterialExpressionStaticSwitchParameter_3, output=""}]
   properties {outputName="Result (MA)"}
  }
  MaterialExpressionGetMaterialAttributes_1 {
   class=MaterialExpressionGetMaterialAttributes
   inputs [None {from=MaterialExpressionNamedRerouteUsage_30, output=""}]
   properties {attributeGetTypes [
     #1
     #2
     #3
     #4
     #5
    ]}
  }
  MaterialExpressionHsvToRgb_0 {class=MaterialExpressionHsvToRgb, inputs [None {from=MaterialExpressionConvert_8, output=""}], properties {}}
  MaterialExpressionLinearInterpolate_15 {
   class=MaterialExpressionLinearInterpolate
   inputs [
    A {from=MaterialExpressionNamedRerouteUsage_137}
    B {from=MaterialExpressionNamedRerouteUsage_130}
    Alpha {from=MaterialExpressionPower_11}
   ]
  }
  MaterialExpressionLinearInterpolate_19 {
   class=MaterialExpressionLinearInterpolate
   inputs [
    A {from=""}
    B {from=MaterialExpressionMultiply_20}
    Alpha {from=MaterialExpressionNamedRerouteUsage_163}
   ]
   properties {constA=1}
  }
  MaterialExpressionLinearInterpolate_2 {
   class=MaterialExpressionLinearInterpolate
   inputs [
    A {from=MaterialExpressionNamedRerouteUsage_2}
    B {from=MaterialExpressionMultiply_22}
    Alpha {from=MaterialExpressionPower_5}
   ]
  }
  MaterialExpressionLinearInterpolate_3 {
   class=MaterialExpressionLinearInterpolate
   inputs [
    A {from=MaterialExpressionNamedRerouteUsage_66}
    B {from=MaterialExpressionMultiply_21}
    Alpha {from=MaterialExpressionPower_3}
   ]
  }
  MaterialExpressionLinearInterpolate_4 {
   class=MaterialExpressionLinearInterpolate
   inputs [
    A {from=MaterialExpressionNamedRerouteUsage_17}
    B {from=MaterialExpressionSaturate_6}
    Alpha {from=MaterialExpressionNamedRerouteUsage_32}
   ]
  }
  MaterialExpressionLinearInterpolate_6 {
   class=MaterialExpressionLinearInterpolate
   inputs [
    A {from=MaterialExpressionSaturate_9}
    B {from=MaterialExpressionSaturate_10}
    Alpha {from=MaterialExpressionNamedRerouteUsage_93}
   ]
  }
  MaterialExpressionLinearInterpolate_9 {
   class=MaterialExpressionLinearInterpolate
   inputs [
    A {from=MaterialExpressionSaturate_18}
    B {from=MaterialExpressionNamedRerouteUsage_60}
    Alpha {from=""}
   ]
   properties {constAlpha=0.25, desc=to avoid glowing cracks}
  }
  MaterialExpressionMaterialFunctionCall_0 {
   class=MaterialExpressionMaterialFunctionCall
   inputs [
    Normal {from=MaterialExpressionNamedRerouteUsage_107}
    Direction {from=MaterialExpressionNamedRerouteUsage_0}
    Facing Ratio Exp {from=MaterialExpressionConstant_7}
    Parallel Intensity {from=MaterialExpressionConstant_9}
    Perpendicular Intensity {from=MaterialExpressionConstant_10}
   ]
   properties {materialFunction=/Game/Global/MaterialFunctions/TechArt/Refraction/MF_ThicknessSimple}
  }
  MaterialExpressionMaterialFunctionCall_1 {
   class=MaterialExpressionMaterialFunctionCall
   inputs [
    Hue Shift Percentage {from=MaterialExpressionNamedRerouteUsage_82}
    Texture {from=MaterialExpressionAdd_2}
   ]
   properties {materialFunction=&HueShift}
  }
  MaterialExpressionMaterialFunctionCall_10 {
   class=MaterialExpressionMaterialFunctionCall
   inputs [
    Normal {from=MaterialExpressionTextureSample_6, output=RGB}
    Intensity {from=MaterialExpressionLinearInterpolate_15}
   ]
   properties {materialFunction=/Game/Global/MaterialFunctions/MF_NormalIntensity}
  }
  MaterialExpressionMaterialFunctionCall_2 {
   class=MaterialExpressionMaterialFunctionCall
   inputs [
    Normal WS {from=MaterialExpressionNamedRerouteUsage_73}
    Sky Illuminance Mult {from=""}
    Sky Desaturation {from=MaterialExpressionConstant_4}
    Sun Illuminance Mult {from=""}
    Sun Desaturation {from=MaterialExpressionConstant_4}
   ]
   properties {materialFunction=/Game/Global/MaterialFunctions/TechArt/Refraction/MF_SunSkyProperties}
  }
  MaterialExpressionMaterialFunctionCall_21 {
   class=MaterialExpressionMaterialFunctionCall
   inputs []
   properties {materialFunction=&TangentBasis}
  }
  MaterialExpressionMaterialFunctionCall_27 {
   class=MaterialExpressionMaterialFunctionCall
   inputs [
    Input {from=MaterialExpressionTextureSample_15, output=R}
    Input Low {from=MaterialExpressionNamedRerouteUsage_31}
    Input High {from=MaterialExpressionNamedRerouteUsage_43}
   ]
   properties {materialFunction=/Engine/Functions/Engine_MaterialFunctions03/Math/RemapValueRangeNormalized}
  }
  MaterialExpressionMaterialFunctionCall_3 {
   class=MaterialExpressionMaterialFunctionCall
   inputs [
    Color_Morning {from=MaterialExpressionVectorParameter_7, output=RGB}
    Color_Day {from=MaterialExpressionVectorParameter_12, output=RGB}
    Color_Evening {from=MaterialExpressionVectorParameter_13, output=RGB}
    Color_Night {from=MaterialExpressionVectorParameter_15, output=RGB}
   ]
   properties {materialFunction=&Color_Time_of_Day}
  }
  MaterialExpressionMaterialFunctionCall_33 {
   class=MaterialExpressionMaterialFunctionCall
   inputs [
    View Vector {from=MaterialExpressionNamedRerouteUsage_103}
    Normal {from=MaterialExpressionNamedRerouteUsage_49}
    "Relative IOR (Eta)" {from=MaterialExpressionDivide_0}
   ]
   properties {materialFunction=/Game/Global/MaterialFunctions/TechArt/Refraction/MF_Refract}
  }
  MaterialExpressionMaterialFunctionCall_34 {
   class=MaterialExpressionMaterialFunctionCall
   inputs [
    Direction WS {from=MaterialExpressionNamedRerouteUsage_5}
    Roughness {from=MaterialExpressionNamedRerouteUsage_6}
    Sky Desaturation {from=MaterialExpressionNamedRerouteUsage_141}
    Sky Brightness {from=""}
    Use Ground Map {from=MaterialExpressionStaticBool_0}
    Ground LatLong Map {from=MaterialExpressionTextureObject_0}
    Group Map Rotation {from=""}
    Ground Exp {from=MaterialExpressionNamedRerouteUsage_117}
    Ground Brightness {from=MaterialExpressionNamedRerouteUsage_122}
   ]
   properties {materialFunction=/Game/Global/MaterialFunctions/TechArt/Refraction/MF_EnvironmentIlluminance}
  }
  MaterialExpressionMaterialFunctionCall_35 {
   class=MaterialExpressionMaterialFunctionCall
   inputs [
    Normal WS {from=MaterialExpressionNamedRerouteUsage_40}
    Roughness {from=MaterialExpressionNamedRerouteUsage_37}
    Surface Opacity {from=MaterialExpressionNamedRerouteUsage_16}
    Refraction Vector WS {from=MaterialExpressionNamedRerouteUsage_28}
    Refraction IOR {from=MaterialExpressionNamedRerouteUsage_38}
    Scene Luminance {from=MaterialExpressionNamedRerouteUsage_39}
    Interior Emission {from=MaterialExpressionNamedRerouteUsage_42}
    Volume Thickness {from=MaterialExpressionNamedRerouteUsage_41}
    Volume Density {from=MaterialExpressionNamedRerouteUsage_63}
    Absorption Coeffs {from=MaterialExpressionNamedRerouteUsage_50}
    Scattering Coeffs {from=MaterialExpressionNamedRerouteUsage_62}
    Scattering Phase {from=MaterialExpressionNamedRerouteUsage_70}
    Is Scene Inside Volume {from=MaterialExpressionConstant_15}
    Thin Transmission Exp {from=MaterialExpressionNamedRerouteUsage_64}
    //---- ALPHA-BLENDING ONLY ----// {from=""}
    Haze Amount {from=""}
    //----------- LIGHTING -----------// {from=""}
    Sky Illuminance {from=MaterialExpressionNamedRerouteUsage_74}
    Sun Illuminance {from=MaterialExpressionNamedRerouteUsage_98}
    Sun Direction {from=MaterialExpressionNamedRerouteUsage_100}
   ]
   properties {materialFunction=/Game/Global/MaterialFunctions/TechArt/Refraction/MF_DielectricInteriorVolume}
  }
  MaterialExpressionMaterialFunctionCall_4 {
   class=MaterialExpressionMaterialFunctionCall
   inputs [
    Hue Shift Percentage {from=MaterialExpressionNamedRerouteUsage_111}
    Texture {from=MaterialExpressionNamedRerouteUsage_57}
   ]
   properties {materialFunction=&HueShift}
  }
  MaterialExpressionMaterialFunctionCall_6 {
   class=MaterialExpressionMaterialFunctionCall
   inputs [
    Normal_1 {from=MaterialExpressionNamedRerouteUsage_18}
    Normal_2 {from=MaterialExpressionMaterialFunctionCall_10, output="Result (V3)"}
   ]
   properties {materialFunction=/Game/Global/MaterialFunctions/MF_NormalBlend_Whiteout}
  }
  MaterialExpressionMax_1 {
   class=MaterialExpressionMax
   inputs [
    A {from=MaterialExpressionNamedRerouteUsage_10}
    B {from=MaterialExpressionNamedRerouteUsage_46}
   ]
  }
  MaterialExpressionMax_10 {
   class=MaterialExpressionMax
   inputs [
    A {from=MaterialExpressionMultiply_8}
    B {from=""}
   ]
   properties {constB=0.0001}
  }
  MaterialExpressionMax_11 {
   class=MaterialExpressionMax
   inputs [
    A {from=MaterialExpressionComponentMask_1}
    B {from=""}
   ]
   properties {constB=0.001}
  }
  MaterialExpressionMax_2 {
   class=MaterialExpressionMax
   inputs [
    A {from=MaterialExpressionMultiply_44}
    B {from=""}
   ]
   properties {constB=0.001}
  }
  MaterialExpressionMax_6 {
   class=MaterialExpressionMax
   inputs [
    A {from=MaterialExpressionNamedRerouteUsage_3}
    B {from=""}
   ]
   properties {constB=0}
  }
  MaterialExpressionMax_9 {
   class=MaterialExpressionMax
   inputs [
    A {from=MaterialExpressionScalarParameter_40}
    B {from=""}
   ]
   properties {constB=0}
  }
  MaterialExpressionMultiply_0 {
   class=MaterialExpressionMultiply
   inputs [
    A {from=MaterialExpressionNamedRerouteUsage_15}
    B {from=MaterialExpressionDivide_2}
   ]
  }
  MaterialExpressionMultiply_1 {
   class=MaterialExpressionMultiply
   inputs [
    A {from=MaterialExpressionNamedRerouteUsage_118}
    B {from=MaterialExpressionChannelMaskParameter_1, output=R}
   ]
  }
  MaterialExpressionMultiply_10 {
   class=MaterialExpressionMultiply
   inputs [
    A {from=MaterialExpressionNamedRerouteUsage_99}
    B {from=""}
   ]
   properties {constB=0.5}
  }
  MaterialExpressionMultiply_11 {
   class=MaterialExpressionMultiply
   inputs [
    A {from=MaterialExpressionPower_12}
    B {from=MaterialExpressionNamedRerouteUsage_91}
   ]
  }
  MaterialExpressionMultiply_12 {
   class=MaterialExpressionMultiply
   inputs [
    A {from=MaterialExpressionNamedRerouteUsage_25}
    B {from=MaterialExpressionNamedRerouteUsage_45}
   ]
  }
  MaterialExpressionMultiply_13 {
   class=MaterialExpressionMultiply
   inputs [
    A {from=MaterialExpressionTime_0}
    B {from=MaterialExpressionNamedRerouteUsage_113}
   ]
  }
  MaterialExpressionMultiply_14 {
   class=MaterialExpressionMultiply
   inputs [
    A {from=MaterialExpressionNamedRerouteUsage_55}
    B {from=MaterialExpressionTextureSample_7, output=RGB}
   ]
  }
  MaterialExpressionMultiply_15 {
   class=MaterialExpressionMultiply
   inputs [
    A {from=MaterialExpressionNamedRerouteUsage_162}
    B {from=MaterialExpressionLinearInterpolate_19}
   ]
  }
  MaterialExpressionMultiply_16 {
   class=MaterialExpressionMultiply
   inputs [
    A {from=MaterialExpressionNamedRerouteUsage_36}
    B {from=MaterialExpressionMultiply_62}
   ]
  }
  MaterialExpressionMultiply_17 {
   class=MaterialExpressionMultiply
   inputs [
    A {from=MaterialExpressionGetMaterialAttributes_1, output=EmissiveColor}
    B {from=MaterialExpressionNamedRerouteUsage_65}
   ]
  }
  MaterialExpressionMultiply_18 {
   class=MaterialExpressionMultiply
   inputs [
    A {from=MaterialExpressionPower_7}
    B {from=MaterialExpressionNamedRerouteUsage_11}
   ]
  }
  MaterialExpressionMultiply_19 {
   class=MaterialExpressionMultiply
   inputs [
    A {from=MaterialExpressionScalarParameter_3}
    B {from=MaterialExpressionChannelMaskParameter_0, output=R}
   ]
  }
  MaterialExpressionMultiply_2 {
   class=MaterialExpressionMultiply
   inputs [
    A {from=MaterialExpressionMultiply_13}
    B {from=""}
   ]
   properties {constB=0.004}
  }
  MaterialExpressionMultiply_20 {
   class=MaterialExpressionMultiply
   inputs [
    A {from=MaterialExpressionNamedRerouteUsage_8}
    B {from=MaterialExpressionNamedRerouteUsage_8}
   ]
  }
  MaterialExpressionMultiply_21 {
   class=MaterialExpressionMultiply
   inputs [
    A {from=MaterialExpressionNamedRerouteUsage_66}
    B {from=MaterialExpressionNamedRerouteUsage_69}
   ]
  }
  MaterialExpressionMultiply_22 {
   class=MaterialExpressionMultiply
   inputs [
    A {from=MaterialExpressionNamedRerouteUsage_2}
    B {from=MaterialExpressionNamedRerouteUsage_9}
   ]
  }
  MaterialExpressionMultiply_24 {
   class=MaterialExpressionMultiply
   inputs [
    A {from=MaterialExpressionNamedRerouteUsage_131}
    B {from=MaterialExpressionNamedRerouteUsage_132}
   ]
  }
  MaterialExpressionMultiply_25 {
   class=MaterialExpressionMultiply
   inputs [
    A {from=MaterialExpressionChannelMaskParameter_1, output=R}
    B {from=MaterialExpressionNamedRerouteUsage_120}
   ]
  }
  MaterialExpressionMultiply_26 {
   class=MaterialExpressionMultiply
   inputs [
    A {from=MaterialExpressionNamedRerouteUsage_126}
    B {from=MaterialExpressionOneMinus_3}
   ]
  }
  MaterialExpressionMultiply_29 {
   class=MaterialExpressionMultiply
   inputs [
    A {from=MaterialExpressionDesaturation_2}
    B {from=MaterialExpressionNamedRerouteUsage_133}
   ]
  }
  MaterialExpressionMultiply_3 {
   class=MaterialExpressionMultiply
   inputs [
    A {from=MaterialExpressionNamedRerouteUsage_1}
    B {from=""}
   ]
   properties {constB=-0.75}
  }
  MaterialExpressionMultiply_30 {
   class=MaterialExpressionMultiply
   inputs [
    A {from=MaterialExpressionVectorParameter_1, output=RGB}
    B {from=MaterialExpressionVectorParameter_1, output=RGB}
   ]
  }
  MaterialExpressionMultiply_34 {
   class=MaterialExpressionMultiply
   inputs [
    A {from=MaterialExpressionConvert_0}
    B {from=MaterialExpressionNamedRerouteUsage_96}
   ]
  }
  MaterialExpressionMultiply_35 {
   class=MaterialExpressionMultiply
   inputs [
    A {from=MaterialExpressionDivide_9}
    B {from=MaterialExpressionNamedRerouteUsage_106}
   ]
   properties {desc=UV Parallax Offset}
  }
  MaterialExpressionMultiply_38 {
   class=MaterialExpressionMultiply
   inputs [
    A {from=MaterialExpressionPower_6}
    B {from=MaterialExpressionNamedRerouteUsage_87}
   ]
  }
  MaterialExpressionMultiply_4 {
   class=MaterialExpressionMultiply
   inputs [
    A {from=MaterialExpressionNamedRerouteUsage_123}
    B {from=MaterialExpressionChannelMaskParameter_2, output=R}
   ]
  }
  MaterialExpressionMultiply_44 {
   class=MaterialExpressionMultiply
   inputs [
    A {from=MaterialExpressionComponentMask_5}
    B {from=""}
   ]
   properties {constB=-1, desc=cosTheta}
  }
  MaterialExpressionMultiply_47 {
   class=MaterialExpressionMultiply
   inputs [
    A {from=MaterialExpressionMultiply_66}
    B {from=MaterialExpressionNamedRerouteUsage_153}
   ]
  }
  MaterialExpressionMultiply_5 {
   class=MaterialExpressionMultiply
   inputs [
    A {from=MaterialExpressionNamedRerouteUsage_44}
    B {from=MaterialExpressionMaterialFunctionCall_35, output="Refraction Result - Opaque (V3)"}
   ]
  }
  MaterialExpressionMultiply_53 {
   class=MaterialExpressionMultiply
   inputs [
    A {from=MaterialExpressionNamedRerouteUsage_128}
    B {from=MaterialExpressionDivide_15}
   ]
  }
  MaterialExpressionMultiply_6 {
   class=MaterialExpressionMultiply
   inputs [
    A {from=MaterialExpressionNamedRerouteUsage_17}
    B {from=MaterialExpressionNamedRerouteUsage_51}
   ]
  }
  MaterialExpressionMultiply_62 {
   class=MaterialExpressionMultiply
   inputs [
    A {from=MaterialExpressionMaterialFunctionCall_34, output="Environment Illuminance (V3)"}
    B {from=MaterialExpressionMax_6}
   ]
  }
  MaterialExpressionMultiply_66 {
   class=MaterialExpressionMultiply
   inputs [
    A {from=MaterialExpressionNamedRerouteUsage_22}
    B {from=MaterialExpressionPower_9}
   ]
  }
  MaterialExpressionMultiply_7 {
   class=MaterialExpressionMultiply
   inputs [
    A {from=MaterialExpressionPower_10}
    B {from=MaterialExpressionNamedRerouteUsage_105}
   ]
  }
  MaterialExpressionMultiply_71 {
   class=MaterialExpressionMultiply
   inputs [
    A {from=MaterialExpressionNamedRerouteUsage_134}
    B {from=MaterialExpressionPower_18}
   ]
  }
  MaterialExpressionMultiply_8 {
   class=MaterialExpressionMultiply
   inputs [
    A {from=MaterialExpressionConvert_0}
    B {from=MaterialExpressionNamedRerouteUsage_109}
   ]
  }
  MaterialExpressionMultiply_9 {
   class=MaterialExpressionMultiply
   inputs [
    A {from=MaterialExpressionPower_4}
    B {from=MaterialExpressionNamedRerouteUsage_155}
   ]
  }
  MaterialExpressionNamedRerouteDeclaration_0 {
   class=MaterialExpressionNamedRerouteDeclaration
   inputs [None {from=MaterialExpressionMultiply_16}]
   properties {name="Refracted Illuminance (V3)", nodeColor=<rgba 0.9688 0 1 1>}
  }
  MaterialExpressionNamedRerouteDeclaration_1 {
   class=MaterialExpressionNamedRerouteDeclaration
   inputs [None {from=MaterialExpressionNormalize_0}]
   properties {name="Surface Normal WS (V3)", nodeColor=<rgba 0 0.5913 1 1>}
  }
  MaterialExpressionNamedRerouteDeclaration_10 {
   class=MaterialExpressionNamedRerouteDeclaration
   inputs [None {from=MaterialExpressionOneMinus_0}]
   properties {name="Parallax Mask (S)", nodeColor=<rgba 1 0.5859 0 1>}
  }
  MaterialExpressionNamedRerouteDeclaration_100 {
   class=MaterialExpressionNamedRerouteDeclaration
   inputs [None {from=MaterialExpressionScalarParameter_50}]
   properties {name="Surface Detail UV Scale (S)", nodeColor=<rgba 0 1 0.1328 1>}
  }
  MaterialExpressionNamedRerouteDeclaration_101 {
   class=MaterialExpressionNamedRerouteDeclaration
   inputs [None {from=MaterialExpressionScalarParameter_7}]
   properties {name="Ice Out Emissive <-> Diffuse (S)", nodeColor=<rgba 0.5469 0 1 1>}
  }
  MaterialExpressionNamedRerouteDeclaration_102 {
   class=MaterialExpressionNamedRerouteDeclaration
   inputs [None {from=MaterialExpressionScalarParameter_23}]
   properties {name="Ice Diffuse Out Desaturation (S)", nodeColor=<rgba 0 1 0.2266 1>}
  }
  MaterialExpressionNamedRerouteDeclaration_103 {
   class=MaterialExpressionNamedRerouteDeclaration
   inputs [None {from=MaterialExpressionScalarParameter_24}]
   properties {name="Ice Diffuse Out Gain (S)", nodeColor=<rgba 1 0 0.0937 1>}
  }
  MaterialExpressionNamedRerouteDeclaration_104 {
   class=MaterialExpressionNamedRerouteDeclaration
   inputs [None {from=MaterialExpressionMaterialFunctionCall_2, output="Sky Illuminance (V3)"}]
   properties {name="Lighting - Sky Illuminance (V3)", nodeColor=<rgba 0 1 0.4141 1>}
  }
  MaterialExpressionNamedRerouteDeclaration_105 {
   class=MaterialExpressionNamedRerouteDeclaration
   inputs [None {from=MaterialExpressionScalarParameter_1}]
   properties {name="Diffuse Constant Depth (S)", nodeColor=<rgba 0.0781 1 0 1>}
  }
  MaterialExpressionNamedRerouteDeclaration_106 {
   class=MaterialExpressionNamedRerouteDeclaration
   inputs [None {from=MaterialExpressionScalarParameter_47}]
   properties {name="Diffuse Constant Depth Attenuation (S)", nodeColor=<rgba 1 0 0.3984 1>}
  }
  MaterialExpressionNamedRerouteDeclaration_107 {
   class=MaterialExpressionNamedRerouteDeclaration
   inputs [None {from=MaterialExpressionTextureObjectParameter_3}]
   properties {name="Surface Detail Roughness Map (TO)", nodeColor=<rgba 0.125 1 0 1>}
  }
  MaterialExpressionNamedRerouteDeclaration_108 {
   class=MaterialExpressionNamedRerouteDeclaration
   inputs [None {from=MaterialExpressionTextureObjectParameter_4}]
   properties {name="Surface Detail Normal Map (TO)", nodeColor=<rgba 0.6406 1 0 1>}
  }
  MaterialExpressionNamedRerouteDeclaration_109 {
   class=MaterialExpressionNamedRerouteDeclaration
   inputs [None {from=MaterialExpressionTextureObjectParameter_0}]
   properties {name="Emissive Map (TO)", nodeColor=<rgba 1 0 0.5156 1>}
  }
  MaterialExpressionNamedRerouteDeclaration_11 {
   class=MaterialExpressionNamedRerouteDeclaration
   inputs [None {from=MaterialExpressionChannelMaskParameter_3, output=R}]
   properties {name="Parallax Thickness Map (S)", nodeColor=<rgba 1 0.7734 0 1>}
  }
  MaterialExpressionNamedRerouteDeclaration_115 {
   class=MaterialExpressionNamedRerouteDeclaration
   inputs [None {from=MaterialExpressionVectorParameter_8, output=Breakup Pattern Mult}]
   properties {name="Noise Depth Variation Mult (S)", nodeColor=<rgba 0 0.0156 1 1>}
  }
  MaterialExpressionNamedRerouteDeclaration_116 {
   class=MaterialExpressionNamedRerouteDeclaration
   inputs [None {from=MaterialExpressionVectorParameter_8, output=Add Constant Depth}]
   properties {name="Noise Add Constant Depth (S)", nodeColor=<rgba 0.3359 1 0 1>}
  }
  MaterialExpressionNamedRerouteDeclaration_117 {
   class=MaterialExpressionNamedRerouteDeclaration
   inputs [None {from=MaterialExpressionMaterialFunctionCall_3, output=Result}]
   properties {name="TOD Color (V3)", nodeColor=<rgba 0.5938 0 1 1>}
  }
  MaterialExpressionNamedRerouteDeclaration_118 {
   class=MaterialExpressionNamedRerouteDeclaration
   inputs [None {from=MaterialExpressionMultiply_71}]
   properties {name="Emissive Fresnel (V3)", nodeColor=<rgba 1 0 0.4687 1>}
  }
  MaterialExpressionNamedRerouteDeclaration_119 {
   class=MaterialExpressionNamedRerouteDeclaration
   inputs [None {from=MaterialExpressionMultiply_14}]
   properties {name="Emission Map Lookup (V3)", nodeColor=<rgba 0 0.7891 1 1>}
  }
  MaterialExpressionNamedRerouteDeclaration_12 {
   class=MaterialExpressionNamedRerouteDeclaration
   inputs [None {from=MaterialExpressionScalarParameter_92}]
   properties {name=Refraction Intensity, nodeColor=<rgba 0 1 0.6953 1>}
  }
  MaterialExpressionNamedRerouteDeclaration_120 {
   class=MaterialExpressionNamedRerouteDeclaration
   inputs [None {from=MaterialExpressionSaturate_2}]
   properties {name="Diffuse Use Parallax Map (S)", nodeColor=<rgba 0 0.3906 1 1>}
  }
  MaterialExpressionNamedRerouteDeclaration_121 {
   class=MaterialExpressionNamedRerouteDeclaration
   inputs [None {from=MaterialExpressionScalarParameter_26}]
   properties {name="Interior Noise Speed (S)", nodeColor=<rgba 0 1 0.0625 1>}
  }
  MaterialExpressionNamedRerouteDeclaration_122 {
   class=MaterialExpressionNamedRerouteDeclaration
   inputs [None {from=MaterialExpressionSaturate_3}]
   properties {name="Emission Use Parallax Map (S)", nodeColor=<rgba 0 0.3906 1 1>}
  }
  MaterialExpressionNamedRerouteDeclaration_124 {
   class=MaterialExpressionNamedRerouteDeclaration
   inputs [None {from=MaterialExpressionScalarParameter_2}]
   properties {name="Refracted Sky Desaturation (S)", nodeColor=<rgba 1 0.7734 0 1>}
  }
  MaterialExpressionNamedRerouteDeclaration_13 {
   class=MaterialExpressionNamedRerouteDeclaration
   inputs [None {from=MaterialExpressionDivide_1}]
   properties {name="View Dependent Thickness (S)", nodeColor=<rgba 1 0 0.5156 1>}
  }
  MaterialExpressionNamedRerouteDeclaration_14 {
   class=MaterialExpressionNamedRerouteDeclaration
   inputs [None {from=MaterialExpressionGetMaterialAttributes_1, output=Specular}]
   properties {name="Surface Specular In (S)", nodeColor=<rgba 1 0 0.5156 1>}
  }
  MaterialExpressionNamedRerouteDeclaration_15 {
   class=MaterialExpressionNamedRerouteDeclaration
   inputs [None {from=MaterialExpressionAdd_18}]
   properties {name="Noise UV (V2)", nodeColor=<rgba 0 1 0.3203 1>}
  }
  MaterialExpressionNamedRerouteDeclaration_16 {
   class=MaterialExpressionNamedRerouteDeclaration
   inputs [None {from=MaterialExpressionConstant_14}]
   properties {name="Parallax Depth Norm (S)", nodeColor=<rgba 0 0.8594 1 1>}
  }
  MaterialExpressionNamedRerouteDeclaration_17 {
   class=MaterialExpressionNamedRerouteDeclaration
   inputs [None {from=MaterialExpressionScalarParameter_39}]
   properties {name="Thickness Exp (S)", nodeColor=<rgba 1 0.375 0 1>}
  }
  MaterialExpressionNamedRerouteDeclaration_18 {
   class=MaterialExpressionNamedRerouteDeclaration
   inputs [None {from=MaterialExpressionGetMaterialAttributes_1, output=Roughness}]
   properties {name="Surface Roughness In (S)", nodeColor=<rgba 1 0 0.5156 1>}
  }
  MaterialExpressionNamedRerouteDeclaration_19 {
   class=MaterialExpressionNamedRerouteDeclaration
   inputs [None {from=MaterialExpressionScalarParameter_9}]
   properties {name="Refraction Roughness (S)", nodeColor=<rgba 0.1719 0 1 1>}
  }
  MaterialExpressionNamedRerouteDeclaration_2 {
   class=MaterialExpressionNamedRerouteDeclaration
   inputs [None {from=MaterialExpressionDesaturation_1}]
   properties {name="Refracted Emission (V3)", nodeColor=<rgba 0 1 0.6484 1>}
  }
  MaterialExpressionNamedRerouteDeclaration_20 {
   class=MaterialExpressionNamedRerouteDeclaration
   inputs [None {from=MaterialExpressionPower_8}]
   properties {name="Refraction Depth Variation (S)", nodeColor=<rgba 1 0 0 1>}
  }
  MaterialExpressionNamedRerouteDeclaration_21 {
   class=MaterialExpressionNamedRerouteDeclaration
   inputs [None {from=MaterialExpressionPower_1}]
   properties {name="Ice Mask In (S)", nodeColor=<rgba 1 0.8438 0 1>}
  }
  MaterialExpressionNamedRerouteDeclaration_22 {
   class=MaterialExpressionNamedRerouteDeclaration
   inputs [None {from=MaterialExpressionSaturate_4}]
   properties {name="Surface Opacity In (S)", nodeColor=<rgba 0.5234 0 1 1>}
  }
  MaterialExpressionNamedRerouteDeclaration_23 {
   class=MaterialExpressionNamedRerouteDeclaration
   inputs [None {from=MaterialExpressionOneMinus_1}]
   properties {name="Emission Mask (S)", nodeColor=<rgba 1 0 0.9609 1>}
  }
  MaterialExpressionNamedRerouteDeclaration_24 {
   class=MaterialExpressionNamedRerouteDeclaration
   inputs [None {from=MaterialExpressionMax_9}]
   properties {name="Max Thickness (S)", nodeColor=<rgba 0.0547 0 1 1>}
  }
  MaterialExpressionNamedRerouteDeclaration_25 {
   class=MaterialExpressionNamedRerouteDeclaration
   inputs [None {from=MaterialExpressionAdd_1}]
   properties {name="Emission Out (V3)", nodeColor=<rgba 0.7578 1 0 1>}
  }
  MaterialExpressionNamedRerouteDeclaration_26 {
   class=MaterialExpressionNamedRerouteDeclaration
   inputs [None {from=MaterialExpressionSaturate_11}]
   properties {name="Interface Opacity (S)", nodeColor=<rgba 0.7109 0 1 1>}
  }
  MaterialExpressionNamedRerouteDeclaration_27 {
   class=MaterialExpressionNamedRerouteDeclaration
   inputs [None {from=MaterialExpressionMaterialFunctionCall_0, output="Thickness Normalized (S)"}]
   properties {name="Normalized Thickness (S)", nodeColor=<rgba 0 0.6953 1 1>}
  }
  MaterialExpressionNamedRerouteDeclaration_28 {
   class=MaterialExpressionNamedRerouteDeclaration
   inputs [None {from=MaterialExpressionVectorParameter_5, output=Pattern Scale}]
   properties {name="Interior Noise Scale (S)", nodeColor=<rgba 1 0 0.0937 1>}
  }
  MaterialExpressionNamedRerouteDeclaration_29 {
   class=MaterialExpressionNamedRerouteDeclaration
   inputs [None {from=MaterialExpressionMultiply_12}]
   properties {name="Base Color Out (V3)", nodeColor=<rgba 0 0.9297 1 1>}
  }
  MaterialExpressionNamedRerouteDeclaration_3 {
   class=MaterialExpressionNamedRerouteDeclaration
   inputs [None {from=MaterialExpressionFunctionInput_0}]
   properties {name="Material In (MA)", nodeColor=<rgba 1 0.7969 0 1>}
  }
  MaterialExpressionNamedRerouteDeclaration_30 {
   class=MaterialExpressionNamedRerouteDeclaration
   inputs [None {from=MaterialExpressionNamedRerouteUsage_150}]
   properties {name="Specular Out (S)", nodeColor=<rgba 1 0.75 0 1>}
  }
  MaterialExpressionNamedRerouteDeclaration_31 {
   class=MaterialExpressionNamedRerouteDeclaration
   inputs [None {from=MaterialExpressionNamedRerouteUsage_149}]
   properties {name="Roughness Out (S)", nodeColor=<rgba 0.4297 0 1 1>}
  }
  MaterialExpressionNamedRerouteDeclaration_32 {
   class=MaterialExpressionNamedRerouteDeclaration
   inputs [None {from=MaterialExpressionScalarParameter_118}]
   properties {name="Emission Constant Depth (S)", nodeColor=<rgba 0.5703 1 0 1>}
  }
  MaterialExpressionNamedRerouteDeclaration_33 {
   class=MaterialExpressionNamedRerouteDeclaration
   inputs [None {from=MaterialExpressionDivide_12}]
   properties {name="Ice Surface UV (V2)", nodeColor=<rgba 1 0 0.2109 1>}
  }
  MaterialExpressionNamedRerouteDeclaration_34 {
   class=MaterialExpressionNamedRerouteDeclaration
   inputs [None {from=MaterialExpressionVectorParameter_0, output=Enable Remap}]
   properties {name="Enable Parallax Map Remap (S)", nodeColor=<rgba 0.4531 0 1 1>}
  }
  MaterialExpressionNamedRerouteDeclaration_35 {
   class=MaterialExpressionNamedRerouteDeclaration
   inputs [None {from=MaterialExpressionVectorParameter_14, output=RGB}]
   properties {name="Refraction Tint (V3)", nodeColor=<rgba 0 0.1562 1 1>}
  }
  MaterialExpressionNamedRerouteDeclaration_36 {
   class=MaterialExpressionNamedRerouteDeclaration
   inputs [None {from=MaterialExpressionMaterialFunctionCall_2, output="Sun Direction (V3)"}]
   properties {name="Lighting - Sun Direction (V3)", nodeColor=<rgba 0 0.0625 1 1>}
  }
  MaterialExpressionNamedRerouteDeclaration_37 {
   class=MaterialExpressionNamedRerouteDeclaration
   inputs [None {from=MaterialExpressionMaterialFunctionCall_2, output="Sun Illuminance (V3)"}]
   properties {name="Lighting - Sun Illuminance (V3)", nodeColor=<rgba 0.7344 0 1 1>}
  }
  MaterialExpressionNamedRerouteDeclaration_38 {
   class=MaterialExpressionNamedRerouteDeclaration
   inputs [None {from=MaterialExpressionNamedRerouteUsage_143}]
   properties {name="Normal Out TS (V3)", nodeColor=<rgba 1 0.1875 0 1>}
  }
  MaterialExpressionNamedRerouteDeclaration_39 {
   class=MaterialExpressionNamedRerouteDeclaration
   inputs [None {from=MaterialExpressionHsvToRgb_0}]
   properties {name="Volume Absorption Coeffs (V3)", nodeColor=<rgba 1 0.7266 0 1>}
  }
  MaterialExpressionNamedRerouteDeclaration_4 {
   class=MaterialExpressionNamedRerouteDeclaration
   inputs [None {from=MaterialExpressionSetMaterialAttributes_3}]
   properties {name="Material Out (MA)", nodeColor=<rgba 0.3125 1 0 1>}
  }
  MaterialExpressionNamedRerouteDeclaration_40 {
   class=MaterialExpressionNamedRerouteDeclaration
   inputs [None {from=MaterialExpressionScalarParameter_10}]
   properties {name="Ice Mask Min (S)", nodeColor=<rgba 1 0.9375 0 1>}
  }
  MaterialExpressionNamedRerouteDeclaration_41 {
   class=MaterialExpressionNamedRerouteDeclaration
   inputs [None {from=MaterialExpressionFunctionInput_1}]
   properties {name="Tech Art Mask In (V3)", nodeColor=<rgba 0.875 0 1 1>}
  }
  MaterialExpressionNamedRerouteDeclaration_42 {
   class=MaterialExpressionNamedRerouteDeclaration
   inputs [None {from=MaterialExpressionGetMaterialAttributes_1, output=BaseColor}]
   properties {name="Surface Color In (V3)", nodeColor=<rgba 1 0.1172 0 1>}
  }
  MaterialExpressionNamedRerouteDeclaration_43 {
   class=MaterialExpressionNamedRerouteDeclaration
   inputs [None {from=MaterialExpressionMax_1}]
   properties {name="Mask Out (S)", nodeColor=<rgba 1 0.3984 0 1>}
  }
  MaterialExpressionNamedRerouteDeclaration_44 {
   class=MaterialExpressionNamedRerouteDeclaration
   inputs [None {from=MaterialExpressionMultiply_30}]
   properties {name="Volume Scattering Coeffs (V3)", nodeColor=<rgba 0.4062 0 1 1>}
  }
  MaterialExpressionNamedRerouteDeclaration_45 {
   class=MaterialExpressionNamedRerouteDeclaration
   inputs [None {from=MaterialExpressionMultiply_17}]
   properties {name="Emissive Color In (V3)", nodeColor=<rgba 0 1 0.6953 1>}
  }
  MaterialExpressionNamedRerouteDeclaration_46 {
   class=MaterialExpressionNamedRerouteDeclaration
   inputs [None {from=MaterialExpressionSaturate_0}]
   properties {name="DIffuse Min Depth Mult (S)", nodeColor=<rgba 0 0.7891 1 1>}
  }
  MaterialExpressionNamedRerouteDeclaration_47 {
   class=MaterialExpressionNamedRerouteDeclaration
   inputs [None {from=MaterialExpressionScalarParameter_76}]
   properties {name="Volume Density (S)", nodeColor=<rgba 0 1 0.9531 1>}
  }
  MaterialExpressionNamedRerouteDeclaration_48 {
   class=MaterialExpressionNamedRerouteDeclaration
   inputs [None {from=MaterialExpressionScalarParameter_78}]
   properties {name="Transmission Attenuation (S)", nodeColor=<rgba 1 0.6328 0 1>}
  }
  MaterialExpressionNamedRerouteDeclaration_49 {
   class=MaterialExpressionNamedRerouteDeclaration
   inputs [None {from=MaterialExpressionScalarParameter_77}]
   properties {name="Volume Scattering Phase (S)", nodeColor=<rgba 0.3125 0 1 1>}
  }
  MaterialExpressionNamedRerouteDeclaration_5 {
   class=MaterialExpressionNamedRerouteDeclaration
   inputs [None {from=MaterialExpressionVectorParameter_2, output=Max Depth}]
   properties {name="DIffuse Max Depth (S)", nodeColor=<rgba 1 0 0.4687 1>}
  }
  MaterialExpressionNamedRerouteDeclaration_50 {
   class=MaterialExpressionNamedRerouteDeclaration
   inputs [None {from=MaterialExpressionVectorParameter_2, output=Thickness Map Exp}]
   properties {name="Diffuse Thickness Map Exp (S)", nodeColor=<rgba 1 0.8906 0 1>}
  }
  MaterialExpressionNamedRerouteDeclaration_51 {
   class=MaterialExpressionNamedRerouteDeclaration
   inputs [None {from=MaterialExpressionConvert_1}]
   properties {name="Refraction Vector TS (V3)", nodeColor=<rgba 0 1 0.3203 1>}
  }
  MaterialExpressionNamedRerouteDeclaration_52 {
   class=MaterialExpressionNamedRerouteDeclaration
   inputs [None {from=MaterialExpressionVectorParameter_3, output=Max Depth}]
   properties {name="Interior Emission Max Depth (S)", nodeColor=<rgba 1 0 0.4688 1>}
  }
  MaterialExpressionNamedRerouteDeclaration_53 {
   class=MaterialExpressionNamedRerouteDeclaration
   inputs [None {from=MaterialExpressionSaturate_1}]
   properties {name="Interior Emission Min Depth Mult (S)", nodeColor=<rgba 0 0.7891 1 1>}
  }
  MaterialExpressionNamedRerouteDeclaration_54 {
   class=MaterialExpressionNamedRerouteDeclaration
   inputs [None {from=MaterialExpressionVectorParameter_3, output=Thickness Map Exp}]
   properties {name="Interior Emission Thickness Map Exp (S)", nodeColor=<rgba 1 0.8906 0 1>}
  }
  MaterialExpressionNamedRerouteDeclaration_55 {
   class=MaterialExpressionNamedRerouteDeclaration
   inputs [None {from=MaterialExpressionScalarParameter_6}]
   properties {name="Emission Hue Shift (S)", nodeColor=<rgba 1 0.8203 0 1>}
  }
  MaterialExpressionNamedRerouteDeclaration_56 {
   class=MaterialExpressionNamedRerouteDeclaration
   inputs [None {from=MaterialExpressionFunctionInput_2}]
   properties {name="Additional Interior Emission (V3)", nodeColor=<rgba 1 0 0.3047 1>}
  }
  MaterialExpressionNamedRerouteDeclaration_57 {
   class=MaterialExpressionNamedRerouteDeclaration
   inputs [None {from=MaterialExpressionVectorParameter_5, output=Pattern OldMin}]
   properties {name="Interior Noise OldMin (S)", nodeColor=<rgba 0 0.4141 1 1>}
  }
  MaterialExpressionNamedRerouteDeclaration_58 {
   class=MaterialExpressionNamedRerouteDeclaration
   inputs [None {from=MaterialExpressionSaturate_8}]
   properties {name="Facing Ratio VtxN (S)", nodeColor=<rgba 0 1 0.6484 1>}
  }
  MaterialExpressionNamedRerouteDeclaration_59 {
   class=MaterialExpressionNamedRerouteDeclaration
   inputs [None {from=MaterialExpressionVectorParameter_5, output=Pattern OldMax}]
   properties {name="Interior Noise OldMax (S)", nodeColor=<rgba 0.7344 1 0 1>}
  }
  MaterialExpressionNamedRerouteDeclaration_6 {
   class=MaterialExpressionNamedRerouteDeclaration
   inputs [None {from=MaterialExpressionGetMaterialAttributes_1, output=Normal}]
   properties {name="Surface Normal In TS (V3)", nodeColor=<rgba 0.4297 0 1 1>}
  }
  MaterialExpressionNamedRerouteDeclaration_60 {
   class=MaterialExpressionNamedRerouteDeclaration
   inputs [None {from=MaterialExpressionVectorParameter_10, output=Hue Shift}]
   properties {name="Diffuse Map Hue Shift (S)", nodeColor=<rgba 0.1953 1 0 1>}
  }
  MaterialExpressionNamedRerouteDeclaration_61 {
   class=MaterialExpressionNamedRerouteDeclaration
   inputs [None {from=MaterialExpressionDivide_6}]
   properties {name="Refracted Unit UV Offset TS (V2)", nodeColor=<rgba 1 0 0.1875 1>}
  }
  MaterialExpressionNamedRerouteDeclaration_62 {
   class=MaterialExpressionNamedRerouteDeclaration
   inputs [None {from=MaterialExpressionVectorParameter_4, output=Mix Amount}]
   properties {name="Detail Roughness Mix (S)", nodeColor=<rgba 0 0.9297 1 1>}
  }
  MaterialExpressionNamedRerouteDeclaration_63 {
   class=MaterialExpressionNamedRerouteDeclaration
   inputs [None {from=MaterialExpressionMultiply_2}]
   properties {name="Time (S)", nodeColor=<rgba 0 0.7656 1 1>}
  }
  MaterialExpressionNamedRerouteDeclaration_64 {
   class=MaterialExpressionNamedRerouteDeclaration
   inputs [None {from=MaterialExpressionTextureCoordinate_0}]
   properties {name="UV Tiles (V2)", nodeColor=<rgba 1 0.5859 0 1>}
  }
  MaterialExpressionNamedRerouteDeclaration_65 {
   class=MaterialExpressionNamedRerouteDeclaration
   inputs [None {from=MaterialExpressionCameraVectorWS_5}]
   properties {name="Camera Vector WS (V3)", nodeColor=<rgba 1 0 0 1>}
  }
  MaterialExpressionNamedRerouteDeclaration_66 {
   class=MaterialExpressionNamedRerouteDeclaration
   inputs [None {from=MaterialExpressionTextureSampleParameter2D_0, output=RGB}]
   properties {name="Diffuse Map Lookup (V3)", nodeColor=<rgba 0.6172 1 0 1>}
  }
  MaterialExpressionNamedRerouteDeclaration_67 {
   class=MaterialExpressionNamedRerouteDeclaration
   inputs [None {from=MaterialExpressionVectorParameter_6, output=Roughness Exp}]
   properties {name="Asset Roughness Exp (S)", nodeColor=<rgba 1 0.75 0 1>}
  }
  MaterialExpressionNamedRerouteDeclaration_68 {
   class=MaterialExpressionNamedRerouteDeclaration
   inputs [None {from=MaterialExpressionVectorParameter_6, output=Roughness Mult}]
   properties {name="Asset Roughness Mult (S)", nodeColor=<rgba 0.4297 0 1 1>}
  }
  MaterialExpressionNamedRerouteDeclaration_69 {
   class=MaterialExpressionNamedRerouteDeclaration
   inputs [None {from=MaterialExpressionVectorParameter_6, output=Specular Exp}]
   properties {name="Asset Specular Exp (S)", nodeColor=<rgba 0 1 0.1094 1>}
  }
  MaterialExpressionNamedRerouteDeclaration_7 {
   class=MaterialExpressionNamedRerouteDeclaration
   inputs [None {from=MaterialExpressionClamp_0}]
   properties {name="Refraction - IOR (S)", nodeColor=<rgba 1 0.0938 0 1>}
  }
  MaterialExpressionNamedRerouteDeclaration_70 {
   class=MaterialExpressionNamedRerouteDeclaration
   inputs [None {from=MaterialExpressionTextureCoordinate_0}]
   properties {name="UV Asset (V2)", nodeColor=<rgba 0.9219 0 1 1>}
  }
  MaterialExpressionNamedRerouteDeclaration_71 {
   class=MaterialExpressionNamedRerouteDeclaration
   inputs [None {from=MaterialExpressionVectorParameter_10, output=Saturation}]
   properties {name="Diffuse Map Saturation (S)", nodeColor=<rgba 1 0 0.5156 1>}
  }
  MaterialExpressionNamedRerouteDeclaration_72 {
   class=MaterialExpressionNamedRerouteDeclaration
   inputs [None {from=MaterialExpressionVectorParameter_6, output=Specular Mult}]
   properties {name="Asset Specular Mult (S)", nodeColor=<rgba 1 0 0.2109 1>}
  }
  MaterialExpressionNamedRerouteDeclaration_73 {
   class=MaterialExpressionNamedRerouteDeclaration
   inputs [None {from=MaterialExpressionVectorParameter_0, output=Mult}]
   properties {name="Parallax Map Mult (S)", nodeColor=<rgba 0 1 0.1328 1>}
  }
  MaterialExpressionNamedRerouteDeclaration_74 {
   class=MaterialExpressionNamedRerouteDeclaration
   inputs [None {from=MaterialExpressionVectorParameter_0, output=Exp}]
   properties {name="Parallax Map Exp (S)", nodeColor=<rgba 1 0 0.1875 1>}
  }
  MaterialExpressionNamedRerouteDeclaration_75 {
   class=MaterialExpressionNamedRerouteDeclaration
   inputs [None {from=MaterialExpressionScalarParameter_11}]
   properties {name="Emission Desaturation (S)", nodeColor=<rgba 0.5 0 1 1>}
  }
  MaterialExpressionNamedRerouteDeclaration_76 {
   class=MaterialExpressionNamedRerouteDeclaration
   inputs [None {from=MaterialExpressionVectorParameter_10, output=Value}]
   properties {name="Diffuse Map Value (S)", nodeColor=<rgba 0 0.8359 1 1>}
  }
  MaterialExpressionNamedRerouteDeclaration_77 {
   class=MaterialExpressionNamedRerouteDeclaration
   inputs [None {from=MaterialExpressionMultiply_18}]
   properties {name="Geometric Thickness (S)", nodeColor=<rgba 0 1 0.6484 1>}
  }
  MaterialExpressionNamedRerouteDeclaration_78 {
   class=MaterialExpressionNamedRerouteDeclaration
   inputs [None {from=MaterialExpressionScalarParameter_14}]
   properties {name="Ground Exp (S)", nodeColor=<rgba 1 0.7031 0 1>}
  }
  MaterialExpressionNamedRerouteDeclaration_79 {
   class=MaterialExpressionNamedRerouteDeclaration
   inputs [None {from=MaterialExpressionScalarParameter_12}]
   properties {name="Ice Surface Opacity Mult (S)", nodeColor=<rgba 0 0.6953 1 1>}
  }
  MaterialExpressionNamedRerouteDeclaration_8 {
   class=MaterialExpressionNamedRerouteDeclaration
   inputs [None {from=MaterialExpressionAdd_8}]
   properties {name="Noise Parallax Variation (S)", nodeColor=<rgba 1 0 0.9844 1>}
  }
  MaterialExpressionNamedRerouteDeclaration_80 {
   class=MaterialExpressionNamedRerouteDeclaration
   inputs [None {from=MaterialExpressionVectorParameter_5, output=Pattern Exp}]
   properties {name="Interior Noise Exp (S)", nodeColor=<rgba 0.9453 0 1 1>}
  }
  MaterialExpressionNamedRerouteDeclaration_81 {
   class=MaterialExpressionNamedRerouteDeclaration
   inputs [None {from=MaterialExpressionVectorParameter_9, output=RGB}]
   properties {name="Emissive Fresnel Tint (V3)", nodeColor=<rgba 0 1 0.2734 1>}
  }
  MaterialExpressionNamedRerouteDeclaration_82 {
   class=MaterialExpressionNamedRerouteDeclaration
   inputs [None {from=MaterialExpressionVectorParameter_9, output=Fresnel Exp}]
   properties {name="Emissive Fresnel Exp (S)", nodeColor=<rgba 0.6406 1 0 1>}
  }
  MaterialExpressionNamedRerouteDeclaration_83 {
   class=MaterialExpressionNamedRerouteDeclaration
   inputs [None {from=MaterialExpressionVectorParameter_8, output=Breakup Pattern Scale}]
   properties {name="Noise Depth Variation Scale (S)", nodeColor=<rgba 0 1 0.625 1>}
  }
  MaterialExpressionNamedRerouteDeclaration_84 {
   class=MaterialExpressionNamedRerouteDeclaration
   inputs [None {from=MaterialExpressionVectorParameter_8, output=Breakup Pattern Exp}]
   properties {name="Noise Depth Variation Exp (S)", nodeColor=<rgba 1 0.3047 0 1>}
  }
  MaterialExpressionNamedRerouteDeclaration_85 {
   class=MaterialExpressionNamedRerouteDeclaration
   inputs [None {from=MaterialExpressionVectorParameter_10, output=Diffuse Map Exp}]
   properties {name="Diffuse Map Exp (S)", nodeColor=<rgba 1 0.8438 0 1>}
  }
  MaterialExpressionNamedRerouteDeclaration_86 {
   class=MaterialExpressionNamedRerouteDeclaration
   inputs [None {from=MaterialExpressionScalarParameter_28}]
   properties {name="Ice Interface Opacity Mult (S)", nodeColor=<rgba 0 0.6953 1 1>}
  }
  MaterialExpressionNamedRerouteDeclaration_87 {
   class=MaterialExpressionNamedRerouteDeclaration
   inputs [None {from=MaterialExpressionVectorParameter_11, output=Fresnel Exp}]
   properties {name="Detail Normal Fresnel Exp (S)", nodeColor=<rgba 1 0.0703 0 1>}
  }
  MaterialExpressionNamedRerouteDeclaration_88 {
   class=MaterialExpressionNamedRerouteDeclaration
   inputs [None {from=MaterialExpressionVectorParameter_11, output="Max Intensity (Grazing)"}]
   properties {name="Detail Normal Max Intensity (S)", nodeColor=<rgba 0 0.25 1 1>}
  }
  MaterialExpressionNamedRerouteDeclaration_89 {
   class=MaterialExpressionNamedRerouteDeclaration
   inputs [None {from=MaterialExpressionVectorParameter_11, output="Min Intensity (Facing)"}]
   properties {name="Detail Normal Min Intensity (S)", nodeColor=<rgba 0.5703 1 0 1>}
  }
  MaterialExpressionNamedRerouteDeclaration_9 {
   class=MaterialExpressionNamedRerouteDeclaration
   inputs [None {from=MaterialExpressionMaterialFunctionCall_33, output="Refraction Vector (V3)"}]
   properties {name="Refraction Vector WS (V3)", nodeColor=<rgba 0.5469 1 0 1>}
  }
  MaterialExpressionNamedRerouteDeclaration_90 {
   class=MaterialExpressionNamedRerouteDeclaration
   inputs [None {from=MaterialExpressionScalarParameter_15}]
   properties {name="Ground Brightness (S)", nodeColor=<rgba 0.3828 0 1 1>}
  }
  MaterialExpressionNamedRerouteDeclaration_91 {
   class=MaterialExpressionNamedRerouteDeclaration
   inputs [None {from=MaterialExpressionMaterialFunctionCall_6, output=Result}]
   properties {name="Surface Normal TS (V3)", nodeColor=<rgba 0 0.8594 1 1>}
  }
  MaterialExpressionNamedRerouteDeclaration_92 {
   class=MaterialExpressionNamedRerouteDeclaration
   inputs [None {from=MaterialExpressionScalarParameter_13}]
   properties {name="Use Emission Cutout Channel (S)", nodeColor=<rgba 0 0.6953 1 1>}
  }
  MaterialExpressionNamedRerouteDeclaration_93 {
   class=MaterialExpressionNamedRerouteDeclaration
   inputs [None {from=MaterialExpressionLinearInterpolate_6}]
   properties {name="Ice Roughness (S)", nodeColor=<rgba 0 0.4844 1 1>}
  }
  MaterialExpressionNamedRerouteDeclaration_94 {
   class=MaterialExpressionNamedRerouteDeclaration
   inputs [None {from=MaterialExpressionSaturate_7}]
   properties {name="Ice Specular (S)", nodeColor=<rgba 0.8047 1 0 1>}
  }
  MaterialExpressionNamedRerouteDeclaration_95 {
   class=MaterialExpressionNamedRerouteDeclaration
   inputs [None {from=MaterialExpressionMultiply_5}]
   properties {name="Refraction Luminance (S)", nodeColor=<rgba 1 0 0.75 1>}
  }
  MaterialExpressionNamedRerouteDeclaration_96 {
   class=MaterialExpressionNamedRerouteDeclaration
   inputs [None {from=MaterialExpressionScalarParameter_114}]
   properties {name="Interior Emission Exp (S)", nodeColor=<rgba 0 0.6953 1 1>}
  }
  MaterialExpressionNamedRerouteDeclaration_97 {
   class=MaterialExpressionNamedRerouteDeclaration
   inputs [None {from=MaterialExpressionScalarParameter_112}]
   properties {name="Interior Emission Intensity (S)", nodeColor=<rgba 0 0.6953 1 1>}
  }
  MaterialExpressionNamedRerouteDeclaration_98 {
   class=MaterialExpressionNamedRerouteDeclaration
   inputs [None {from=MaterialExpressionVectorParameter_4, output=Roughness Exp}]
   properties {name="Detail Roughness Exp (S)", nodeColor=<rgba 1 0.9844 0 1>}
  }
  MaterialExpressionNamedRerouteDeclaration_99 {
   class=MaterialExpressionNamedRerouteDeclaration
   inputs [None {from=MaterialExpressionVectorParameter_4, output=Roughness Mult}]
   properties {name="Detail Roughness Mult (S)", nodeColor=<rgba 0.6641 0 1 1>}
  }
  MaterialExpressionNamedRerouteUsage_0 {class=MaterialExpressionNamedRerouteUsage}
  MaterialExpressionNamedRerouteUsage_1 {class=MaterialExpressionNamedRerouteUsage}
  MaterialExpressionNamedRerouteUsage_10 {class=MaterialExpressionNamedRerouteUsage}
  MaterialExpressionNamedRerouteUsage_100 {class=MaterialExpressionNamedRerouteUsage}
  MaterialExpressionNamedRerouteUsage_101 {class=MaterialExpressionNamedRerouteUsage}
  MaterialExpressionNamedRerouteUsage_102 {class=MaterialExpressionNamedRerouteUsage}
  MaterialExpressionNamedRerouteUsage_103 {class=MaterialExpressionNamedRerouteUsage}
  MaterialExpressionNamedRerouteUsage_104 {class=MaterialExpressionNamedRerouteUsage}
  MaterialExpressionNamedRerouteUsage_105 {class=MaterialExpressionNamedRerouteUsage}
  MaterialExpressionNamedRerouteUsage_106 {class=MaterialExpressionNamedRerouteUsage}
  MaterialExpressionNamedRerouteUsage_107 {class=MaterialExpressionNamedRerouteUsage}
  MaterialExpressionNamedRerouteUsage_108 {class=MaterialExpressionNamedRerouteUsage}
  MaterialExpressionNamedRerouteUsage_109 {class=MaterialExpressionNamedRerouteUsage}
  MaterialExpressionNamedRerouteUsage_11 {class=MaterialExpressionNamedRerouteUsage}
  MaterialExpressionNamedRerouteUsage_110 {class=MaterialExpressionNamedRerouteUsage}
  MaterialExpressionNamedRerouteUsage_111 {class=MaterialExpressionNamedRerouteUsage}
  MaterialExpressionNamedRerouteUsage_112 {class=MaterialExpressionNamedRerouteUsage}
  MaterialExpressionNamedRerouteUsage_113 {class=MaterialExpressionNamedRerouteUsage}
  MaterialExpressionNamedRerouteUsage_114 {class=MaterialExpressionNamedRerouteUsage}
  MaterialExpressionNamedRerouteUsage_115 {class=MaterialExpressionNamedRerouteUsage}
  MaterialExpressionNamedRerouteUsage_116 {class=MaterialExpressionNamedRerouteUsage}
  MaterialExpressionNamedRerouteUsage_117 {class=MaterialExpressionNamedRerouteUsage}
  MaterialExpressionNamedRerouteUsage_118 {class=MaterialExpressionNamedRerouteUsage}
  MaterialExpressionNamedRerouteUsage_119 {class=MaterialExpressionNamedRerouteUsage}
  MaterialExpressionNamedRerouteUsage_12 {class=MaterialExpressionNamedRerouteUsage}
  MaterialExpressionNamedRerouteUsage_120 {class=MaterialExpressionNamedRerouteUsage}
  MaterialExpressionNamedRerouteUsage_121 {class=MaterialExpressionNamedRerouteUsage}
  MaterialExpressionNamedRerouteUsage_122 {class=MaterialExpressionNamedRerouteUsage}
  MaterialExpressionNamedRerouteUsage_123 {class=MaterialExpressionNamedRerouteUsage}
  MaterialExpressionNamedRerouteUsage_124 {class=MaterialExpressionNamedRerouteUsage}
  MaterialExpressionNamedRerouteUsage_125 {class=MaterialExpressionNamedRerouteUsage}
  MaterialExpressionNamedRerouteUsage_126 {class=MaterialExpressionNamedRerouteUsage}
  MaterialExpressionNamedRerouteUsage_127 {class=MaterialExpressionNamedRerouteUsage}
  MaterialExpressionNamedRerouteUsage_128 {class=MaterialExpressionNamedRerouteUsage}
  MaterialExpressionNamedRerouteUsage_129 {class=MaterialExpressionNamedRerouteUsage}
  MaterialExpressionNamedRerouteUsage_13 {class=MaterialExpressionNamedRerouteUsage}
  MaterialExpressionNamedRerouteUsage_130 {class=MaterialExpressionNamedRerouteUsage}
  MaterialExpressionNamedRerouteUsage_131 {class=MaterialExpressionNamedRerouteUsage}
  MaterialExpressionNamedRerouteUsage_132 {class=MaterialExpressionNamedRerouteUsage}
  MaterialExpressionNamedRerouteUsage_133 {class=MaterialExpressionNamedRerouteUsage}
  MaterialExpressionNamedRerouteUsage_134 {class=MaterialExpressionNamedRerouteUsage}
  MaterialExpressionNamedRerouteUsage_135 {class=MaterialExpressionNamedRerouteUsage}
  MaterialExpressionNamedRerouteUsage_136 {class=MaterialExpressionNamedRerouteUsage}
  MaterialExpressionNamedRerouteUsage_137 {class=MaterialExpressionNamedRerouteUsage}
  MaterialExpressionNamedRerouteUsage_138 {class=MaterialExpressionNamedRerouteUsage}
  MaterialExpressionNamedRerouteUsage_139 {class=MaterialExpressionNamedRerouteUsage}
  MaterialExpressionNamedRerouteUsage_14 {class=MaterialExpressionNamedRerouteUsage}
  MaterialExpressionNamedRerouteUsage_141 {class=MaterialExpressionNamedRerouteUsage}
  MaterialExpressionNamedRerouteUsage_143 {class=MaterialExpressionNamedRerouteUsage}
  MaterialExpressionNamedRerouteUsage_149 {class=MaterialExpressionNamedRerouteUsage}
  MaterialExpressionNamedRerouteUsage_15 {class=MaterialExpressionNamedRerouteUsage}
  MaterialExpressionNamedRerouteUsage_150 {class=MaterialExpressionNamedRerouteUsage}
  MaterialExpressionNamedRerouteUsage_151 {class=MaterialExpressionNamedRerouteUsage}
  MaterialExpressionNamedRerouteUsage_152 {class=MaterialExpressionNamedRerouteUsage}
  MaterialExpressionNamedRerouteUsage_153 {class=MaterialExpressionNamedRerouteUsage}
  MaterialExpressionNamedRerouteUsage_154 {class=MaterialExpressionNamedRerouteUsage}
  MaterialExpressionNamedRerouteUsage_155 {class=MaterialExpressionNamedRerouteUsage}
  MaterialExpressionNamedRerouteUsage_156 {class=MaterialExpressionNamedRerouteUsage}
  MaterialExpressionNamedRerouteUsage_16 {class=MaterialExpressionNamedRerouteUsage}
  MaterialExpressionNamedRerouteUsage_162 {class=MaterialExpressionNamedRerouteUsage}
  MaterialExpressionNamedRerouteUsage_163 {class=MaterialExpressionNamedRerouteUsage}
  MaterialExpressionNamedRerouteUsage_164 {class=MaterialExpressionNamedRerouteUsage}
  MaterialExpressionNamedRerouteUsage_165 {class=MaterialExpressionNamedRerouteUsage}
  MaterialExpressionNamedRerouteUsage_166 {class=MaterialExpressionNamedRerouteUsage}
  MaterialExpressionNamedRerouteUsage_17 {class=MaterialExpressionNamedRerouteUsage}
  MaterialExpressionNamedRerouteUsage_18 {class=MaterialExpressionNamedRerouteUsage}
  MaterialExpressionNamedRerouteUsage_19 {class=MaterialExpressionNamedRerouteUsage}
  MaterialExpressionNamedRerouteUsage_2 {class=MaterialExpressionNamedRerouteUsage}
  MaterialExpressionNamedRerouteUsage_20 {class=MaterialExpressionNamedRerouteUsage}
  MaterialExpressionNamedRerouteUsage_21 {class=MaterialExpressionNamedRerouteUsage}
  MaterialExpressionNamedRerouteUsage_22 {class=MaterialExpressionNamedRerouteUsage}
  MaterialExpressionNamedRerouteUsage_23 {class=MaterialExpressionNamedRerouteUsage}
  MaterialExpressionNamedRerouteUsage_24 {class=MaterialExpressionNamedRerouteUsage}
  MaterialExpressionNamedRerouteUsage_25 {class=MaterialExpressionNamedRerouteUsage}
  MaterialExpressionNamedRerouteUsage_26 {class=MaterialExpressionNamedRerouteUsage}
  MaterialExpressionNamedRerouteUsage_27 {class=MaterialExpressionNamedRerouteUsage}
  MaterialExpressionNamedRerouteUsage_28 {class=MaterialExpressionNamedRerouteUsage}
  MaterialExpressionNamedRerouteUsage_29 {class=MaterialExpressionNamedRerouteUsage}
  MaterialExpressionNamedRerouteUsage_3 {class=MaterialExpressionNamedRerouteUsage}
  MaterialExpressionNamedRerouteUsage_30 {class=MaterialExpressionNamedRerouteUsage}
  MaterialExpressionNamedRerouteUsage_31 {class=MaterialExpressionNamedRerouteUsage}
  MaterialExpressionNamedRerouteUsage_32 {class=MaterialExpressionNamedRerouteUsage}
  MaterialExpressionNamedRerouteUsage_33 {class=MaterialExpressionNamedRerouteUsage}
  MaterialExpressionNamedRerouteUsage_34 {class=MaterialExpressionNamedRerouteUsage}
  MaterialExpressionNamedRerouteUsage_35 {class=MaterialExpressionNamedRerouteUsage}
  MaterialExpressionNamedRerouteUsage_36 {class=MaterialExpressionNamedRerouteUsage}
  MaterialExpressionNamedRerouteUsage_37 {class=MaterialExpressionNamedRerouteUsage}
  MaterialExpressionNamedRerouteUsage_38 {class=MaterialExpressionNamedRerouteUsage}
  MaterialExpressionNamedRerouteUsage_39 {class=MaterialExpressionNamedRerouteUsage}
  MaterialExpressionNamedRerouteUsage_4 {class=MaterialExpressionNamedRerouteUsage}
  MaterialExpressionNamedRerouteUsage_40 {class=MaterialExpressionNamedRerouteUsage}
  MaterialExpressionNamedRerouteUsage_41 {class=MaterialExpressionNamedRerouteUsage}
  MaterialExpressionNamedRerouteUsage_42 {class=MaterialExpressionNamedRerouteUsage}
  MaterialExpressionNamedRerouteUsage_43 {class=MaterialExpressionNamedRerouteUsage}
  MaterialExpressionNamedRerouteUsage_44 {class=MaterialExpressionNamedRerouteUsage}
  MaterialExpressionNamedRerouteUsage_45 {class=MaterialExpressionNamedRerouteUsage}
  MaterialExpressionNamedRerouteUsage_46 {class=MaterialExpressionNamedRerouteUsage}
  MaterialExpressionNamedRerouteUsage_47 {class=MaterialExpressionNamedRerouteUsage}
  MaterialExpressionNamedRerouteUsage_48 {class=MaterialExpressionNamedRerouteUsage}
  MaterialExpressionNamedRerouteUsage_49 {class=MaterialExpressionNamedRerouteUsage}
  MaterialExpressionNamedRerouteUsage_5 {class=MaterialExpressionNamedRerouteUsage}
  MaterialExpressionNamedRerouteUsage_50 {class=MaterialExpressionNamedRerouteUsage}
  MaterialExpressionNamedRerouteUsage_51 {class=MaterialExpressionNamedRerouteUsage}
  MaterialExpressionNamedRerouteUsage_52 {class=MaterialExpressionNamedRerouteUsage}
  MaterialExpressionNamedRerouteUsage_53 {class=MaterialExpressionNamedRerouteUsage}
  MaterialExpressionNamedRerouteUsage_54 {class=MaterialExpressionNamedRerouteUsage}
  MaterialExpressionNamedRerouteUsage_55 {class=MaterialExpressionNamedRerouteUsage}
  MaterialExpressionNamedRerouteUsage_56 {class=MaterialExpressionNamedRerouteUsage}
  MaterialExpressionNamedRerouteUsage_57 {class=MaterialExpressionNamedRerouteUsage}
  MaterialExpressionNamedRerouteUsage_58 {class=MaterialExpressionNamedRerouteUsage}
  MaterialExpressionNamedRerouteUsage_59 {class=MaterialExpressionNamedRerouteUsage}
  MaterialExpressionNamedRerouteUsage_6 {class=MaterialExpressionNamedRerouteUsage}
  MaterialExpressionNamedRerouteUsage_60 {class=MaterialExpressionNamedRerouteUsage}
  MaterialExpressionNamedRerouteUsage_61 {class=MaterialExpressionNamedRerouteUsage}
  MaterialExpressionNamedRerouteUsage_62 {class=MaterialExpressionNamedRerouteUsage}
  MaterialExpressionNamedRerouteUsage_63 {class=MaterialExpressionNamedRerouteUsage}
  MaterialExpressionNamedRerouteUsage_64 {class=MaterialExpressionNamedRerouteUsage}
  MaterialExpressionNamedRerouteUsage_65 {class=MaterialExpressionNamedRerouteUsage}
  MaterialExpressionNamedRerouteUsage_66 {class=MaterialExpressionNamedRerouteUsage}
  MaterialExpressionNamedRerouteUsage_67 {class=MaterialExpressionNamedRerouteUsage}
  MaterialExpressionNamedRerouteUsage_68 {class=MaterialExpressionNamedRerouteUsage}
  MaterialExpressionNamedRerouteUsage_69 {class=MaterialExpressionNamedRerouteUsage}
  MaterialExpressionNamedRerouteUsage_7 {class=MaterialExpressionNamedRerouteUsage}
  MaterialExpressionNamedRerouteUsage_70 {class=MaterialExpressionNamedRerouteUsage}
  MaterialExpressionNamedRerouteUsage_71 {class=MaterialExpressionNamedRerouteUsage}
  MaterialExpressionNamedRerouteUsage_72 {class=MaterialExpressionNamedRerouteUsage}
  MaterialExpressionNamedRerouteUsage_73 {class=MaterialExpressionNamedRerouteUsage}
  MaterialExpressionNamedRerouteUsage_74 {class=MaterialExpressionNamedRerouteUsage}
  MaterialExpressionNamedRerouteUsage_75 {class=MaterialExpressionNamedRerouteUsage}
  MaterialExpressionNamedRerouteUsage_78 {class=MaterialExpressionNamedRerouteUsage}
  MaterialExpressionNamedRerouteUsage_79 {class=MaterialExpressionNamedRerouteUsage}
  MaterialExpressionNamedRerouteUsage_8 {class=MaterialExpressionNamedRerouteUsage}
  MaterialExpressionNamedRerouteUsage_80 {class=MaterialExpressionNamedRerouteUsage}
  MaterialExpressionNamedRerouteUsage_81 {class=MaterialExpressionNamedRerouteUsage}
  MaterialExpressionNamedRerouteUsage_82 {class=MaterialExpressionNamedRerouteUsage}
  MaterialExpressionNamedRerouteUsage_83 {class=MaterialExpressionNamedRerouteUsage}
  MaterialExpressionNamedRerouteUsage_84 {class=MaterialExpressionNamedRerouteUsage}
  MaterialExpressionNamedRerouteUsage_85 {class=MaterialExpressionNamedRerouteUsage}
  MaterialExpressionNamedRerouteUsage_86 {class=MaterialExpressionNamedRerouteUsage}
  MaterialExpressionNamedRerouteUsage_87 {class=MaterialExpressionNamedRerouteUsage}
  MaterialExpressionNamedRerouteUsage_88 {class=MaterialExpressionNamedRerouteUsage}
  MaterialExpressionNamedRerouteUsage_89 {class=MaterialExpressionNamedRerouteUsage}
  MaterialExpressionNamedRerouteUsage_9 {class=MaterialExpressionNamedRerouteUsage}
  MaterialExpressionNamedRerouteUsage_90 {class=MaterialExpressionNamedRerouteUsage}
  MaterialExpressionNamedRerouteUsage_91 {class=MaterialExpressionNamedRerouteUsage}
  MaterialExpressionNamedRerouteUsage_92 {class=MaterialExpressionNamedRerouteUsage}
  MaterialExpressionNamedRerouteUsage_93 {class=MaterialExpressionNamedRerouteUsage}
  MaterialExpressionNamedRerouteUsage_94 {class=MaterialExpressionNamedRerouteUsage}
  MaterialExpressionNamedRerouteUsage_95 {class=MaterialExpressionNamedRerouteUsage}
  MaterialExpressionNamedRerouteUsage_96 {class=MaterialExpressionNamedRerouteUsage}
  MaterialExpressionNamedRerouteUsage_97 {class=MaterialExpressionNamedRerouteUsage}
  MaterialExpressionNamedRerouteUsage_98 {class=MaterialExpressionNamedRerouteUsage}
  MaterialExpressionNamedRerouteUsage_99 {class=MaterialExpressionNamedRerouteUsage}
  MaterialExpressionNormalize_0 {class=MaterialExpressionNormalize, inputs [VectorInput {from=MaterialExpressionTransform_0, output=""}]}
  MaterialExpressionOneMinus_0 {class=MaterialExpressionOneMinus, inputs [None {from=MaterialExpressionLinearInterpolate_4}]}
  MaterialExpressionOneMinus_1 {class=MaterialExpressionOneMinus, inputs [None {from=MaterialExpressionMultiply_4}]}
  MaterialExpressionOneMinus_3 {class=MaterialExpressionOneMinus, inputs [None {from=MaterialExpressionNamedRerouteUsage_138}]}
  MaterialExpressionOneMinus_4 {class=MaterialExpressionOneMinus, inputs [None {from=MaterialExpressionNamedRerouteUsage_71}]}
  MaterialExpressionOneMinus_5 {class=MaterialExpressionOneMinus, inputs [None {from=MaterialExpressionLinearInterpolate_9}]}
  MaterialExpressionPower_0 {
   class=MaterialExpressionPower
   inputs [
    Base {from=MaterialExpressionNamedRerouteUsage_115}
    Exp {from=MaterialExpressionNamedRerouteUsage_94}
   ]
  }
  MaterialExpressionPower_1 {
   class=MaterialExpressionPower
   inputs [
    Base {from=MaterialExpressionSaturate_13}
    Exp {from=MaterialExpressionScalarParameter_16}
   ]
  }
  MaterialExpressionPower_10 {
   class=MaterialExpressionPower
   inputs [
    Base {from=MaterialExpressionNamedRerouteUsage_56}
    Exp {from=MaterialExpressionNamedRerouteUsage_97}
   ]
  }
  MaterialExpressionPower_11 {
   class=MaterialExpressionPower
   inputs [
    Base {from=MaterialExpressionOneMinus_4}
    Exp {from=MaterialExpressionNamedRerouteUsage_127}
   ]
  }
  MaterialExpressionPower_12 {
   class=MaterialExpressionPower
   inputs [
    Base {from=MaterialExpressionNamedRerouteUsage_58}
    Exp {from=MaterialExpressionNamedRerouteUsage_85}
   ]
  }
  MaterialExpressionPower_18 {
   class=MaterialExpressionPower
   inputs [
    Base {from=MaterialExpressionOneMinus_5}
    Exp {from=MaterialExpressionNamedRerouteUsage_135}
   ]
  }
  MaterialExpressionPower_2 {
   class=MaterialExpressionPower
   inputs [
    Base {from=MaterialExpressionMultiply_6}
    Exp {from=MaterialExpressionNamedRerouteUsage_68}
   ]
  }
  MaterialExpressionPower_3 {
   class=MaterialExpressionPower
   inputs [
    Base {from=MaterialExpressionNamedRerouteUsage_59}
    Exp {from=MaterialExpressionNamedRerouteUsage_72}
   ]
  }
  MaterialExpressionPower_4 {
   class=MaterialExpressionPower
   inputs [
    Base {from=MaterialExpressionTextureSample_4, output=R}
    Exp {from=MaterialExpressionNamedRerouteUsage_154}
   ]
  }
  MaterialExpressionPower_5 {
   class=MaterialExpressionPower
   inputs [
    Base {from=MaterialExpressionNamedRerouteUsage_61}
    Exp {from=MaterialExpressionNamedRerouteUsage_54}
   ]
  }
  MaterialExpressionPower_6 {
   class=MaterialExpressionPower
   inputs [
    Base {from=MaterialExpressionTextureSample_5, output=R}
    Exp {from=MaterialExpressionNamedRerouteUsage_86}
   ]
  }
  MaterialExpressionPower_7 {
   class=MaterialExpressionPower
   inputs [
    Base {from=MaterialExpressionNamedRerouteDeclaration_27}
    Exp {from=MaterialExpressionNamedRerouteUsage_13}
   ]
  }
  MaterialExpressionPower_8 {
   class=MaterialExpressionPower
   inputs [
    Base {from=MaterialExpressionMaterialFunctionCall_27, output=Result}
    Exp {from=MaterialExpressionNamedRerouteUsage_81}
   ]
  }
  MaterialExpressionPower_9 {
   class=MaterialExpressionPower
   inputs [
    Base {from=MaterialExpressionNamedRerouteUsage_116}
    Exp {from=MaterialExpressionNamedRerouteUsage_152}
   ]
  }
  MaterialExpressionRgbToHsv_0 {class=MaterialExpressionRgbToHsv, inputs [None {from=MaterialExpressionPower_0, output=""}], properties {}}
  MaterialExpressionSaturate_0 {class=MaterialExpressionSaturate, inputs [None {from=MaterialExpressionVectorParameter_2, output=Min Depth Mult}]}
  MaterialExpressionSaturate_1 {class=MaterialExpressionSaturate, inputs [None {from=MaterialExpressionVectorParameter_3, output=Min Depth Mult}]}
  MaterialExpressionSaturate_10 {class=MaterialExpressionSaturate, inputs [None {from=MaterialExpressionMultiply_9}]}
  MaterialExpressionSaturate_11 {class=MaterialExpressionSaturate, inputs [None {from=MaterialExpressionMultiply_25}]}
  MaterialExpressionSaturate_12 {class=MaterialExpressionSaturate, inputs [None {from=MaterialExpressionMultiply_34}]}
  MaterialExpressionSaturate_13 {class=MaterialExpressionSaturate, inputs [None {from=MaterialExpressionMultiply_19}]}
  MaterialExpressionSaturate_18 {class=MaterialExpressionSaturate, inputs [None {from=MaterialExpressionDotProduct_9}]}
  MaterialExpressionSaturate_2 {class=MaterialExpressionSaturate, inputs [None {from=MaterialExpressionScalarParameter_4}]}
  MaterialExpressionSaturate_3 {class=MaterialExpressionSaturate, inputs [None {from=MaterialExpressionScalarParameter_5}]}
  MaterialExpressionSaturate_4 {class=MaterialExpressionSaturate, inputs [None {from=MaterialExpressionMultiply_1}]}
  MaterialExpressionSaturate_6 {class=MaterialExpressionSaturate, inputs [None {from=MaterialExpressionPower_2}]}
  MaterialExpressionSaturate_7 {class=MaterialExpressionSaturate, inputs [None {from=MaterialExpressionMultiply_7}]}
  MaterialExpressionSaturate_8 {class=MaterialExpressionSaturate, inputs [None {from=MaterialExpressionDotProduct_1}]}
  MaterialExpressionSaturate_9 {class=MaterialExpressionSaturate, inputs [None {from=MaterialExpressionMultiply_11}]}
  MaterialExpressionScalarParameter_0 {
   class=MaterialExpressionScalarParameter
   properties {defaultValue=1.5, group=Ice | 03_Refraction, parameterName=Refraction - IOR, sliderMax=5, sliderMin=1, sortPriority=30}
  }
  MaterialExpressionScalarParameter_1 {
   class=MaterialExpressionScalarParameter
   properties {defaultValue=0.5, group=Ice | 05 _Absorption, parameterName=Absorption - Constant Depth, sortPriority=30}
  }
  MaterialExpressionScalarParameter_10 {
   class=MaterialExpressionScalarParameter
   properties {group=Ice | 01_Globals, parameterName=Ice Mask Min Amount, sliderMax=1, sortPriority=20}
  }
  MaterialExpressionScalarParameter_11 {
   class=MaterialExpressionScalarParameter
   properties {group=Ice | 06_Emission, parameterName=Emission - Refracted Emission Desaturation, sliderMax=1, sortPriority=50}
  }
  MaterialExpressionScalarParameter_112 {
   class=MaterialExpressionScalarParameter
   properties {defaultValue=1, group=Ice | 06_Emission, parameterName=Emission - Refracted Emission Intensity, sortPriority=30}
  }
  MaterialExpressionScalarParameter_114 {
   class=MaterialExpressionScalarParameter
   properties {defaultValue=1, group=Ice | 06_Emission, parameterName=Emission - Refracted Emission Exp, sortPriority=20}
  }
  MaterialExpressionScalarParameter_118 {
   class=MaterialExpressionScalarParameter
   properties {defaultValue=0.12, group=Ice | 06_Emission, parameterName=Emission - Constant Depth, sortPriority=70}
  }
  MaterialExpressionScalarParameter_12 {
   class=MaterialExpressionScalarParameter
   properties {defaultValue=1, group=Ice | 01_Globals, parameterName=Ice Surface Opacity Mult, sliderMax=1, sortPriority=120}
  }
  MaterialExpressionScalarParameter_13 {
   class=MaterialExpressionScalarParameter
   properties {group=Ice | 01_Globals, parameterName=Use Emission Cutout Channel, sliderMax=1, sortPriority=220}
  }
  MaterialExpressionScalarParameter_14 {
   class=MaterialExpressionScalarParameter
   properties {defaultValue=0.35, group=Ice | 03_Refraction, parameterName=Refraction - Ground Exp, sliderMax=10, sortPriority=90}
  }
  MaterialExpressionScalarParameter_15 {
   class=MaterialExpressionScalarParameter
   properties {defaultValue=1.25, group=Ice | 03_Refraction, parameterName=Refraction - Ground Brightness, sortPriority=100}
  }
  MaterialExpressionScalarParameter_16 {class=MaterialExpressionScalarParameter, properties {defaultValue=1, group=Ice | 01_Globals, parameterName=Ice Mask Exp, sortPriority=16}}
  MaterialExpressionScalarParameter_2 {
   class=MaterialExpressionScalarParameter
   properties {group=Ice | 03_Refraction, parameterName=Refraction - Refracted Sky Desaturation, sliderMax=1, sortPriority=60}
  }
  MaterialExpressionScalarParameter_23 {
   class=MaterialExpressionScalarParameter
   properties {group=Ice | 01_Globals, parameterName=Ice Diffuse Out Desaturation, sliderMax=1, sortPriority=320}
  }
  MaterialExpressionScalarParameter_24 {
   class=MaterialExpressionScalarParameter
   properties {defaultValue=1, group=Ice | 01_Globals, parameterName=Ice Diffuse Out Gain, sliderMax=1, sortPriority=330}
  }
  MaterialExpressionScalarParameter_26 {
   class=MaterialExpressionScalarParameter
   properties {defaultValue=1, group=Ice | 06_Emission, parameterName=Emission - Variation Noise Speed, sortPriority=110}
  }
  MaterialExpressionScalarParameter_28 {
   class=MaterialExpressionScalarParameter
   properties {defaultValue=1, group=Ice | 01_Globals, parameterName=Ice Interface Opacity Mult, sliderMax=1, sortPriority=130}
  }
  MaterialExpressionScalarParameter_3 {class=MaterialExpressionScalarParameter, properties {defaultValue=1, group=Ice | 01_Globals, parameterName=Ice Mask Mult, sortPriority=15}}
  MaterialExpressionScalarParameter_39 {
   class=MaterialExpressionScalarParameter
   properties {defaultValue=1, group=Ice | 04_Volume, parameterName=Volume - Thickness Gradient Exp, sliderMax=10, sliderMin=0.01, sortPriority=10}
  }
  MaterialExpressionScalarParameter_4 {
   class=MaterialExpressionScalarParameter
   properties {group=Ice | 05 _Absorption, parameterName="Absorption - Use Constant Depth (0) or Parallax Map (1)", sliderMax=1, sortPriority=20}
  }
  MaterialExpressionScalarParameter_40 {
   class=MaterialExpressionScalarParameter
   properties {defaultValue=10, group=Ice | 04_Volume, parameterName=Volume - Max Thickness, sortPriority=20}
  }
  MaterialExpressionScalarParameter_47 {
   class=MaterialExpressionScalarParameter
   properties {defaultValue=1, group=Ice | 05 _Absorption, parameterName=Absorption -Constant Depth Thickness Attenuation, sliderMax=1, sortPriority=40}
  }
  MaterialExpressionScalarParameter_5 {
   class=MaterialExpressionScalarParameter
   properties {group=Ice | 06_Emission, parameterName="Emission - Use Constant Depth (0) or Parallax Map (1)", sliderMax=1, sortPriority=60}
  }
  MaterialExpressionScalarParameter_50 {
   class=MaterialExpressionScalarParameter
   properties {defaultValue=0.2, group=Ice | 02_Surface, parameterName=Surface - Detail Tiles UV Scale, sortPriority=120}
  }
  MaterialExpressionScalarParameter_6 {
   class=MaterialExpressionScalarParameter
   properties {group=Ice | 06_Emission, parameterName=Emission - Refracted Emission Hue Shift, sliderMax=1, sortPriority=40}
  }
  MaterialExpressionScalarParameter_7 {
   class=MaterialExpressionScalarParameter
   properties {group=Ice | 01_Globals, parameterName="Ice Out Emissive <-> Diffuse Mix", sliderMax=1, sortPriority=310}
  }
  MaterialExpressionScalarParameter_76 {class=MaterialExpressionScalarParameter, properties {defaultValue=1, group=Ice | 04_Volume, parameterName=Volume - Density, sortPriority=30}}
  MaterialExpressionScalarParameter_77 {
   class=MaterialExpressionScalarParameter
   properties {group=Ice | 04_Volume, parameterName=Volume - Scattering Phase, sliderMax=1, sliderMin=-1, sortPriority=60}
  }
  MaterialExpressionScalarParameter_78 {
   class=MaterialExpressionScalarParameter
   properties {defaultValue=1, group=Ice | 03_Refraction, parameterName=Refraction - Transmission Attenuation, sortPriority=50}
  }
  MaterialExpressionScalarParameter_9 {
   class=MaterialExpressionScalarParameter
   properties {defaultValue=0.2, group=Ice | 03_Refraction, parameterName=Refraction - Roughness, sliderMax=1, sortPriority=40}
  }
  MaterialExpressionScalarParameter_92 {
   class=MaterialExpressionScalarParameter
   properties {defaultValue=1, group=Ice | 03_Refraction, parameterName=Refraction - Refracted Light Intensity, sliderMax=1, sortPriority=70}
  }
  MaterialExpressionSetMaterialAttributes_3 {
   class=MaterialExpressionSetMaterialAttributes
   inputs [
    MaterialAttributes {from=MaterialExpressionNamedRerouteUsage_75, output=""}
    BaseColor {from=MaterialExpressionAdd_3, output=""}
    EmissiveColor {from=MaterialExpressionMultiply_26, output=""}
    Metallic {from=MaterialExpressionConstant_13, output=""}
    Specular {from=MaterialExpressionNamedRerouteUsage_78, output=""}
    Roughness {from=MaterialExpressionNamedRerouteUsage_79, output=""}
    Normal {from=MaterialExpressionNamedRerouteUsage_80, output=""}
   ]
   properties {attributeSetTypes [
     #1
     #5
     #6
     #2
     #3
     #4
    ]}
  }
  MaterialExpressionStaticBool_0 {class=MaterialExpressionStaticBool, properties {value=true}}
  MaterialExpressionStaticSwitchParameter_3 {
   class=MaterialExpressionStaticSwitchParameter
   inputs [
    True {from=MaterialExpressionBlendMaterialAttributes_1, output=""}
    False {from=MaterialExpressionNamedRerouteUsage_23, output=""}
   ]
   properties {group=04_Tech Art, parameterName=UseIce}
  }
  MaterialExpressionSwitch_0 {
   class=MaterialExpressionSwitch
   inputs [
    SwitchValue {from=MaterialExpressionNamedRerouteUsage_124, output=""}
    Default {from="", output=""}
    ConstantDepth {from=MaterialExpressionMultiply_15, output=""}
    ParallaxMap {from=MaterialExpressionLinearInterpolate_2, output=""}
   ]
   properties {inputs [
     {
      inputName=ConstantDepth
      input {
       expression=&self:MaterialExpressionMultiply_15
       inputName=None
       mask=0
       maskA=0
       maskB=0
       maskG=0
       maskR=0
       outputIndex=0
      }
     }
     {
      inputName=ParallaxMap
      input {
       expression=&self:MaterialExpressionLinearInterpolate_2
       inputName=None
       mask=0
       maskA=0
       maskB=0
       maskG=0
       maskR=0
       outputIndex=0
      }
     }
    ]}
  }
  MaterialExpressionSwitch_1 {
   class=MaterialExpressionSwitch
   inputs [
    SwitchValue {from=MaterialExpressionNamedRerouteUsage_139, output=""}
    Default {from="", output=""}
    ConstantDepth {from=MaterialExpressionNamedRerouteUsage_14, output=""}
    ParallaxMap {from=MaterialExpressionLinearInterpolate_3, output=""}
   ]
   properties {inputs [
     {
      inputName=ConstantDepth
      input {
       expression=&self:MaterialExpressionNamedRerouteUsage_14
       inputName=None
       mask=0
       maskA=0
       maskB=0
       maskG=0
       maskR=0
       outputIndex=0
      }
     }
     {
      inputName=ParallaxMap
      input {
       expression=&self:MaterialExpressionLinearInterpolate_3
       inputName=None
       mask=0
       maskA=0
       maskB=0
       maskG=0
       maskR=0
       outputIndex=0
      }
     }
    ]}
  }
  MaterialExpressionTextureCoordinate_0 {class=MaterialExpressionTextureCoordinate}
  MaterialExpressionTextureObjectParameter_0 {
   class=MaterialExpressionTextureObjectParameter
   inputs [
    UVs {from="", output=""}
    Apply View MipBias {from="", output=""}
   ]
   properties {group=Ice | 01_Globals, parameterName=Emissive, sortPriority=210, texture=&T_BlacksRGB}
  }
  MaterialExpressionTextureObjectParameter_3 {
   class=MaterialExpressionTextureObjectParameter
   inputs [
    UVs {from="", output=""}
    Apply View MipBias {from="", output=""}
   ]
   properties {
    group=Ice | 02_Surface
    parameterName=Surface - Detail Roughness Map
    samplerType=SAMPLERTYPE_Masks
    sortPriority=110
    texture=/BRCosmetics/Characters/Player/Male/Medium/Bodies/M_MED_GlacialTrooper/Textures/T_IceSurface_R
   }
  }
  MaterialExpressionTextureObjectParameter_4 {
   class=MaterialExpressionTextureObjectParameter
   inputs [
    UVs {from="", output=""}
    Apply View MipBias {from="", output=""}
   ]
   properties {
    group=Ice | 02_Surface
    parameterName=Surface - Detail Normal Map
    samplerType=SAMPLERTYPE_Normal
    sortPriority=100
    texture=/BRCosmetics/Characters/Player/Male/Medium/Bodies/M_MED_GlacialTrooper/Textures/T_IceSurface_N
   }
  }
  MaterialExpressionTextureObject_0 {
   class=MaterialExpressionTextureObject
   inputs []
   properties {desc=cube, samplerType=SAMPLERTYPE_LinearColor, texture=/Game/Global/MaterialFunctions/TechArt/Refraction/Shared/T_Latlong_Sky360_256_FN}
  }
  MaterialExpressionTextureSampleParameter2D_0 {
   class=MaterialExpressionTextureSampleParameter2D
   inputs [
    UVs {from=MaterialExpressionAdd_0, output=""}
    Apply View MipBias {from="", output=""}
   ]
   properties {
    group=00_Texture Inputs
    parameterName=Diffuse
    samplerSource=SSM_Wrap_WorldGroupSettings
    sortPriority=0
    texture=/Game/Global/Textures/Default/Blanks/T_MidgreysRGB
   }
  }
  MaterialExpressionTextureSample_15 {
   class=MaterialExpressionTextureSample
   inputs [
    UVs {from=MaterialExpressionNamedRerouteUsage_21}
    Tex {from=""}
    Apply View MipBias {from=""}
   ]
   properties {samplerSource=SSM_Wrap_WorldGroupSettings, texture=/BRCosmetics/Animation/Props/Gadgets_SatinCheddar/Materials/T_Perlin_Noise_M1}
  }
  MaterialExpressionTextureSample_4 {
   class=MaterialExpressionTextureSample
   inputs [
    UVs {from=MaterialExpressionNamedRerouteUsage_119}
    Tex {from=MaterialExpressionNamedRerouteUsage_164}
    Apply View MipBias {from=""}
   ]
   properties {samplerSource=SSM_Wrap_WorldGroupSettings}
  }
  MaterialExpressionTextureSample_5 {
   class=MaterialExpressionTextureSample
   inputs [
    UVs {from=MaterialExpressionAdd_17}
    Tex {from=""}
    Apply View MipBias {from=""}
   ]
   properties {samplerSource=SSM_Wrap_WorldGroupSettings, texture=&T_Perlin_Noise_M1}
  }
  MaterialExpressionTextureSample_6 {
   class=MaterialExpressionTextureSample
   inputs [
    UVs {from=MaterialExpressionNamedRerouteUsage_121}
    Tex {from=MaterialExpressionNamedRerouteUsage_165}
    Apply View MipBias {from=""}
   ]
   properties {samplerSource=SSM_Wrap_WorldGroupSettings}
  }
  MaterialExpressionTextureSample_7 {
   class=MaterialExpressionTextureSample
   inputs [
    UVs {from=MaterialExpressionAdd_16}
    Tex {from=MaterialExpressionNamedRerouteUsage_166}
    Apply View MipBias {from=""}
   ]
   properties {samplerSource=SSM_Wrap_WorldGroupSettings}
  }
  MaterialExpressionTime_0 {class=MaterialExpressionTime}
  MaterialExpressionTransform_0 {class=MaterialExpressionTransform, inputs [None {from=MaterialExpressionNamedRerouteUsage_102}]}
  MaterialExpressionVectorParameter_0 {
   class=MaterialExpressionVectorParameter
   properties {
    group=Ice | 01_Globals
    parameterName=Parallax Thickness Remap
    sortPriority=420
    channelNames {a=n/a, b=Exp, g=Mult, r=Enable Remap}
    defaultValue=<rgba 0 1 1 1>
   }
  }
  MaterialExpressionVectorParameter_1 {
   class=MaterialExpressionVectorParameter
   properties {group=Ice | 04_Volume, parameterName=Volume - Scattering Coeffs, sortPriority=50, defaultValue=<rgba 0 0 0 1>}
  }
  MaterialExpressionVectorParameter_10 {
   class=MaterialExpressionVectorParameter
   properties {
    group=Ice | 05 _Absorption
    parameterName=Absorption - Diffuse Color Remap
    sortPriority=10
    channelNames {a=Value, b=Saturation, g=Hue Shift, r=Diffuse Map Exp}
    defaultValue=<rgba 1 0 0.5 0.5>
   }
  }
  MaterialExpressionVectorParameter_11 {
   class=MaterialExpressionVectorParameter
   properties {
    group=Ice | 02_Surface
    parameterName=Surface - Detail Normal Controls
    sortPriority=130
    channelNames {a=N/A, b="Min Intensity (Facing)", g="Max Intensity (Grazing)", r=Fresnel Exp}
    defaultValue=<rgba 5 2 0.6 0>
   }
  }
  MaterialExpressionVectorParameter_12 {
   class=MaterialExpressionVectorParameter
   properties {group=Ice | 01_Globals, parameterName=TOD - Color Day, sortPriority=420, defaultValue=<rgba 1 1 1 1>}
  }
  MaterialExpressionVectorParameter_13 {
   class=MaterialExpressionVectorParameter
   properties {group=Ice | 01_Globals, parameterName=TOD - Color Evening, sortPriority=430, defaultValue=<rgba 1 1 1 1>}
  }
  MaterialExpressionVectorParameter_14 {
   class=MaterialExpressionVectorParameter
   properties {group=Ice | 03_Refraction, parameterName=Refraction - Refracted Light Tint, sortPriority=80, defaultValue=<rgba 1 1 1 0>}
  }
  MaterialExpressionVectorParameter_15 {
   class=MaterialExpressionVectorParameter
   properties {group=Ice | 01_Globals, parameterName=TOD - Color Night, sortPriority=440, defaultValue=<rgba 1 1 1 1>}
  }
  MaterialExpressionVectorParameter_2 {
   class=MaterialExpressionVectorParameter
   properties {
    group=Ice | 05 _Absorption
    parameterName=Absorption - Parallax Map Controls
    sortPriority=50
    channelNames {a=N/A, b=Thickness Map Exp, g=Min Depth Mult, r=Max Depth}
    defaultValue=<rgba 0.2 0 0.5 0>
   }
  }
  MaterialExpressionVectorParameter_3 {
   class=MaterialExpressionVectorParameter
   properties {
    group=Ice | 06_Emission
    parameterName=Emission - Parallax Map Controls
    sortPriority=80
    channelNames {a=N/A, b=Thickness Map Exp, g=Min Depth Mult, r=Max Depth}
    defaultValue=<rgba 0.15 0 0.5 0>
   }
  }
  MaterialExpressionVectorParameter_4 {
   class=MaterialExpressionVectorParameter
   properties {
    group=Ice | 02_Surface
    parameterName=Surface - Detail Roughness Controls
    sortPriority=140
    channelNames {a=N/A, b=Roughness Mult, g=Roughness Exp, r=Mix Amount}
    defaultValue=<rgba 0 1 1 0>
   }
  }
  MaterialExpressionVectorParameter_5 {
   class=MaterialExpressionVectorParameter
   properties {
    group=Ice | 06_Emission
    parameterName=Emission - Variation Noise
    sortPriority=90
    channelNames {a=Pattern Exp, b=Pattern OldMax, g=Pattern OldMin, r=Pattern Scale}
    defaultValue=<rgba 5 0.1 0.8 2.2>
   }
  }
  MaterialExpressionVectorParameter_6 {
   class=MaterialExpressionVectorParameter
   properties {
    group=Ice | 02_Surface
    parameterName=Surface - Asset Specular Overrides
    sortPriority=10
    channelNames {a=Roughness Mult, b=Roughness Exp, g=Specular Mult, r=Specular Exp}
    defaultValue=<rgba 1 1 1 1>
   }
  }
  MaterialExpressionVectorParameter_7 {
   class=MaterialExpressionVectorParameter
   properties {group=Ice | 01_Globals, parameterName=TOD - Color Morning, sortPriority=410, defaultValue=<rgba 1 1 1 1>}
  }
  MaterialExpressionVectorParameter_8 {
   class=MaterialExpressionVectorParameter
   properties {
    group=Ice | 06_Emission
    parameterName=Emission - Noise Depth Breakup
    sortPriority=100
    channelNames {a=Add Constant Depth, b=Breakup Pattern Mult, g=Breakup Pattern Exp, r=Breakup Pattern Scale}
    defaultValue=<rgba 2 0.7 5 1.7>
   }
  }
  MaterialExpressionVectorParameter_9 {
   class=MaterialExpressionVectorParameter
   properties {
    group=Ice | 06_Emission
    parameterName=Emission - Emissive Fresnel
    sortPriority=10
    channelNames {a=Fresnel Exp, b=Fresnel Tint B, g=Fresnel Tint G, r=Fresnel Tint R}
    defaultValue=<rgba 0 0.3438 0.55 4>
   }
  }
  MaterialExpressionVertexNormalWS_0 {class=MaterialExpressionVertexNormalWS, inputs [], properties {}}
 }}
