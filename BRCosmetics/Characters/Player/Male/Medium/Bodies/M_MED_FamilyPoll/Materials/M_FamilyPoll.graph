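# M_FamilyPoll is the skeletal-mesh material for the FamilyPoll male medium body cosmetic, adding a volumetric ice effect to the shared character shading.
# MF_CharacterParent_2023_PreFX supplies material attributes and a tech art mask to MF_VolumetricIce_FamilyPoll through named reroutes.
# The ice result and PreFX data feed MF_CharacterParent_2023_PostFX.
# It uses expression-driven shading models with a hero skin subsurface profile and supports morph targets and mesh deformers.
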Material /BRCosmetics/Characters/Player/Male/Medium/Bodies/M_MED_FamilyPoll/Materials/M_FamilyPoll
graph {
 properties {
  bAutomaticallySetUsageInEditor=false
  bUseMaterialAttributes=true
  bUsedWithMeshDeformer=true
  bUsedWithMorphTargets=true
  bUsedWithSkeletalMesh=true
  shadingModel=MSM_FromMaterialExpression
  subsurfaceProfile=&SS_HeroSkin_02
  usedShadingModels=Default Lit | Subsurface | Clear Coat | Subsurface Profile | Cloth
 }
 nodes [
  n1 = MaterialFunctionCall() {materialFunction=&MF_CharacterParent_2023_PreFX}
  MaterialFunctionCall(
   Material=MaterialFunctionCall!(Material=NamedRerouteUsage(), Tech Art Mask=NamedRerouteUsage(), Additional Interior Emission) {materialFunction=&/MF_VolumetricIce_FamilyPoll}."Result (MA)"
   M=@n1.M
   Pre Skinned Local Position=@n1.M
   Distance Blend=@n1.M
   Fresnel=@n1.M
  ) {materialFunction=&MF_CharacterParent_2023_PostFX}
  NamedRerouteDeclaration(None=@n1.Result) {name="PreFX Material (MA)", nodeColor=<rgba 0 0.4522 1 1>}
  NamedRerouteDeclaration(None=@n1.Tech Art Mask) {name="Tech Art Mask (V3)", nodeColor=<rgba 0.6406 0 1 1>}
 ]
}
