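# Control Rig for the M_MED_FamilyPoll player body, adding procedural corrections to its skirt, loincloth, clavicle pads and armor collar.
# Parent constraints anchor skirt bases; swing-twist thigh and calf angles become weights for layered ModifyTransforms offsets.
# State flags such as crouch, DBNO, emote, jump and cape gate those offsets, often smoothed with AlphaInterp.
# Clavicle pads adjust only while hang gliding; head and neck spherical pose readers drive the collar.
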ControlRigBlueprint /BRCosmetics/Characters/Player/Male/Medium/Bodies/M_MED_FamilyPoll/Meshes/M_MED_FamilyPoll_CtrlRig
graphs [
 CollapseNode_ContainedGraph {nodes [
   n1 = RigVMFunctionEntryNode()
   &RigUnit_ParentConstraint(
    AdvancedSettings=(InterpolationType=Average,RotationOrderForFilter=XZY)
    Child=(Type=Bone,Name="base_skirt_front")
    Filter=(TranslationFilter=(bX=True,bY=True,bZ=True),RotationFilter=(bX=true,bY=true,bZ=true),ScaleFilter=(bX=true,bY=true,bZ=true))
    Parents=((Item=(Type=Bone,Name="dyn_skirt_R_A_01"),Weight=0.5),(Item=(Type=Bone,Name="dyn_skirt_L_A_01"),Weight=0.5))
    Weight=1
    ExecutePin=@n1.ExecutePin
   )
   > &RigUnit_ParentConstraint(
    AdvancedSettings=(InterpolationType=Average,RotationOrderForFilter=XZY)
    Child=(Type=Bone,Name="base_skirt_L_A")
    Parents=((Item=(Type=Bone,Name="thigh_twist_01_l"),Weight=1))
    Weight=1
   )
   > &RigUnit_ParentConstraint(
    AdvancedSettings=(InterpolationType=Average,RotationOrderForFilter=XZY)
    Child=(Type=Bone,Name="base_skirt_L_B")
    Parents=((Item=(Type=Bone,Name="thigh_twist_01_l"),Weight=1))
    Weight=1
   )
   > &RigUnit_ParentConstraint(
    AdvancedSettings=(InterpolationType=Average,RotationOrderForFilter=XZY)
    Child=(Type=Bone,Name="base_skirt_L_C")
    Parents=((Item=(Type=Bone,Name="thigh_twist_01_l"),Weight=1))
    Weight=1
   )
   > &RigUnit_ParentConstraint(
    AdvancedSettings=(InterpolationType=Average,RotationOrderForFilter=XZY)
    Child=(Type=Bone,Name="base_skirt_L_D")
    Parents=((Item=(Type=Bone,Name="thigh_twist_01_l"),Weight=1))
    Weight=1
   )
   > &RigUnit_ParentConstraint(
    AdvancedSettings=(InterpolationType=Average,RotationOrderForFilter=XZY)
    Child=(Type=Bone,Name="base_skirt_R_A")
    Parents=((Item=(Type=Bone,Name="thigh_twist_01_r"),Weight=1))
    Weight=1
   )
   > &RigUnit_ParentConstraint(
    AdvancedSettings=(InterpolationType=Average,RotationOrderForFilter=XZY)
    Child=(Type=Bone,Name="base_skirt_R_B")
    Parents=((Item=(Type=Bone,Name="thigh_twist_01_r"),Weight=1))
    Weight=1
   )
   > &RigUnit_ParentConstraint(
    AdvancedSettings=(InterpolationType=Average,RotationOrderForFilter=XZY)
    Child=(Type=Bone,Name="base_skirt_R_C")
    Parents=((Item=(Type=Bone,Name="thigh_twist_01_r"),Weight=1))
    Weight=1
   )
   > n2 = &RigUnit_ParentConstraint(
    AdvancedSettings=(InterpolationType=Average,RotationOrderForFilter=XZY)
    Child=(Type=Bone,Name="base_skirt_R_D")
    Parents=((Item=(Type=Bone,Name="thigh_twist_01_r"),Weight=1))
    Weight=1
   )
   RigVMFunctionReturnNode(ExecutePin=@n2.ExecutePin)
   &RigUnit_ParentConstraint(
    AdvancedSettings=(RotationOrderForFilter=XZY)
    Child=(Type=Bone,Name="base_skirt_front")
    ExecutePin=()
    Parents="((Item=(Type=Bone,Name=\"spine_03\"),Weight=0.0162),(Item=(Type=Bone,Name=\"spine_01\"),Weight=0.0244),(Item=(Type=Bone,Name=\"spine_02\"),Weight=0.0286),(Item=(Type=Bone,Name=\"thigh_twist_01_r\"),Weight=0.0456),(Item=(Type=Bone,Name=\"thigh_twist_01_l\"),Weight=0. ...(truncated)"
    Weight=1
   )
   > &RigUnit_ParentConstraint(
    AdvancedSettings=(RotationOrderForFilter=XZY)
    Child=(Type=Bone,Name="base_skirt_L_A")
    Parents="((Item=(Type=Bone,Name=\"spine_03\"),Weight=0.0018),(Item=(Type=Bone,Name=\"thigh_l\"),Weight=0.0022),(Item=(Type=Bone,Name=\"spine_01\"),Weight=0.003),(Item=(Type=Bone,Name=\"thigh_twist_01_r\"),Weight=0.0038),(Item=(Type=Bone,Name=\"spine_02\"),Weight=0.0053),(Ite ...(truncated)"
    Weight=1
   )
   > &RigUnit_ParentConstraint(
    AdvancedSettings=(RotationOrderForFilter=XZY)
    Child=(Type=Bone,Name="base_skirt_L_B")
    Parents=((Item=(Type=Bone,Name="thigh_l"),Weight=0.0065),(Item=(Type=Bone,Name="thigh_twist_01_l"),Weight=0.3077),(Item=(Type=Bone,Name="pelvis"),Weight=0.6857))
    Weight=1
   )
   > &RigUnit_ParentConstraint(
    AdvancedSettings=(RotationOrderForFilter=XZY)
    Child=(Type=Bone,Name="base_skirt_L_C")
    Parents="((Item=(Type=Bone,Name=\"thigh_twist_01_r\"),Weight=0.0006),(Item=(Type=Bone,Name=\"thigh_l\"),Weight=0.0008),(Item=(Type=Bone,Name=\"spine_02\"),Weight=0.0022),(Item=(Type=Bone,Name=\"thigh_twist_01_l\"),Weight=0.1092),(Item=(Type=Bone,Name=\"pelvis\"),Weight=0.887 ...(truncated)"
    Weight=1
   )
   > &RigUnit_ParentConstraint(
    AdvancedSettings=(RotationOrderForFilter=XZY)
    Child=(Type=Bone,Name="base_skirt_L_D")
    Parents="((Item=(Type=Bone,Name=\"spine_02\"),Weight=0.0001),(Item=(Type=Bone,Name=\"spine_01\"),Weight=0.0001),(Item=(Type=Bone,Name=\"thigh_l\"),Weight=0.0002),(Item=(Type=Bone,Name=\"thigh_twist_01_r\"),Weight=0.012),(Item=(Type=Bone,Name=\"thigh_twist_01_l\"),Weight=0.04 ...(truncated)"
    Weight=1
   )
   > &RigUnit_ParentConstraint(
    AdvancedSettings=(RotationOrderForFilter=XZY)
    Child=(Type=Bone,Name="base_skirt_R_A")
    Parents="((Item=(Type=Bone,Name=\"spine_03\"),Weight=0.0011),(Item=(Type=Bone,Name=\"spine_01\"),Weight=0.0019),(Item=(Type=Bone,Name=\"thigh_twist_01_l\"),Weight=0.0028),(Item=(Type=Bone,Name=\"thigh_r\"),Weight=0.0029),(Item=(Type=Bone,Name=\"spine_02\"),Weight=0.004),(Ite ...(truncated)"
    Weight=1
   )
   > &RigUnit_ParentConstraint(
    AdvancedSettings=(RotationOrderForFilter=XZY)
    Child=(Type=Bone,Name="base_skirt_R_B")
    Parents=((Item=(Type=Bone,Name="thigh_r"),Weight=0.0037),(Item=(Type=Bone,Name="thigh_twist_01_r"),Weight=0.2694),(Item=(Type=Bone,Name="pelvis"),Weight=0.7268))
    Weight=1
   )
   > &RigUnit_ParentConstraint(
    AdvancedSettings=(RotationOrderForFilter=XZY)
    Child=(Type=Bone,Name="base_skirt_R_C")
    Parents="((Item=(Type=Bone,Name=\"thigh_twist_01_l\"),Weight=0.0006),(Item=(Type=Bone,Name=\"spine_02\"),Weight=0.0008),(Item=(Type=Bone,Name=\"thigh_r\"),Weight=0.0018),(Item=(Type=Bone,Name=\"thigh_twist_01_r\"),Weight=0.1554),(Item=(Type=Bone,Name=\"pelvis\"),Weight=0.841 ...(truncated)"
    Weight=1
   )
   > &RigUnit_ParentConstraint(
    AdvancedSettings=(RotationOrderForFilter=XZY)
    Child=(Type=Bone,Name="base_skirt_R_D")
    Parents="((Item=(Type=Bone,Name=\"thigh_r\"),Weight=0.0001),(Item=(Type=Bone,Name=\"spine_02\"),Weight=0.0002),(Item=(Type=Bone,Name=\"spine_01\"),Weight=0.0002),(Item=(Type=Bone,Name=\"thigh_twist_01_l\"),Weight=0.0118),(Item=(Type=Bone,Name=\"thigh_twist_01_r\"),Weight=0.0 ...(truncated)"
    Weight=1
   )
   > &RigUnit_ParentConstraint(
    AdvancedSettings=(RotationOrderForFilter=XZY)
    Child=(Type=Bone,Name="base_LoinCloth_back")
    Parents="((Item=(Type=Bone,Name=\"thigh_twist_01_l\"),Weight=0.0029),(Item=(Type=Bone,Name=\"spine_02\"),Weight=0.0029),(Item=(Type=Bone,Name=\"spine_01\"),Weight=0.0039),(Item=(Type=Bone,Name=\"thigh_twist_01_r\"),Weight=0.0066),(Item=(Type=Bone,Name=\"pelvis\"),Weight=0.98 ...(truncated)"
    Weight=1
   )
  ]}
 CollapseNode_ContainedGraph {nodes [
   RigVMFunctionEntryNode()
   n1 = &RigUnit_GetRelativeTransformForItem(Child=(Type=Bone,Name="calf_l"), Parent=(Type=Null,Name="calf_l_null"), bChildInitial=False, bParentInitial=False)
   > RigVMVariableNode(Value=&RigVMFunction_MathQuaternionToEuler(
     RotationOrder=ZYX
     Value=&RigVMFunction_MathQuaternionSwingTwist(TwistAxis=(X=1,Y=0,Z=0), Input=@n1.RelativeTransform.Rotation).Swing
    ).Result.Z) {variable=l calfLocal rz float}
   n2 = &RigUnit_GetRelativeTransformForItem(Child=(Type=Bone,Name="calf_r"), Parent=(Type=Null,Name="calf_r_null"), bChildInitial=False, bParentInitial=False)
   > RigVMVariableNode(Value=&RigVMFunction_MathQuaternionToEuler(
     RotationOrder=ZYX
     Value=&RigVMFunction_MathQuaternionSwingTwist(TwistAxis=(X=1,Y=0,Z=0), Input=@n2.RelativeTransform.Rotation).Swing
    ).Result.Z) {variable=r calfLocal rz float}
   RigVMFunction_MathQuaternionSwingTwist = &RigVMFunction_MathQuaternionSwingTwist(
    TwistAxis=(X=1,Y=0,Z=0)
    Input=&RigUnit_GetRelativeTransformForItem(Child=(Type=Bone,Name="thigh_l"), Parent=(Type=Null,Name="thigh_l_null"), bChildInitial=False, bParentInitial=False).RelativeTransform.Rotation
   )
   RigVMFunction_MathQuaternionToEuler = &RigVMFunction_MathQuaternionToEuler(RotationOrder=ZYX, Value=@RigVMFunction_MathQuaternionSwingTwist.Swing)
   > RigVMVariableNode(Value=@RigVMFunction_MathQuaternionToEuler.Result.Y) {variable=l thighLocal ry float}
   > RigVMVariableNode(Value=@RigVMFunction_MathQuaternionToEuler.Result.Z) {variable=l thighLocal rz float}
   n3 = &RigVMFunction_MathQuaternionToEuler(
    RotationOrder=ZYX
    Value=&RigVMFunction_MathQuaternionSwingTwist(
     TwistAxis=(X=1,Y=0,Z=0)
     Input=&RigUnit_GetRelativeTransformForItem(Child=(Type=Bone,Name="thigh_r"), Parent=(Type=Null,Name="thigh_r_null"), bChildInitial=False, bParentInitial=False).RelativeTransform.Rotation
    ).Swing
   )
   > RigVMVariableNode(Value=@n3.Result.Y) {variable=r thighLocal ry float}
   > RigVMVariableNode(Value=@n3.Result.Z) {variable=r thighLocal rz float}
   > RigVMFunctionReturnNode()
  ]}
 CollapseNode_ContainedGraph {nodes [
   n1 = RigVMFunctionEntryNode()
   RigVMFunction_Sequence = &RigVMFunction_Sequence(ExecuteContext=@n1.ExecuteContext)
   RigVMFunction_Sequence_1 = RigVMAggregateNode(ExecuteContext=@RigVMFunction_Sequence.A)
   GetAlphaFromQuatDelta = RigVMFunctionReferenceNode(
    Rotation Order=ZYX
    child=(Type=Bone,Name="calf_l")
    debugRotation=false
    onlyTwist=false
    outputMinMax=(X=0,Y=1)
    parent=(Type=Null,Name="calf_l_null")
    premult=1
    rotationMinMax=(X=6,Y=89)
    triggerAxis=Z
    twistAxis=(X=1,Y=0,Z=0)
    ExecuteContext=@RigVMFunction_Sequence_1.B
   )
   > &RigUnit_ModifyTransforms(
    ItemToModify="((Item=(Type=Bone,Name=\"base_skirt_L_D_03\"),Transform=(Rotation=(X=0,Y=-0.5736,Z=0,W=0.8192),Translation=(X=0,Y=0,Z=0),Scale3D=(X=1,Y=1,Z=1))),(Item=(Type=Bone,Name=\"base_skirt_L_C_03\"),Transform=(Rotation=(X=0,Y=-0.4226,Z=0,W=0.9063),Translation=(X=0,Y=0, ...(truncated)"
    WeightMinimum=0
    Weight=@GetAlphaFromQuatDelta.alpha
   )
   > GetAlphaFromQuatDelta_1 = RigVMFunctionReferenceNode(
    Rotation Order=ZYX
    child=(Type=Bone,Name="calf_r")
    debugRotation=false
    onlyTwist=false
    outputMinMax=(X=0,Y=1)
    parent=(Type=Null,Name="calf_r_null")
    premult=1
    rotationMinMax=(X=6,Y=89)
    triggerAxis=Z
    twistAxis=(X=1,Y=0,Z=0)
   )
   > &RigUnit_ModifyTransforms(
    ItemToModify="((Item=(Type=Bone,Name=\"base_skirt_R_D_03\"),Transform=(Rotation=(X=0,Y=-0.5736,Z=0,W=0.8192),Translation=(X=0,Y=0,Z=0),Scale3D=(X=1,Y=1,Z=1))),(Item=(Type=Bone,Name=\"base_skirt_R_C_03\"),Transform=(Rotation=(X=0,Y=-0.4226,Z=0,W=0.9063),Translation=(X=0,Y=0, ...(truncated)"
    WeightMinimum=0
    Weight=@GetAlphaFromQuatDelta_1.alpha
   )
   > GetAlphaFromQuatDelta_2 = RigVMFunctionReferenceNode(
    Rotation Order=ZYX
    child=(Type=Bone,Name="calf_l")
    debugRotation=false
    onlyTwist=false
    outputMinMax=(X=0,Y=1)
    parent=(Type=Null,Name="calf_l_null")
    premult=1
    rotationMinMax=(X=90,Y=135)
    triggerAxis=Z
    twistAxis=(X=1,Y=0,Z=0)
   )
   > &RigUnit_ModifyTransforms(
    ItemToModify="((Item=(Type=Bone,Name=\"base_skirt_L_D_03\"),Transform=(Rotation=(X=0,Y=-0.6088,Z=0,W=0.7934),Translation=(X=-3,Y=0,Z=-6),Scale3D=(X=1,Y=1,Z=1))),(Item=(Type=Bone,Name=\"base_skirt_L_C_03\"),Transform=(Rotation=(X=-0.1495,Y=-0.3965,Z=0.0356,W=0.9051),Translat ...(truncated)"
    WeightMinimum=0
    Weight=@GetAlphaFromQuatDelta_2.alpha
   )
   > GetAlphaFromQuatDelta_3 = RigVMFunctionReferenceNode(
    Rotation Order=ZYX
    child=(Type=Bone,Name="calf_r")
    debugRotation=false
    onlyTwist=false
    outputMinMax=(X=0,Y=1)
    parent=(Type=Null,Name="calf_r_null")
    premult=1
    rotationMinMax=(X=90,Y=135)
    triggerAxis=Z
    twistAxis=(X=1,Y=0,Z=0)
   )
   > &RigUnit_ModifyTransforms(
    ItemToModify="((Item=(Type=Bone,Name=\"base_skirt_R_D_03\"),Transform=(Rotation=(X=0,Y=-0.6088,Z=0,W=0.7934),Translation=(X=3,Y=0,Z=6),Scale3D=(X=1,Y=1,Z=1))),(Item=(Type=Bone,Name=\"base_skirt_R_C_03\"),Transform=(Rotation=(X=-0.1495,Y=-0.3965,Z=0.0356,W=0.9051),Translatio ...(truncated)"
    WeightMinimum=0
    Weight=@GetAlphaFromQuatDelta_3.alpha
   )
   &RigUnit_ModifyTransforms(
    ItemToModify="((Item=(Type=Bone,Name=\"dyn_skirt_L_A_01\"),Transform=(Rotation=(X=0,Y=0,Z=0,W=1),Translation=(X=0,Y=0,Z=0),Scale3D=(X=1,Y=1,Z=1))),(Item=(Type=Bone,Name=\"base_skirt_L_B\"),Transform=(Rotation=(X=0,Y=0,Z=0,W=1),Translation=(X=0,Y=0,Z=0),Scale3D=(X=1,Y=1,Z=1) ...(truncated)"
    WeightMinimum=0
    Weight=&RigVMFunction_MathDoubleRemap(SourceMaximum=120, SourceMinimum=20, TargetMaximum=1, TargetMinimum=0, Value=RigVMVariableNode() {variable=l calfLocal rz float}.Value).Result
    ExecutePin=@RigVMFunction_Sequence_1.C
   )
   > &RigUnit_ModifyTransforms(
    ItemToModify="((Item=(Type=Bone,Name=\"dyn_skirt_R_A_01\"),Transform=(Rotation=(X=0,Y=0,Z=0,W=1),Translation=(X=0,Y=0,Z=0),Scale3D=(X=1,Y=1,Z=1))),(Item=(Type=Bone,Name=\"dyn_skirt_R_B_01\"),Transform=(Rotation=(X=0,Y=0,Z=0,W=1),Translation=(X=0,Y=0,Z=0),Scale3D=(X=1,Y=1,Z= ...(truncated)"
    WeightMinimum=0
    Weight=&RigVMFunction_MathDoubleRemap(SourceMaximum=120, SourceMinimum=20, TargetMaximum=1, TargetMinimum=0, Value=RigVMVariableNode() {variable=r calfLocal rz float}.Value).Result
   )
   > GetAlphaFromQuatDelta_4 = RigVMFunctionReferenceNode(
    Rotation Order=ZYX
    child=(Type=Bone,Name="thigh_l")
    debugRotation=false
    onlyTwist=false
    outputMinMax=(X=0,Y=1)
    parent=(Type=Null,Name="thigh_l_null")
    premult=1
    rotationMinMax=(X=0,Y=-25)
    triggerAxis=Z
    twistAxis=(X=1,Y=0,Z=0)
   )
   > &RigUnit_ModifyTransforms(
    ItemToModify="((Item=(Type=Bone,Name=\"dyn_skirt_L_A_01\"),Transform=(Rotation=(X=-0.0872,Y=0,Z=0,W=0.9962),Translation=(X=0,Y=0,Z=0),Scale3D=(X=1,Y=1,Z=1))),(Item=(Type=Bone,Name=\"dyn_skirt_R_A_01\"),Transform=(Rotation=(X=0,Y=0,Z=0,W=1),Translation=(X=0,Y=0,Z=0),Scale3D= ...(truncated)"
    WeightMinimum=0
    Weight=@GetAlphaFromQuatDelta_4.alpha
   )
   > GetAlphaFromQuatDelta_5 = RigVMFunctionReferenceNode(
    Rotation Order=ZYX
    child=(Type=Bone,Name="thigh_r")
    debugRotation=false
    onlyTwist=false
    outputMinMax=(X=0,Y=1)
    parent=(Type=Null,Name="thigh_r_null")
    premult=1
    rotationMinMax=(X=0,Y=-25)
    triggerAxis=Z
    twistAxis=(X=1,Y=0,Z=0)
   )
   > &RigUnit_ModifyTransforms(
    ItemToModify="((Item=(Type=Bone,Name=\"dyn_skirt_R_A_01\"),Transform=(Rotation=(X=-0.1736,Y=0,Z=0,W=0.9848),Translation=(X=0,Y=0,Z=0),Scale3D=(X=1,Y=1,Z=1))),(Item=(Type=Bone,Name=\"dyn_skirt_L_A_01\"),Transform=(Rotation=(X=0,Y=0,Z=0,W=1),Translation=(X=0,Y=0,Z=0),Scale3D= ...(truncated)"
    WeightMinimum=0
    Weight=@GetAlphaFromQuatDelta_5.alpha
   )
   n2 = &RigVMFunction_AnimEvalRichCurve(
    Curve="(EditorCurveData=(Keys=((InterpMode=RCIM_Cubic,TangentMode=RCTM_User,TangentWeightMode=RCTWM_WeightedNone,Time=0,Value=0,ArriveTangent=1.6904,ArriveTangentWeight=0,LeaveTangent=1.6904,LeaveTangentWeight=0),(InterpMode=RCIM_Cubic,TangentMode=RCTM_User,Tange ...(truncated)"
    SourceMinimum=0
    TargetMinimum=0
    Value=&RigVMFunction_MathDoubleMul(A=@GetAlphaFromQuatDelta_4.alpha, B=@GetAlphaFromQuatDelta_5.alpha).Result
   )
   > &RigUnit_ModifyTransforms(
    ItemToModify="((Item=(Type=Bone,Name=\"dyn_skirt_R_A_01\"),Transform=(Rotation=(X=0.1736,Y=0,Z=0,W=0.9848),Translation=(X=0,Y=0,Z=0),Scale3D=(X=1,Y=1,Z=1))),(Item=(Type=Bone,Name=\"dyn_skirt_L_A_01\"),Transform=(Rotation=(X=0.1736,Y=0,Z=0,W=0.9848),Translation=(X=0,Y=0,Z=0) ...(truncated)"
    WeightMinimum=0
    Weight=@n2.Result
   )
   n3 = &RigVMFunction_AlphaInterp(
    InterpSpeedDecreasing=1
    InterpSpeedIncreasing=5
    Scale=1
    bClampResult=true
    Value=&RigVMFunction_MathBoolToFloat(Value=RigVMVariableNode() {variable=Is Crouching}.Value).Result
   )
   n4 = &RigVMFunction_MathDoubleMul(
    B=@n3.Result
    A=&RigVMFunction_MathDoubleRemap(
     SourceMaximum=45
     SourceMinimum=89
     TargetMaximum=1
     TargetMinimum=0
     Value=&RigUnit_GetRelativeTransformForItem(Child=(Type=Bone,Name="pelvis"), Parent=(Type=Bone,Name="root"), bChildInitial=False, bParentInitial=False).RelativeTransform.Translation.Z
    ).Result
   )
   > &RigUnit_ModifyTransforms(
    ItemToModify="((Item=(Type=Bone,Name=\"base_skirt_R_D_03\"),Transform=(Rotation=(X=0,Y=0,Z=0,W=1),Translation=(X=0,Y=0,Z=0),Scale3D=(X=1,Y=1,Z=1))),(Item=(Type=Bone,Name=\"base_skirt_R_C_03\"),Transform=(Rotation=(X=0,Y=0.0872,Z=0,W=0.9962),Translation=(X=2,Y=0,Z=-2),Scale3 ...(truncated)"
    WeightMinimum=0
    Weight=@n4.Result
   )
   > &RigUnit_ModifyTransforms(
    ItemToModify="((Item=(Type=Bone,Name=\"base_skirt_L_D_03\"),Transform=(Rotation=(X=0,Y=0,Z=-0.2588,W=0.9659),Translation=(X=0,Y=0,Z=0),Scale3D=(X=1,Y=1,Z=1))),(Item=(Type=Bone,Name=\"base_skirt_L_C_03\"),Transform=(Rotation=(X=0.0338,Y=0.2566,Z=-0.1261,W=0.9577),Translation ...(truncated)"
    WeightMinimum=0
    Weight=@n4.Result
   )
   n5 = &RigVMFunction_MathBoolToFloat(Value=RigVMVariableNode() {variable=Anim dynamics On}.Value)
   n6 = &RigVMFunction_AlphaInterp(InterpSpeedDecreasing=5, InterpSpeedIncreasing=10, Scale=1, bClampResult=true, Value=@n5.Result)
   > &RigUnit_ModifyTransforms(
    ItemToModify="((Item=(Type=Bone,Name=\"dyn_skirt_R_D_04\"),Transform=(Rotation=(X=0.3315,Y=-0.1039,Z=0.2803,W=-0.8948),Translation=(X=0,Y=0,Z=0),Scale3D=(X=1,Y=1,Z=1))),(Item=(Type=Bone,Name=\"dyn_skirt_R_D_05\"),Transform=(Rotation=(X=0,Y=0,Z=0,W=1),Translation=(X=0,Y=0,Z= ...(truncated)"
    WeightMinimum=0
    Weight=&RigVMFunction_MathDoubleMul(B=@n6.Result, A=@n4.Result).Result
   )
   n7 = &RigVMFunction_AlphaInterp(
    InterpSpeedDecreasing=5
    InterpSpeedIncreasing=5
    Scale=1
    bClampResult=true
    Value=&RigVMFunction_MathBoolToFloat(Value=RigVMVariableNode() {variable=isDBNO}.Value).Result
   )
   > &RigUnit_ModifyTransforms(
    ItemToModify="((Item=(Type=Bone,Name=\"dyn_skirt_L_D_01\"),Transform=(Rotation=(X=0.1305,Y=0,Z=0,W=0.9914),Translation=(X=0,Y=0,Z=0),Scale3D=(X=1,Y=1,Z=1))),(Item=(Type=Bone,Name=\"dyn_skirt_R_D_01\"),Transform=(Rotation=(X=0.1735,Y=0.043,Z=-0.0076,W=0.9839),Translation=(X= ...(truncated)"
    WeightMinimum=0
    Weight=@n7.Result
   )
   GetAlphaFromQuatDelta_6 = RigVMFunctionReferenceNode(
    Rotation Order=ZYX
    child=(Type=Bone,Name="thigh_l")
    debugRotation=false
    onlyTwist=false
    outputMinMax=(X=0,Y=1)
    parent=(Type=Null,Name="thigh_l_null")
    premult=1
    rotationMinMax=(X=0,Y=-25)
    triggerAxis=Y
    twistAxis=(X=1,Y=0,Z=0)
    ExecuteContext=@RigVMFunction_Sequence_1.A
   )
   n8 = RigVMVariableNode() {variable=Is Crouching}
   n9 = &RigVMFunction_AlphaInterp(
    InterpSpeedDecreasing=5
    InterpSpeedIncreasing=10
    Scale=1
    bClampResult=true
    Value=&RigVMFunction_MathDoubleRemap(
     SourceMaximum=1
     SourceMinimum=0
     TargetMaximum=0
     TargetMinimum=1
     Value=&RigVMFunction_MathBoolToFloat(Value=@n8.Value).Result
    ).Result
   )
   Multiply_5 = RigVMAggregateNode(
    B=@n9.Result
    A=@GetAlphaFromQuatDelta_6.alpha
    C=RigVMDispatchNode(False=1, True=0, Condition=RigVMVariableNode() {variable=Is Playing Emote}.Value) {template="DISPATCH_RigVMDispatch_If(in Condition,in True,in False,out Result)"}.Result
   )
   > &RigUnit_ModifyTransforms(
    ItemToModify="((Item=(Type=Bone,Name=\"base_LoinCloth_front\"),Transform=(Rotation=(X=-0.0411,Y=0.0854,Z=0.0467,W=0.9944),Translation=(X=0,Y=0,Z=0),Scale3D=(X=1,Y=1,Z=1))),(Item=(Type=Bone,Name=\"dyn_LoinCloth_back_01\"),Transform=(Rotation=(X=0,Y=0,Z=-0.0436,W=0.999),Trans ...(truncated)"
    WeightMinimum=0
    Weight=@Multiply_5.Result
   )
   > &RigUnit_ModifyTransforms(
    ItemToModify="((Item=(Type=Bone,Name=\"dyn_skirt_L_A_01\"),Transform=(Rotation=(X=-0.5536,Y=-0.121,Z=0.1275,W=0.814),Translation=(X=0,Y=0,Z=0),Scale3D=(X=1,Y=1,Z=1))),(Item=(Type=Bone,Name=\"base_skirt_L_B\"),Transform=(Rotation=(X=0,Y=0,Z=0,W=1),Translation=(X=0,Y=0,Z=0),S ...(truncated)"
    WeightMinimum=0
    Weight=@Multiply_5.Result
   )
   > GetAlphaFromQuatDelta_7 = RigVMFunctionReferenceNode(
    Rotation Order=ZYX
    child=(Type=Bone,Name="thigh_r")
    debugRotation=false
    onlyTwist=false
    outputMinMax=(X=0,Y=1)
    parent=(Type=Null,Name="thigh_r_null")
    premult=1
    rotationMinMax=(X=0,Y=-25)
    triggerAxis=Y
    twistAxis=(X=1,Y=0,Z=0)
   )
   n10 = RigVMVariableNode() {variable=Is Crouching}
   n11 = &RigVMFunction_AlphaInterp(
    InterpSpeedDecreasing=5
    InterpSpeedIncreasing=10
    Scale=1
    bClampResult=true
    Value=&RigVMFunction_MathDoubleRemap(
     SourceMaximum=1
     SourceMinimum=0
     TargetMaximum=0
     TargetMinimum=1
     Value=&RigVMFunction_MathBoolToFloat(Value=@n10.Value).Result
    ).Result
   )
   Multiply_6 = RigVMAggregateNode(
    B=@n11.Result
    A=@GetAlphaFromQuatDelta_7.alpha
    C=RigVMDispatchNode(False=1, True=0, Condition=RigVMVariableNode() {variable=Is Playing Emote}.Value) {template="DISPATCH_RigVMDispatch_If(in Condition,in True,in False,out Result)"}.Result
   )
   > &RigUnit_ModifyTransforms(
    ItemToModify="((Item=(Type=Bone,Name=\"base_LoinCloth_front\"),Transform=(Rotation=(X=0.0038,Y=-0.0871,Z=0.0435,W=0.9952),Translation=(X=0,Y=0,Z=0),Scale3D=(X=1,Y=1,Z=1))),(Item=(Type=Bone,Name=\"dyn_LoinCloth_back_01\"),Transform=(Rotation=(X=0,Y=0,Z=-0.0436,W=0.999),Trans ...(truncated)"
    WeightMinimum=0
    Weight=@Multiply_6.Result
   )
   > &RigUnit_ModifyTransforms(
    ItemToModify="((Item=(Type=Bone,Name=\"dyn_skirt_R_A_01\"),Transform=(Rotation=(X=-0.1708,Y=-0.0314,Z=0.1533,W=0.9728),Translation=(X=0,Y=0,Z=0),Scale3D=(X=1,Y=1,Z=1))),(Item=(Type=Bone,Name=\"dyn_skirt_R_B_01\"),Transform=(Rotation=(X=0,Y=0,Z=0,W=1),Translation=(X=0,Y=0,Z= ...(truncated)"
    WeightMinimum=0
    Weight=@Multiply_6.Result
   )
   > GetAlphaFromQuatDelta_8 = RigVMFunctionReferenceNode(
    Rotation Order=ZYX
    child=(Type=Bone,Name="thigh_l")
    debugRotation=false
    onlyTwist=false
    outputMinMax=(X=0,Y=1)
    parent=(Type=Null,Name="thigh_l_null")
    premult=1
    rotationMinMax=(X=0,Y=24)
    triggerAxis=Y
    twistAxis=(X=1,Y=0,Z=0)
   )
   n12 = RigVMDispatchNode(False=1, True=0, Condition=RigVMVariableNode() {variable=Is Playing Emote}.Value) {template="DISPATCH_RigVMDispatch_If(in Condition,in True,in False,out Result)"}
   n13 = &RigVMFunction_MathDoubleMul(B=@n12.Result, A=@GetAlphaFromQuatDelta_8.alpha)
   > &RigUnit_ModifyTransforms(
    ItemToModify="((Item=(Type=Bone,Name=\"dyn_skirt_L_A_01\"),Transform=(Rotation=(X=-0.0091,Y=-0.0427,Z=-0.2077,W=0.9772),Translation=(X=0,Y=0,Z=0),Scale3D=(X=1,Y=1,Z=1))),(Item=(Type=Bone,Name=\"base_skirt_L_B\"),Transform=(Rotation=(X=0,Y=-0.2079,Z=0,W=0.9781),Translation=( ...(truncated)"
    WeightMinimum=0
    Weight=@n13.Result
   )
   > &RigUnit_ModifyTransforms(
    ItemToModify=((Item=(Type=Bone,Name="base_LoinCloth_front"),Transform=(Rotation=(X=0,Y=-0.0872,Z=0,W=0.9962),Translation=(X=0,Y=0,Z=0),Scale3D=(X=1,Y=1,Z=1))))
    WeightMinimum=0
    Weight=@n13.Result
   )
   > GetAlphaFromQuatDelta_9 = RigVMFunctionReferenceNode(
    Rotation Order=ZYX
    child=(Type=Bone,Name="thigh_r")
    debugRotation=false
    onlyTwist=false
    outputMinMax=(X=0,Y=1)
    parent=(Type=Null,Name="thigh_r_null")
    premult=1
    rotationMinMax=(X=0,Y=24)
    triggerAxis=Y
    twistAxis=(X=1,Y=0,Z=0)
   )
   n14 = &RigVMFunction_MathDoubleMul(B=@n12.Result, A=@GetAlphaFromQuatDelta_9.alpha)
   > &RigUnit_ModifyTransforms(
    ItemToModify="((Item=(Type=Bone,Name=\"dyn_skirt_R_A_01\"),Transform=(Rotation=(X=-0.0091,Y=-0.0427,Z=-0.2077,W=0.9772),Translation=(X=0,Y=0,Z=0),Scale3D=(X=1,Y=1,Z=1))),(Item=(Type=Bone,Name=\"dyn_skirt_R_B_01\"),Transform=(Rotation=(X=0,Y=-0.2079,Z=0,W=0.9781),Translation ...(truncated)"
    WeightMinimum=0
    Weight=@n14.Result
   )
   > &RigUnit_ModifyTransforms(
    ItemToModify=((Item=(Type=Bone,Name="base_LoinCloth_front"),Transform=(Rotation=(X=0,Y=0.0872,Z=0,W=0.9962),Translation=(X=0,Y=0,Z=0),Scale3D=(X=1,Y=1,Z=1))))
    WeightMinimum=0
    Weight=@n14.Result
   )
   > GetAlphaFromQuatDelta_11 = RigVMFunctionReferenceNode(
    Rotation Order=ZYX
    child=(Type=Bone,Name="thigh_r")
    debugRotation=false
    onlyTwist=false
    outputMinMax=(X=0,Y=1)
    parent=(Type=Null,Name="thigh_r_null")
    premult=1
    rotationMinMax=(X=0,Y=-25)
    triggerAxis=Y
    twistAxis=(X=1,Y=0,Z=0)
   )
   n15 = &RigVMFunction_AlphaInterp(
    InterpSpeedDecreasing=5
    InterpSpeedIncreasing=10
    Scale=1
    bClampResult=true
    Value=&RigVMFunction_MathBoolToFloat(Value=RigVMVariableNode() {variable=Anim dynamics On}.Value).Result
   )
   > &RigUnit_ModifyTransforms(
    ItemToModify="((Item=(Type=Bone,Name=\"base_skirt_R_A_03\"),Transform=(Rotation=(X=0,Y=0.0824,Z=0,W=-0.9966),Translation=(X=0,Y=0,Z=0),Scale3D=(X=1,Y=1,Z=1))),(Item=(Type=Bone,Name=\"base_skirt_R_B_03\"),Transform=(Rotation=(X=0,Y=0,Z=-0.1804,W=-0.9836),Translation=(X=0,Y=0 ...(truncated)"
    WeightMinimum=0
    Weight=&RigVMFunction_MathDoubleMul(B=@n15.Result, A=@GetAlphaFromQuatDelta_11.alpha).Result
   )
   > GetAlphaFromQuatDelta_10 = RigVMFunctionReferenceNode(
    Rotation Order=ZYX
    child=(Type=Bone,Name="thigh_l")
    debugRotation=false
    onlyTwist=false
    outputMinMax=(X=0,Y=1)
    parent=(Type=Null,Name="thigh_l_null")
    premult=1
    rotationMinMax=(X=0,Y=24)
    triggerAxis=Y
    twistAxis=(X=1,Y=0,Z=0)
   )
   n16 = RigVMDispatchNode(False=0, True=1, Condition=RigVMVariableNode() {variable=Is Playing Emote}.Value) {template="DISPATCH_RigVMDispatch_If(in Condition,in True,in False,out Result)"}
   > &RigUnit_ModifyTransforms(
    ItemToModify="((Item=(Type=Bone,Name=\"dyn_skirt_L_A_01\"),Transform=(Rotation=(X=0.0091,Y=0.0427,Z=-0.2077,W=0.9772),Translation=(X=0,Y=0,Z=0),Scale3D=(X=1,Y=1,Z=1))),(Item=(Type=Bone,Name=\"base_skirt_L_B\"),Transform=(Rotation=(X=0,Y=0,Z=0,W=1),Translation=(X=0,Y=0,Z=0), ...(truncated)"
    WeightMinimum=0
    Weight=&RigVMFunction_MathDoubleMul(A=@GetAlphaFromQuatDelta_10.alpha, B=@n16.Result).Result
   )
   > GetAlphaFromQuatDelta_12 = RigVMFunctionReferenceNode(
    Rotation Order=ZYX
    child=(Type=Bone,Name="thigh_r")
    debugRotation=false
    onlyTwist=false
    outputMinMax=(X=0,Y=1)
    parent=(Type=Null,Name="thigh_r_null")
    premult=1
    rotationMinMax=(X=0,Y=24)
    triggerAxis=Y
    twistAxis=(X=1,Y=0,Z=0)
   )
   > &RigUnit_ModifyTransforms(
    ItemToModify="((Item=(Type=Bone,Name=\"dyn_skirt_R_A_01\"),Transform=(Rotation=(X=-0.0091,Y=-0.0427,Z=-0.2077,W=0.9772),Translation=(X=0,Y=0,Z=0),Scale3D=(X=1,Y=1,Z=1))),(Item=(Type=Bone,Name=\"dyn_skirt_R_B_01\"),Transform=(Rotation=(X=0,Y=0,Z=0,W=1),Translation=(X=0,Y=0,Z ...(truncated)"
    WeightMinimum=0
    Weight=&RigVMFunction_MathDoubleMul(A=@GetAlphaFromQuatDelta_12.alpha, B=@n16.Result).Result
   )
   RigVMFunctionReturnNode(ExecuteContext=@RigVMFunction_Sequence.B)
   &RigVMFunction_MathDoubleRemap(SourceMaximum=1, SourceMinimum=0, TargetMaximum=0, TargetMinimum=1, Value=@n5.Result)
  ]}
 CollapseNode_ContainedGraph {nodes [
   RigVMFunctionEntryNode()
   > RigVMFunction_Sequence = RigVMAggregateNode()
   RigVMFunctionReturnNode(ExecuteContext=@RigVMFunction_Sequence.D)
   n1 = RigVMVariableNode() {variable=Is Crouching}
   n2 = RigVMVariableNode() {variable=Is Walking and targeting}
   n3 = &RigVMFunction_AlphaInterp(
    InterpSpeedDecreasing=5
    InterpSpeedIncreasing=5
    Scale=1
    bClampResult=true
    Value=&RigVMFunction_MathDoubleMul(
     B=RigVMDispatchNode(False=1, True=0, Condition=@n1.Value) {template="DISPATCH_RigVMDispatch_If(in Condition,in True,in False,out Result)"}.Result
     A=&RigVMFunction_MathBoolToFloat(Value=@n2.Value).Result
    ).Result
   )
   &RigUnit_ModifyTransforms(
    ItemToModify=((Item=(Type=Bone,Name="base_LoinCloth_front"),Transform=(Rotation=(X=0,Y=0,Z=0.0599,W=0.9982),Translation=(X=0,Y=-2.6051,Z=0),Scale3D=(X=1,Y=1,Z=1))))
    WeightMinimum=0
    Weight=@n3.Result
    ExecutePin=@RigVMFunction_Sequence.C
   )
   n4 = RigVMVariableNode() {variable=l thighLocal rz float}
   n5 = RigVMVariableNode() {variable=r thighLocal rz float}
   &RigUnit_ModifyTransforms(
    ItemToModify=((Item=(Type=Bone,Name="dyn_LoinCloth_back_01"),Transform=(Rotation=(X=0,Y=0,Z=-0.8192,W=0.5736),Translation=(X=0,Y=0,Z=0),Scale3D=(X=1,Y=1,Z=1))))
    WeightMinimum=0
    Weight=&RigVMFunction_MathDoubleMax(
     A=&RigVMFunction_MathDoubleRemap(SourceMaximum=90, SourceMinimum=0, TargetMaximum=1, TargetMinimum=0, Value=@n4.Value).Result
     B=&RigVMFunction_MathDoubleRemap(SourceMaximum=90, SourceMinimum=0, TargetMaximum=1, TargetMinimum=0, Value=@n5.Value).Result
    ).Result
    ExecutePin=@RigVMFunction_Sequence.A
   )
   n6 = RigVMVariableNode() {variable=l thighLocal rz float}
   n7 = RigVMVariableNode() {variable=r thighLocal rz float}
   > &RigUnit_ModifyTransforms(
    ItemToModify=((Item=(Type=Bone,Name="base_LoinCloth_front"),Transform=(Rotation=(X=0,Y=0,Z=0.342,W=0.9397),Translation=(X=0,Y=0,Z=0),Scale3D=(X=1,Y=1,Z=1))))
    WeightMinimum=0
    Weight=&RigVMFunction_MathDoubleMax(
     A=&RigVMFunction_MathDoubleRemap(SourceMaximum=-45, SourceMinimum=0, TargetMaximum=1, TargetMinimum=0, Value=@n6.Value).Result
     B=&RigVMFunction_MathDoubleRemap(SourceMaximum=-45, SourceMinimum=0, TargetMaximum=1, TargetMinimum=0, Value=@n7.Value).Result
    ).Result
   )
   n8 = RigVMVariableNode() {variable=l thighLocal rz float}
   n9 = RigVMVariableNode() {variable=r thighLocal rz float}
   > &RigUnit_ModifyTransforms(
    ItemToModify=((Item=(Type=Bone,Name="base_LoinCloth_front"),Transform=(Rotation=(X=0,Y=0,Z=0.5,W=0.866),Translation=(X=0,Y=0,Z=0),Scale3D=(X=1,Y=1,Z=1))))
    WeightMinimum=0
    Weight=&RigVMFunction_MathDoubleMax(
     A=&RigVMFunction_MathDoubleRemap(SourceMaximum=-90, SourceMinimum=-45, TargetMaximum=1, TargetMinimum=0, Value=@n8.Value).Result
     B=&RigVMFunction_MathDoubleRemap(SourceMaximum=-90, SourceMinimum=-45, TargetMaximum=1, TargetMinimum=0, Value=@n9.Value).Result
    ).Result
   )
   n10 = RigVMVariableNode() {variable=Is Wearing a cape}
   Maximum = &RigVMFunction_MathDoubleMax(
    A=&RigVMFunction_MathDoubleRemap(SourceMaximum=120, SourceMinimum=20, TargetMaximum=1, TargetMinimum=0, Value=RigVMVariableNode() {variable=l calfLocal rz float}.Value).Result
    B=&RigVMFunction_MathDoubleRemap(SourceMaximum=120, SourceMinimum=20, TargetMaximum=1, TargetMinimum=0, Value=RigVMVariableNode() {variable=r calfLocal rz float}.Value).Result
   )
   n11 = RigVMVariableNode() {variable=Is Crouching}
   Multiply = &RigVMFunction_MathDoubleMul(
    B=&RigVMFunction_MathDoubleMul(
     B=RigVMDispatchNode(False=1, True=0.4, Condition=@n10.Value) {template="DISPATCH_RigVMDispatch_If(in Condition,in True,in False,out Result)"}.Result
     A=@Maximum.Result
    ).Result
    A=&RigVMFunction_MathDoubleRemap(
     SourceMaximum=1
     SourceMinimum=0
     TargetMaximum=0
     TargetMinimum=1
     Value=&RigVMFunction_MathBoolToFloat(Value=@n11.Value).Result
    ).Result
   )
   > &RigUnit_ModifyTransforms(
    ItemToModify="((Item=(Type=Bone,Name=\"dyn_LoinCloth_back_02\"),Transform=(Rotation=(X=0,Y=0,Z=-0.3007,W=0.9537),Translation=(X=0,Y=0,Z=0),Scale3D=(X=1,Y=1,Z=1))),(Item=(Type=Bone,Name=\"dyn_LoinCloth_back_03\"),Transform=(Rotation=(X=0,Y=0,Z=-0.2164,W=0.9763),Translation=( ...(truncated)"
    WeightMinimum=0
    Weight=@Multiply.Result
   )
   n12 = RigVMVariableNode() {variable=Is Crouching}
   n13 = &RigVMFunction_AlphaInterp(
    InterpSpeedDecreasing=5
    InterpSpeedIncreasing=5
    Scale=1
    bClampResult=true
    Value=&RigVMFunction_MathDoubleMul(B=1, A=&RigVMFunction_MathBoolToFloat(Value=@n12.Value).Result).Result
   )
   &RigUnit_ParentConstraint(
    AdvancedSettings=(InterpolationType=Average,RotationOrderForFilter=XZY)
    Child=(Type=Bone,Name="base_LoinCloth_front")
    Filter=(TranslationFilter=(bX=false,bY=false,bZ=false),RotationFilter=(bX=true,bY=true,bZ=true),ScaleFilter=(bX=false,bY=false,bZ=false))
    Parents=((Item=(Type=Bone,Name="dyn_skirt_R_A_01"),Weight=1))
    Weight=@n13.Result
    ExecutePin=@RigVMFunction_Sequence.B
   )
   > AlterTransform = RigVMFunctionReferenceNode(
    inheritsTransform=false
    transforms="((Item=(Type=Bone,Name=\"dyn_LoinCloth_front_01\"),Transform=(Rotation=(X=0,Y=0,Z=0,W=1),Translation=(X=0,Y=-4,Z=0),Scale3D=(X=1,Y=1,Z=1))),(Item=(Type=Bone,Name=\"dyn_LoinCloth_front_02\"),Transform=(Rotation=(X=0,Y=0,Z=0,W=1),Translation=(X=0,Y=-1,Z=0),Scale ...(truncated)"
    weights=@n13.Result
   )
   n14 = RigVMVariableNode() {variable=Is Wearing a cape}
   n15 = RigVMVariableNode() {variable=Is Jumping}
   n16 = &RigVMFunction_AlphaInterp(
    InterpSpeedDecreasing=5
    InterpSpeedIncreasing=5
    Scale=1
    bClampResult=true
    Value=&RigVMFunction_MathDoubleMul(
     B=RigVMDispatchNode(False=1, True=0, Condition=@n14.Value) {template="DISPATCH_RigVMDispatch_If(in Condition,in True,in False,out Result)"}.Result
     A=&RigVMFunction_MathBoolToFloat(Value=@n15.Value).Result
    ).Result
   )
   > &RigUnit_ModifyTransforms(
    ItemToModify=((Item=(Type=Bone,Name="base_LoinCloth_back"),Transform=(Rotation=(X=0,Y=0,Z=-0.2588,W=0.9659),Translation=(X=0,Y=0,Z=0),Scale3D=(X=1,Y=1,Z=1))))
    WeightMinimum=0
    Weight=@n16.Result
   )
   n17 = &RigVMFunction_AlphaInterp(
    InterpSpeedDecreasing=5
    InterpSpeedIncreasing=5
    Scale=1
    bClampResult=true
    Value=&RigVMFunction_MathBoolToFloat(Value=RigVMVariableNode() {variable=isDBNO}.Value).Result
   )
   > &RigUnit_ModifyTransforms(
    ItemToModify="((Item=(Type=Bone,Name=\"dyn_LoinCloth_back_01\"),Transform=(Rotation=(X=0,Y=0,Z=0.0698,W=0.9976),Translation=(X=0,Y=-0.25,Z=0),Scale3D=(X=1,Y=1,Z=1))),(Item=(Type=Bone,Name=\"dyn_LoinCloth_back_02\"),Transform=(Rotation=(X=0,Y=0,Z=0.5,W=0.866),Translation=(X= ...(truncated)"
    WeightMinimum=0
    Weight=@n17.Result
   )
   &RigVMFunction_MathDoubleRemap(
    SourceMaximum=1
    SourceMinimum=0
    TargetMaximum=0
    TargetMinimum=1
    Value=&RigVMFunction_MathBoolToFloat(Value=RigVMVariableNode() {variable=Is 2D moving}.Value).Result
   )
  ]}
 CollapseNode_ContainedGraph {nodes [
   n1 = RigVMFunctionEntryNode()
   RigVMFunction_ControlFlowBranch = &RigVMFunction_ControlFlowBranch(ExecuteContext=@n1.ExecuteContext, Condition=RigVMVariableNode() {variable=Is Hang Gliding}.Value)
   RigVMFunctionReturnNode(ExecuteContext=@RigVMFunction_ControlFlowBranch.Completed)
   GetAlphaFromQuatDelta = RigVMFunctionReferenceNode(
    Rotation Order=ZYX
    child=(Type=Bone,Name="clavicle_l")
    debugRotation=false
    outputMinMax=(X=0,Y=1)
    parent=(Type=Null,Name="clavicle_l_null")
    premult=1
    rotationMinMax=(X=-18,Y=-40)
    triggerAxis=Y
    twistAxis=(X=1,Y=0,Z=0)
    ExecuteContext=@RigVMFunction_ControlFlowBranch.True
   )
   > &RigUnit_ModifyTransforms(
    ItemToModify=((Item=(Type=Bone,Name="clavicle_armorpad_l"),Transform=(Rotation=(X=0,Y=0.0911,Z=0,W=0.9958),Translation=(X=3.9,Y=0,Z=-0.7199),Scale3D=(X=1,Y=1,Z=1))))
    WeightMinimum=0
    Weight=@GetAlphaFromQuatDelta.alpha
   )
   > GetAlphaFromQuatDelta_1 = RigVMFunctionReferenceNode(
    Rotation Order=ZYX
    child=(Type=Bone,Name="clavicle_r")
    debugRotation=false
    outputMinMax=(X=0,Y=1)
    parent=(Type=Null,Name="clavicle_r_null")
    premult=1
    rotationMinMax=(X=17,Y=-5)
    triggerAxis=Y
    twistAxis=(X=1,Y=0,Z=0)
   )
   > &RigUnit_ModifyTransforms(
    ItemToModify=((Item=(Type=Bone,Name="clavicle_armorpad_r"),Transform=(Rotation=(X=0,Y=0.0368,Z=0,W=0.9993),Translation=(X=-2.4542,Y=0,Z=0.1811),Scale3D=(X=1,Y=1,Z=1))))
    WeightMinimum=0
    Weight=@GetAlphaFromQuatDelta_1.alpha
   )
  ]}
 CollapseNode_ContainedGraph {nodes [
   n1 = RigVMFunctionEntryNode()
   n2 = &RigUnit_SphericalPoseReader(
    ActiveRegionScaleFactors=(PositiveWidth=1,NegativeWidth=1,PositiveHeight=1,NegativeHeight=1)
    ActiveRegionSize=0.1
    Debug=(bDrawDebug=false,bDraw2D=False,bDrawLocalAxes=False,DebugScale=25,DebugSegments=20,DebugThickness=0.25)
    DriverAxis=(X=1,Y=0,Z=0)
    DriverItem=(Type=Bone,Name="head")
    FalloffRegionScaleFactors=(PositiveWidth=1,NegativeWidth=1,PositiveHeight=1,NegativeHeight=1)
    FalloffSize=0.2
    FlipHeightScaling=False
    FlipWidthScaling=False
    OptionalParentItem=(Type=Bone,Name="None")
    RotationOffset=(X=90,Y=0,Z=230)
    ExecutePin=@n1.ExecutePin
   )
   > n3 = &RigUnit_SphericalPoseReader(
    ActiveRegionScaleFactors=(PositiveWidth=1,NegativeWidth=1,PositiveHeight=1,NegativeHeight=1)
    ActiveRegionSize=0.1
    Debug=(bDrawDebug=false,bDraw2D=False,bDrawLocalAxes=False,DebugScale=25,DebugSegments=20,DebugThickness=0.25)
    DriverAxis=(X=1,Y=0,Z=0)
    DriverItem=(Type=Bone,Name="neck_01")
    FalloffRegionScaleFactors=(PositiveWidth=1,NegativeWidth=1,PositiveHeight=1,NegativeHeight=1)
    FalloffSize=0.2
    FlipHeightScaling=False
    FlipWidthScaling=False
    OptionalParentItem=(Type=Bone,Name="None")
    RotationOffset=(X=90,Y=0,Z=-80)
   )
   Multiply = &RigVMFunction_MathDoubleMul(A=@n3.OutputParam, B=@n2.OutputParam)
   > n4 = &RigUnit_ModifyTransforms(
    ItemToModify=((Item=(Type=Bone,Name="armor_collar_offset"),Transform=(Rotation=(X=0,Y=-0.2588,Z=0,W=0.9659),Translation=(X=0,Y=0,Z=3),Scale3D=(X=1,Y=1,Z=1))))
    WeightMinimum=0
    Weight=@Multiply.Result
   )
   RigVMFunctionReturnNode(ExecutePin=@n4.ExecutePin)
  ]}
 Multiply_9_ContainedGraph {nodes [
   n1 = RigVMFunctionEntryNode()
   RigVMFunctionReturnNode(Result=&RigVMFunction_MathDoubleMul(B=@n1.C, A=&RigVMFunction_MathDoubleMul(A=@n1.A, B=@n1.B).Result).Result)
  ]}
 Multiply_9_ContainedGraph {nodes [
   n1 = RigVMFunctionEntryNode()
   RigVMFunctionReturnNode(Result=&RigVMFunction_MathDoubleMul(B=@n1.C, A=&RigVMFunction_MathDoubleMul(A=@n1.A, B=@n1.B).Result).Result)
  ]}
 New Function_ContainedGraph {nodes [
   RigVMFunctionEntryNode()
   RigVMFunction_MathBoolToInteger = &RigVMFunction_MathBoolToInteger(Value=RigVMVariableNode() {variable=onlyTwist}.Value)
   RigVMFunction_MathQuaternionSwingTwist = &RigVMFunction_MathQuaternionSwingTwist(
    Input=&RigUnit_GetRelativeTransformForItem(
     bChildInitial=False
     bParentInitial=False
     Child=RigVMVariableNode() {variable=child}.Value
     Parent=RigVMVariableNode() {variable=parent}.Value
    ).RelativeTransform.Rotation
    TwistAxis=RigVMVariableNode() {variable=twistAxis}.Value
   )
   n1 = RigVMDispatchNode(
    Values=((),())
    Index=@RigVMFunction_MathBoolToInteger.Result
    Values.0=@RigVMFunction_MathQuaternionSwingTwist.Swing
    Values.1=@RigVMFunction_MathQuaternionSwingTwist.Twist
   ) {template="DISPATCH_RigVMDispatch_SelectInt32(in Index,in Values,out Result)"}
   > RigVMVariableNode(Value=@n1.Result) {variable=entry_outputQuat}
   RigVMFunction_MathQuaternionToEuler = &RigVMFunction_MathQuaternionToEuler(Value=@n1.Result, RotationOrder=RigVMVariableNode() {variable=Rotation Order}.Value)
   > DISPATCH_RigVMDispatch_Print = RigVMDispatchNode(
    ScreenColor=(R=1,G=1,B=1,A=1)
    ScreenDuration=1.5
    Value=@RigVMFunction_MathQuaternionToEuler.Result
    Enabled=RigVMVariableNode() {variable=debugRotation}.Value
   ) {template="DISPATCH_RigVMDispatch_Print(in Prefix,in Value,in Enabled,in ScreenDuration,in ScreenColor)"}
   Cast = RigVMDispatchNode(Value=RigVMVariableNode() {variable=triggerAxis}.Value) {template="DISPATCH_RigVMDispatch_CastEnumToInt(in Value,out Result)"}
   DISPATCH_RigVMDispatch_SelectInt32 = RigVMDispatchNode(
    Values=(0,0,0)
    Index=@Cast.Result
    Values.0=@RigVMFunction_MathQuaternionToEuler.Result.X
    Values.1=@RigVMFunction_MathQuaternionToEuler.Result.Y
    Values.2=@RigVMFunction_MathQuaternionToEuler.Result.Z
   ) {template="DISPATCH_RigVMDispatch_SelectInt32(in Index,in Values,out Result)"}
   n2 = RigVMVariableNode() {variable=rotationMinMax}
   n3 = RigVMVariableNode() {variable=outputMinMax}
   Remap = &RigVMFunction_MathDoubleRemap(
    bClamp=False
    Value=@DISPATCH_RigVMDispatch_SelectInt32.Result
    SourceMinimum=@n2.Value.X
    SourceMaximum=@n2.Value.Y
    TargetMinimum=@n3.Value.X
    TargetMaximum=@n3.Value.Y
   )
   Absolute = &RigVMFunction_MathDoubleAbs(Value=@Remap.Result)
   n4 = RigVMVariableNode() {variable=absoluteValue}
   Multiply = &RigVMFunction_MathDoubleMul(
    A=&RigVMFunction_MathDoubleClamp(Value=RigVMDispatchNode(True=@Absolute.Result, False=@Remap.Result, Condition=@n4.Value) {template="DISPATCH_RigVMDispatch_If(in Condition,in True,in False,out Result)"}.Result).Result
    B=RigVMVariableNode() {variable=premult}.Value
   )
   > RigVMFunctionReturnNode(alpha=@Multiply.Result, rawQuat=RigVMVariableNode() {variable=entry_outputQuat}.Value)
  ]}
 New Function_ContainedGraph {nodes [
   n1 = RigVMFunctionEntryNode()
   > For_Each = RigVMDispatchNode(Array=@n1.transforms) {template="DISPATCH_RigVMDispatch_ArrayIterator(in Array,out Element,out Index,out Count,out Ratio)"}
   > &RigUnit_OffsetTransformForItem(Item=@For_Each.Element.Item, OffsetTransform=@For_Each.Element.Transform, Weight=@n1.weights, bPropagateToChildren=@n1.inheritsTransform)
   RigVMFunctionReturnNode(ExecuteContext=@For_Each.Completed)
  ]}
 RigVMFunctionLibrary {nodes [
   AlterTransform = RigVMCollapseNode(ExecuteContext=(), inheritsTransform=false, transforms=(), weights=1)
   GetAlphaFromQuatDelta = RigVMCollapseNode(
    ExecuteContext=()
    Rotation Order=ZYX
    child=(Type=Bone,Name="None")
    outputMinMax=(X=0,Y=1)
    parent=(Type=Bone,Name="None")
    premult=1
    rotationMinMax=(X=0,Y=0)
    twistAxis=(X=1,Y=0,Z=0)
   )
  ]}
 RigVMFunction_Sequence_1_ContainedGraph {nodes [
   n1 = RigVMFunctionEntryNode()
   RigVMFunction_Sequence = &RigVMFunction_Sequence(ExecuteContext=@n1.ExecuteContext)
   n2 = &RigVMFunction_Sequence(ExecuteContext=@RigVMFunction_Sequence.B)
   n3 = &RigVMFunction_Sequence(ExecuteContext=@n2.B)
   RigVMFunctionReturnNode(A=@RigVMFunction_Sequence.A, B=@n2.A, C=@n3.A, D=@n3.B)
  ]}
 RigVMFunction_Sequence_2_ContainedGraph {nodes [
   n1 = RigVMFunctionEntryNode()
   n2 = &RigVMFunction_Sequence(ExecuteContext=@n1.ExecuteContext)
   n3 = &RigVMFunction_Sequence(ExecuteContext=@n2.B)
   RigVMFunctionReturnNode(A=@n2.A, B=@n3.A, C=@n3.B)
  ]}
 RigVMModel {nodes [
   RigUnit_BeginExecution = &RigUnit_BeginExecution()
   Constraints = RigVMCollapseNode(ExecutePin=@RigUnit_BeginExecution.ExecutePin)
   /Script/FortniteGame.RigUnit_FortDeformRig(
    Bool=(bUpperArm=true,bElbow=true,bWrist=true,bGroin=true,bGlute=true,bKnee=true,bPec=true,bBicep=true,bLat=true)
    Curve="(bLUpperArmUpAlpha=False,bRUpperArmUpAlpha=False,bLUpperArmFrAlpha=False,bRUpperArmFrAlpha=False,bLUpperArmBkAlpha=False,bRUpperArmBkAlpha=False,bLWristUpAlpha=False,bRWristUpAlpha=False,bLWristDnAlpha=False,bRWristDnAlpha=False,bLKneeAlpha=False,bRKneeAlp ...(truncated)"
    Debug=(bUpperArmDebug=False,bWristDebug=False,bLatDebug=False,bGroinGluteDebug=False)
    Value="(UpperArmUp=(Rotation=(X=0.0,Y=-0.0,Z=0.0,W=1.0),Translation=(X=0.0,Y=0.0,Z=8.0),Scale3D=(X=0.70,Y=1.0,Z=1.0)),UpperArmFwdFr=(Rotation=(X=0.0,Y=-0.0,Z=0.0,W=1.0),Translation=(X=0.0,Y=10.0,Z=0.0),Scale3D=(X=0.50,Y=1.0,Z=1.0)),UpperArmFwdBk=(Rotation=(X=0.0, ...(truncated)"
    ExecutePin=@Constraints.ExecutePin
   ) {is=struct RigVMUnitNode}
   > Variables_st = RigVMCollapseNode()
   > Skirt_St_03 = RigVMCollapseNode()
   > Loin_St = RigVMCollapseNode()
   > Clavicles_pads = RigVMCollapseNode()
   Armor_collar_fix = RigVMCollapseNode(ExecutePin=@Clavicles_pads.ExecuteContext)
  ]}
]
hierarchy [
 root {loc=<0 0 0>, rot=<0 0 0 1>}
 pelvis {parent=root, loc=<0 -0.7071 93.1191>, rot=<0.0127 0.707 -0.0127 -0.707>}
 spine_01 {parent=pelvis, loc=<6.7494 -0.6388 0>, rot=<0 0 0.034 -0.9994>}
 spine_02 {parent=spine_01, loc=<7.2518 -0.035 0>, rot=<0 0 0.0238 -0.9997>}
 spine_03 {parent=spine_02, loc=<6.0104 0 0>, rot=<0 0 0.0149 -0.9999>}
 spine_04 {parent=spine_03, loc=<8.1024 0 0>, rot=<0 0 -0.0132 -0.9999>}
 spine_05 {parent=spine_04, loc=<10.0146 0 0>, rot=<0 0 -0.0015 -1>}
 clavicle_l_null {parent=spine_05, type=Null, loc=<4.3575 -0.1466 -2.3223>, rot=<0.0465 -0.7289 -0.009 -0.6829>}
 neck_01 {parent=spine_05, loc=<11.3138 0.1529 0>, rot=<0 0 -0.2003 -0.9797>}
 neck_02 {parent=neck_01, loc=<4.5627 0 0>, rot=<0 0 0.0504 -0.9987>}
 head {parent=neck_02, loc=<5.3914 0 0>, rot=<0 0 0.115 -0.9934>}
 head_null {parent=neck_02, type=Null, loc=<5.3914 0 0>, rot=<0 0 0.115 -0.9934>}
 armor_collar_offset {parent=neck_01, loc=<3.5478 -5.3801 -0.0802>, rot=<-0.7082 0.024 0.0714 -0.702>}
 neck_02_null {parent=neck_01, type=Null, loc=<4.5627 0 0>, rot=<0 0 0.0504 -0.9987>}
 attach_cape {parent=spine_05, loc=<9.017 -11.714 0>, rot=<0.0616 -0.7044 -0.0616 -0.7044>}
 cape {parent=attach_cape, type=Null, loc=<0 0 0>, rot=<0 0 0 1>}
 attach_backpack {parent=spine_05, loc=<-0.123 -13.473 0>, rot=<-0.0093 -0.707 0.0093 -0.707>}
 Backpack_BR {parent=attach_backpack, type=Null, loc=<0 0 0>, rot=<0 0 0 1>}
 clavicle_l {parent=spine_05, loc=<4.3575 -0.1466 -2.3223>, rot=<0.0465 -0.7289 -0.009 -0.6829>}
 upperarm_l {parent=clavicle_l, loc=<15.5111 0 0>, rot=<0.0226 -0.3234 -0.0088 -0.9459>}
 lowerarm_l {parent=upperarm_l, loc=<26.6514 0 0>, rot=<0.0391 -0.0099 -0.2453 -0.9686>}
 hand_l {parent=lowerarm_l, loc=<22.3491 0 0>, rot=<-0.7081 -0.0507 0.0975 -0.6975>}
 thumb_01_l {parent=hand_l, loc=<1.3118 -0.3191 -2.459>, rot=<0.6546 -0.4042 0.0664 -0.6354>}
 thumb_02_l {parent=thumb_01_l, loc=<5.9336 -0.0448 0.0401>, rot=<-0.0037 0.0213 0.298 -0.9543>}
 thumb_03_l {parent=thumb_02_l, loc=<2.2223 0.3682 -0.0907>, rot=<0.0118 -0.013 0.0066 -0.9998>}
 index_metacarpal_l {parent=hand_l, loc=<1.6535 0.0183 -2.96>, rot=<0.0158 -0.0124 -0.0173 -0.9996>}
 index_01_l {parent=index_metacarpal_l, loc=<8.3008 0.0357 -0.0524>, rot=<-0.0044 0.0485 0.0924 -0.9945>}
 index_02_l {parent=index_01_l, loc=<3.3587 -0.3004 0.0452>, rot=<-0.002 0.0027 0.3225 -0.9466>}
 index_03_l {parent=index_02_l, loc=<2.5578 -0.123 -0.0252>, rot=<-0.002 0.0067 0.2558 -0.9667>}
 middle_metacarpal_l {parent=hand_l, loc=<2.1856 -0.1725 -0.5984>, rot=<0.006 0.0168 -0.0329 -0.9993>}
 middle_01_l {parent=middle_metacarpal_l, loc=<7.3275 -0.2447 0.0129>, rot=<-0.0208 0.0042 0.0858 -0.9961>}
 middle_02_l {parent=middle_01_l, loc=<4.0123 0.1401 0.1175>, rot=<-0.0218 0.0258 0.3512 -0.9357>}
 middle_03_l {parent=middle_02_l, loc=<2.6077 0.0534 0.1373>, rot=<0.0082 0.0188 0.1708 -0.9851>}
 ring_metacarpal_l {parent=hand_l, loc=<2.2074 -0.3625 1.3638>, rot=<-0.0284 0.0722 -0.0458 -0.9959>}
 ring_01_l {parent=ring_metacarpal_l, loc=<7.0295 -0.6197 -0.1446>, rot=<-0.0489 -0.0252 0.1892 -0.9804>}
 ring_02_l {parent=ring_01_l, loc=<3.6174 0.4647 -0.1027>, rot=<0 0 0.2806 -0.9598>}
 ring_03_l {parent=ring_02_l, loc=<2.4152 0.114 0.0199>, rot=<0 0 0.1189 -0.9929>}
 pinky_metacarpal_l {parent=hand_l, loc=<1.7616 -0.9753 3.3814>, rot=<-0.0397 0.0814 -0.069 -0.9935>}
 pinky_01_l {parent=pinky_metacarpal_l, loc=<7.1709 -0.4831 -0.1094>, rot=<-0.0654 -0.0035 0.2242 -0.9723>}
 pinky_02_l {parent=pinky_01_l, loc=<2.8274 0.0893 -0.1043>, rot=<0.0037 0.0011 0.2854 -0.9584>}
 pinky_03_l {parent=pinky_02_l, loc=<2.2046 0.0307 -0.0078>, rot=<-0.0002 0.0098 0.1445 -0.9895>}
 deform_wrist_dn_base_l {parent=hand_l, loc=<0 -2.2342 0>, rot=<0 0 0 -1>}
 deform_wrist_dn_l {parent=deform_wrist_dn_base_l, loc=<0 0 0>, rot=<0 0 0 1>}
 deform_wrist_up_base_l {parent=hand_l, loc=<0 2.2682 0>, rot=<0 0 0 -1>}
 deform_wrist_up_l {parent=deform_wrist_up_base_l, loc=<0 0 0>, rot=<0 0 0 1>}
 lowerarm_twist_02_l {parent=lowerarm_l, loc=<9.6199 0 0>, rot=<0 0 0 -1>}
 lowerarm_twist_01_l {parent=lowerarm_l, loc=<17.8549 0 0>, rot=<0 0 0 -1>}
 deform_elbow_in_base_l {parent=lowerarm_l, loc=<0.2112 5.259 0.8274>, rot=<0 0 0 -1>}
 deform_elbow_in_l {parent=deform_elbow_in_base_l, loc=<0 0 0>, rot=<0 0 0 1>}
 deform_elbow_out_base_l {parent=lowerarm_l, loc=<0 -4 0>, rot=<0 0 0 -1>}
 deform_elbow_out_l {parent=deform_elbow_out_base_l, loc=<0 0 0>, rot=<0 0 0 1>}
 upperarm_twist_01_l {parent=upperarm_l, loc=<0.0041 0 0>, rot=<0 0 0 -1>}
 deform_upperarm_bk_base_l {parent=upperarm_twist_01_l, loc=<1.0813 -7.2193 -1.9453>, rot=<0 0 0 1>}
 deform_upperarm_bk_l {parent=deform_upperarm_bk_base_l, loc=<0 0 0>, rot=<0 0 0 1>}
 deform_upperarm_fr_base_l {parent=upperarm_twist_01_l, loc=<-0.3691 7.0506 -2.765>, rot=<0 0 0 1>}
 deform_upperarm_fr_l {parent=deform_upperarm_fr_base_l, loc=<0 0 0>, rot=<0 0 0 1>}
 deform_upperarm_up_base_l {parent=upperarm_twist_01_l, loc=<-3.2216 -1.1391 3.3311>, rot=<0 0 0 1>}
 deform_upperarm_up_l {parent=deform_upperarm_up_base_l, loc=<0 0 0>, rot=<0 0 0 1>}
 upperarm_twist_02_l {parent=upperarm_l, loc=<13.7941 0 0>, rot=<0 0 0 -1>}
 clavicle_armorpad_l {parent=clavicle_l, loc=<2.7462 0 -0.8207>, rot=<0 0 0 1>}
 clavicle_r {parent=spine_05, loc=<4.3575 -0.1466 2.3223>, rot=<0.7289 0.0465 -0.6829 0.009>}
 upperarm_r {parent=clavicle_r, loc=<-15.5111 0 0>, rot=<0.0226 -0.3234 -0.0088 -0.9459>}
 lowerarm_r {parent=upperarm_r, loc=<-26.6514 0 0>, rot=<0.0391 -0.0099 -0.2453 -0.9686>}
 hand_r {parent=lowerarm_r, loc=<-22.3491 0 0>, rot=<-0.7081 -0.0507 0.0975 -0.6975>}
 thumb_01_r {parent=hand_r, loc=<-1.3118 0.3191 2.459>, rot=<0.6546 -0.4042 0.0664 -0.6354>}
 thumb_02_r {parent=thumb_01_r, loc=<-5.9336 0.0448 -0.0401>, rot=<-0.0037 0.0213 0.298 -0.9543>}
 thumb_03_r {parent=thumb_02_r, loc=<-2.2223 -0.3682 0.0907>, rot=<0.0118 -0.013 0.0066 -0.9998>}
 index_metacarpal_r {parent=hand_r, loc=<-1.6535 -0.0183 2.96>, rot=<0.0158 -0.0124 -0.0173 -0.9996>}
 index_01_r {parent=index_metacarpal_r, loc=<-8.3008 -0.0357 0.0524>, rot=<-0.0044 0.0485 0.0924 -0.9945>}
 index_02_r {parent=index_01_r, loc=<-3.3587 0.3004 -0.0452>, rot=<-0.002 0.0027 0.3225 -0.9466>}
 index_03_r {parent=index_02_r, loc=<-2.5578 0.123 0.0252>, rot=<-0.002 0.0067 0.2558 -0.9667>}
 middle_metacarpal_r {parent=hand_r, loc=<-2.1856 0.1725 0.5984>, rot=<0.006 0.0168 -0.0329 -0.9993>}
 middle_01_r {parent=middle_metacarpal_r, loc=<-7.3275 0.2447 -0.0129>, rot=<-0.0208 0.0042 0.0858 -0.9961>}
 middle_02_r {parent=middle_01_r, loc=<-4.0123 -0.1401 -0.1175>, rot=<-0.0218 0.0258 0.3512 -0.9357>}
 middle_03_r {parent=middle_02_r, loc=<-2.6077 -0.0534 -0.1373>, rot=<0.0082 0.0188 0.1708 -0.9851>}
 ring_metacarpal_r {parent=hand_r, loc=<-2.2074 0.3625 -1.3638>, rot=<-0.0284 0.0722 -0.0458 -0.9959>}
 ring_01_r {parent=ring_metacarpal_r, loc=<-7.0295 0.6197 0.1446>, rot=<-0.0489 -0.0252 0.1892 -0.9804>}
 ring_02_r {parent=ring_01_r, loc=<-3.6174 -0.4647 0.1027>, rot=<0 0 0.2806 -0.9598>}
 ring_03_r {parent=ring_02_r, loc=<-2.4152 -0.114 -0.0199>, rot=<0 0 0.1189 -0.9929>}
 pinky_metacarpal_r {parent=hand_r, loc=<-1.7616 0.9753 -3.3814>, rot=<-0.0397 0.0814 -0.069 -0.9935>}
 pinky_01_r {parent=pinky_metacarpal_r, loc=<-7.1709 0.4831 0.1094>, rot=<-0.0654 -0.0035 0.2242 -0.9723>}
 pinky_02_r {parent=pinky_01_r, loc=<-2.8274 -0.0893 0.1043>, rot=<0.0037 0.0011 0.2854 -0.9584>}
 pinky_03_r {parent=pinky_02_r, loc=<-2.2046 -0.0307 0.0078>, rot=<-0.0002 0.0098 0.1445 -0.9895>}
 deform_wrist_dn_base_r {parent=hand_r, loc=<-0.0004 2.2345 -0.0001>, rot=<0 0 0 -1>}
 deform_wrist_dn_r {parent=deform_wrist_dn_base_r, loc=<0 0 0>, rot=<0 0 0 1>}
 deform_wrist_up_base_r {parent=hand_r, loc=<0.0003 -2.2684 0.0001>, rot=<0 0 0 -1>}
 deform_wrist_up_r {parent=deform_wrist_up_base_r, loc=<0 0 0>, rot=<0 0 0 1>}
 lowerarm_twist_02_r {parent=lowerarm_r, loc=<-9.6199 0 0>, rot=<0 0 0 -1>}
 lowerarm_twist_01_r {parent=lowerarm_r, loc=<-17.8549 0 0>, rot=<0 0 0 -1>}
 deform_elbow_in_base_r {parent=lowerarm_r, loc=<-0.2111 -5.2591 -0.8275>, rot=<0 0 0 -1>}
 deform_elbow_in_r {parent=deform_elbow_in_base_r, loc=<-0.0001 0 0.0002>, rot=<0 0 0 1>}
 deform_elbow_out_base_r {parent=lowerarm_r, loc=<0 4 -0.0001>, rot=<0 0 0 -1>}
 deform_elbow_out_r {parent=deform_elbow_out_base_r, loc=<0 0 0>, rot=<0 0 0 1>}
 upperarm_twist_01_r {parent=upperarm_r, loc=<-0.0041 0 0>, rot=<0 0 0 -1>}
 deform_upperarm_bk_base_r {parent=upperarm_twist_01_r, loc=<-1.0814 7.2192 1.9455>, rot=<0 0 0 1>}
 deform_upperarm_bk_r {parent=deform_upperarm_bk_base_r, loc=<0 0 0>, rot=<0 0 0 1>}
 deform_upperarm_fr_base_r {parent=upperarm_twist_01_r, loc=<0.3689 -7.0506 2.7652>, rot=<0 0 0 1>}
 deform_upperarm_fr_r {parent=deform_upperarm_fr_base_r, loc=<0 0 0>, rot=<0 0 0 1>}
 deform_upperarm_up_base_r {parent=upperarm_twist_01_r, loc=<3.2214 1.1391 -3.3308>, rot=<0 0 0 1>}
 deform_upperarm_up_r {parent=deform_upperarm_up_base_r, loc=<0 0 0>, rot=<0 0 0 -1>}
 upperarm_twist_02_r {parent=upperarm_r, loc=<-13.7941 0 0>, rot=<0 0 0 -1>}
 clavicle_armorpad_r {parent=clavicle_r, loc=<0 0 0>, rot=<0 0 0 1>}
 clavicle_r_null {parent=spine_05, type=Null, loc=<4.3575 -0.1466 2.3223>, rot=<0.7289 0.0465 -0.6829 0.009>}
 neck_01_null {parent=spine_05, type=Null, loc=<-1.3436 14.6312 11.481>, rot=<0.0889 -0.043 0.5086 0.8553>}
 deform_pec_aim_base_l {parent=spine_04, loc=<7.8732 12.6388 -2.4123>, rot=<0.0102 -0.4602 0.3979 -0.7936>}
 deform_pec_aim_l {parent=deform_pec_aim_base_l, loc=<0 0 0>, rot=<0 0 0 1>}
 deform_pec_base_l {parent=deform_pec_aim_l, loc=<10.766 0 0>, rot=<0 0 0 1>}
 deform_pec_l {parent=deform_pec_base_l, loc=<0.055 0 0>, rot=<0 0 0 1>}
 deform_pec_aim_base_r {parent=spine_04, loc=<7.8727 12.6388 2.4123>, rot=<-0.4602 -0.0102 0.7936 0.3979>}
 deform_pec_aim_r {parent=deform_pec_aim_base_r, loc=<0 0 0>, rot=<0 0 0 1>}
 deform_pec_base_r {parent=deform_pec_aim_r, loc=<-10.7662 -0.0006 -0.0007>, rot=<0 0 0 1>}
 deform_pec_r {parent=deform_pec_base_r, loc=<0 0 0>, rot=<0 0 0 1>}
 thigh_l {parent=pelvis, loc=<-2.5693 -0.0109 -7.7501>, rot=<0.0012 0.0572 0.0155 -0.9982>}
 calf_l {parent=thigh_l, loc=<-39.0484 0 0>, rot=<0.0006 -0.0161 -0.0391 -0.9991>}
 foot_l {parent=calf_l, loc=<-42.561 0.1077 0.4292>, rot=<0.0418 -0.0455 0.0343 -0.9975>}
 ball_l {parent=foot_l, loc=<-6.1599 13.8264 0.6806>, rot=<0.0001 -0.0001 -0.7063 -0.7079>}
 calf_twist_02_l {parent=calf_l, loc=<-18.8064 -0.0967 0.0397>, rot=<0.0237 0.0095 -0.0035 -0.9997>}
 calf_twist_01_l {parent=calf_l, loc=<-35.813 -0.0967 0.0397>, rot=<0.0237 0.0095 -0.0035 -0.9997>}
 deform_knee_in_base_l {parent=calf_l, loc=<1.298 -5.2957 -0.0496>, rot=<0 0 0 -1>}
 deform_knee_in_l {parent=deform_knee_in_base_l, loc=<0 0 0>, rot=<0 0 0 1>}
 deform_knee_out_base_l {parent=calf_l, loc=<0 4 0>, rot=<0 0 0 -1>}
 deform_knee_out_l {parent=deform_knee_out_base_l, loc=<0 0 0>, rot=<0 0 0 1>}
 thigh_twist_01_l {parent=thigh_l, loc=<-21.6303 0 0>, rot=<0.0013 0 0 -1>}
 thigh_twist_skirt_l {parent=thigh_twist_01_l, loc=<0 0 0>, rot=<0 0 0 1>}
 deform_groin_base_l {parent=thigh_l, loc=<-2.1908 8.1142 -1.7442>, rot=<0 0 0 -1>}
 deform_groin_l {parent=deform_groin_base_l, loc=<0 0 0>, rot=<0 0 0 1>}
 calf_l_null {parent=thigh_l, type=Null, loc=<-39.0484 0 0>, rot=<0.0006 -0.0161 -0.0391 -0.9991>}
 thigh_r {parent=pelvis, loc=<-2.5693 -0.0109 7.7501>, rot=<0.0572 -0.0012 0.9982 0.0155>}
 calf_r {parent=thigh_r, loc=<39.0484 0 0>, rot=<0.0006 -0.0161 -0.0391 -0.9991>}
 foot_r {parent=calf_r, loc=<42.561 -0.1077 -0.4292>, rot=<0.0418 -0.0455 0.0343 -0.9975>}
 ball_r {parent=foot_r, loc=<6.1599 -13.8264 -0.6806>, rot=<0.0001 -0.0001 -0.7063 -0.7079>}
 calf_twist_02_r {parent=calf_r, loc=<18.8064 0.0967 -0.0397>, rot=<0.0237 0.0095 -0.0035 -0.9997>}
 calf_twist_01_r {parent=calf_r, loc=<35.813 0.0967 -0.0397>, rot=<0.0237 0.0095 -0.0035 -0.9997>}
 deform_knee_in_base_r {parent=calf_r, loc=<-1.298 5.2957 0.0496>, rot=<0 0 0 -1>}
 deform_knee_in_r {parent=deform_knee_in_base_r, loc=<0 0 0>, rot=<0 0 0 1>}
 deform_knee_out_base_r {parent=calf_r, loc=<0 -4 0>, rot=<0 0 0 -1>}
 deform_knee_out_r {parent=deform_knee_out_base_r, loc=<0 0 0>, rot=<0 0 0 1>}
 thigh_twist_01_r {parent=thigh_r, loc=<21.6303 0 0>, rot=<0.0013 0 0 -1>}
 thigh_twist_skirt_r {parent=thigh_twist_01_r, loc=<0 0 0>, rot=<0 0 0 1>}
 deform_groin_base_r {parent=thigh_r, loc=<2.1908 -8.1142 1.7442>, rot=<0 0 0 -1>}
 calf_r_null {parent=thigh_r, type=Null, loc=<39.0484 0 0>, rot=<0.0006 -0.0161 -0.0391 -0.9991>}
 base_skirt_front {parent=pelvis, loc=<0.6825 9.3734 -1.6148>, rot=<0.0347 0.7063 0.7063 0.0347>}
 dyn_skirt_front_01 {parent=base_skirt_front, loc=<0 0 0>, rot=<0 0 0 1>}
 dyn_skirt_front_02 {parent=dyn_skirt_front_01, loc=<5.1915 0 0>, rot=<0 0 0 1>}
 base_skirt_front_03 {parent=dyn_skirt_front_02, loc=<11.9293 0 0>, rot=<0 0 0 1>}
 dyn_skirt_front_03 {parent=base_skirt_front_03, loc=<0 0 0>, rot=<0 0 0 1>}
 dyn_skirt_front_04 {parent=dyn_skirt_front_03, loc=<11.5979 0 0>, rot=<0 0 0 1>}
 dyn_skirt_front_05 {parent=dyn_skirt_front_04, loc=<11.7084 0 0>, rot=<0 0 0 1>}
 base_LoinCloth_front {parent=base_skirt_front, loc=<-4.2709 2.1801 0.4533>, rot=<-0.7067 0.0241 -0.0241 -0.7067>}
 dyn_LoinCloth_front_01 {parent=base_LoinCloth_front, loc=<8.5715 0 0>, rot=<0 0 0.0053 -1>}
 dyn_LoinCloth_front_02 {parent=dyn_LoinCloth_front_01, loc=<7.3259 0 0>, rot=<0 0 0 1>}
 dyn_LoinCloth_front_03 {parent=dyn_LoinCloth_front_02, loc=<11.5562 0 0>, rot=<0 0 0 1>}
 dyn_LoinCloth_front_04 {parent=dyn_LoinCloth_front_03, loc=<7.6924 0 0>, rot=<0 0 0 1>}
 base_skirt_L_A {parent=pelvis, loc=<0.8013 6.4402 -10.9832>, rot=<-0.0109 -0.8096 -0.5833 -0.0644>}
 dyn_skirt_L_A_01 {parent=base_skirt_L_A, loc=<0 0 0>, rot=<0 0 0 1>}
 dyn_skirt_L_A_02 {parent=dyn_skirt_L_A_01, loc=<5.1915 0 0>, rot=<0 0 0 1>}
 base_skirt_L_A_03 {parent=dyn_skirt_L_A_02, loc=<11.9293 0 0>, rot=<0 0 0 1>}
 dyn_skirt_L_A_03 {parent=base_skirt_L_A_03, loc=<0 0 0>, rot=<0 0 0 1>}
 dyn_skirt_L_A_04 {parent=dyn_skirt_L_A_03, loc=<11.5979 0 0>, rot=<0 0 0 1>}
 dyn_skirt_L_A_05 {parent=dyn_skirt_L_A_04, loc=<11.7084 0 0>, rot=<0 0 0 1>}
 base_skirt_L_B {parent=pelvis, loc=<0.2724 -2.0654 -15.0726>, rot=<-0.0178 -0.9953 0.0017 -0.0957>}
 dyn_skirt_L_B_01 {parent=base_skirt_L_B, loc=<0 0 0>, rot=<0 0 0 1>}
 dyn_skirt_L_B_02 {parent=dyn_skirt_L_B_01, loc=<5.1915 0 0>, rot=<0 0 0 1>}
 base_skirt_L_B_03 {parent=dyn_skirt_L_B_02, loc=<11.9293 0 0>, rot=<0 0 0 1>}
 dyn_skirt_L_B_03 {parent=base_skirt_L_B_03, loc=<0 0 0>, rot=<0 0 0 1>}
 dyn_skirt_L_B_04 {parent=dyn_skirt_L_B_03, loc=<11.5979 0 0>, rot=<0 0 0 1>}
 dyn_skirt_L_B_05 {parent=dyn_skirt_L_B_04, loc=<11.7084 0 0>, rot=<0 0 0 1>}
 base_skirt_L_C {parent=pelvis, loc=<0.2597 -9.5405 -10.9005>, rot=<-0.0142 -0.8297 0.5561 -0.0473>}
 dyn_skirt_L_C_01 {parent=base_skirt_L_C, loc=<0 0 0>, rot=<0 0 0 1>}
 dyn_skirt_L_C_02 {parent=dyn_skirt_L_C_01, loc=<5.1915 0 0>, rot=<0 0 0 1>}
 base_skirt_L_C_03 {parent=dyn_skirt_L_C_02, loc=<11.9293 0 0>, rot=<0 0 0 1>}
 dyn_skirt_L_C_03 {parent=base_skirt_L_C_03, loc=<0 0 0>, rot=<0 0 0 1>}
 dyn_skirt_L_C_04 {parent=dyn_skirt_L_C_03, loc=<11.5979 0 0>, rot=<0 0 0 1>}
 dyn_skirt_L_C_05 {parent=dyn_skirt_L_C_04, loc=<11.7084 0 0>, rot=<0 0 0 1>}
 base_skirt_L_D {parent=pelvis, loc=<-0.0417 -10.8297 -3.883>, rot=<0.0103 -0.7639 0.6448 -0.0241>}
 dyn_skirt_L_D_01 {parent=base_skirt_L_D, loc=<0 0 0>, rot=<0 0 0 1>}
 dyn_skirt_L_D_02 {parent=dyn_skirt_L_D_01, loc=<5.1915 0 0>, rot=<0 0 0 1>}
 base_skirt_L_D_03 {parent=dyn_skirt_L_D_02, loc=<11.9293 0 0>, rot=<0 0 0 1>}
 dyn_skirt_L_D_03 {parent=base_skirt_L_D_03, loc=<0 0 0>, rot=<0 0 0 1>}
 dyn_skirt_L_D_04 {parent=dyn_skirt_L_D_03, loc=<11.5979 0 0>, rot=<0 0 0 1>}
 dyn_skirt_L_D_05 {parent=dyn_skirt_L_D_04, loc=<11.7084 0 0>, rot=<0 0 0 1>}
 base_skirt_R_A {parent=pelvis, loc=<0.8013 6.4402 10.9832>, rot=<-0.8096 0.0109 0.0644 -0.5833>}
 dyn_skirt_R_A_01 {parent=base_skirt_R_A, loc=<0 0 0>, rot=<0 0 0 -1>}
 dyn_skirt_R_A_02 {parent=dyn_skirt_R_A_01, loc=<-5.1914 0 0>, rot=<0 0 0 -1>}
 base_skirt_R_A_03 {parent=dyn_skirt_R_A_02, loc=<-11.9293 0 0>, rot=<0 0 0 1>}
 dyn_skirt_R_A_03 {parent=base_skirt_R_A_03, loc=<0 0 0>, rot=<0 0 0 1>}
 dyn_skirt_R_A_04 {parent=dyn_skirt_R_A_03, loc=<-11.598 0 0>, rot=<0 0 0 -1>}
 dyn_skirt_R_A_05 {parent=dyn_skirt_R_A_04, loc=<-11.7083 0 0>, rot=<0 0 0 -1>}
 base_skirt_R_B {parent=pelvis, loc=<0.2725 -2.0654 15.0726>, rot=<0.9953 -0.0178 -0.0957 -0.0017>}
 dyn_skirt_R_B_01 {parent=base_skirt_R_B, loc=<0 0 0>, rot=<0 0 0 1>}
 dyn_skirt_R_B_02 {parent=dyn_skirt_R_B_01, loc=<-5.1915 0 0.0001>, rot=<0 0 0 1>}
 base_skirt_R_B_03 {parent=dyn_skirt_R_B_02, loc=<-11.9293 0 -0.0001>, rot=<0 0 0 1>}
 dyn_skirt_R_B_03 {parent=base_skirt_R_B_03, loc=<0 0 0>, rot=<0 0 0 1>}
 dyn_skirt_R_B_04 {parent=dyn_skirt_R_B_03, loc=<-11.5979 0 0.0001>, rot=<0 0 0 1>}
 dyn_skirt_R_B_05 {parent=dyn_skirt_R_B_04, loc=<-11.7084 0 0>, rot=<0 0 0 1>}
 base_skirt_R_C {parent=pelvis, loc=<0.2597 -9.5405 10.9005>, rot=<0.8297 -0.0142 -0.0473 -0.5561>}
 dyn_skirt_R_C_01 {parent=base_skirt_R_C, loc=<0 0 0>, rot=<0 0 0 1>}
 dyn_skirt_R_C_02 {parent=dyn_skirt_R_C_01, loc=<-5.1914 -0.0001 0>, rot=<0 0 0 1>}
 base_skirt_R_C_03 {parent=dyn_skirt_R_C_02, loc=<-11.9294 0 0>, rot=<0 0 0 1>}
 dyn_skirt_R_C_03 {parent=base_skirt_R_C_03, loc=<0 0 0>, rot=<0 0 0 1>}
 dyn_skirt_R_C_04 {parent=dyn_skirt_R_C_03, loc=<-11.5979 0 0>, rot=<0 0 0 1>}
 dyn_skirt_R_C_05 {parent=dyn_skirt_R_C_04, loc=<-11.7084 0.0001 0.0001>, rot=<0 0 0 1>}
 base_skirt_R_D {parent=pelvis, loc=<-0.0417 -10.8296 3.883>, rot=<0.7639 0.0103 -0.0241 -0.6448>}
 dyn_skirt_R_D_01 {parent=base_skirt_R_D, loc=<0 0 0>, rot=<0 0 0 1>}
 dyn_skirt_R_D_02 {parent=dyn_skirt_R_D_01, loc=<-5.1915 0 -0.0001>, rot=<0 0 0 1>}
 base_skirt_R_D_03 {parent=dyn_skirt_R_D_02, loc=<-11.9293 0 0>, rot=<0 0 0 1>}
 dyn_skirt_R_D_03 {parent=base_skirt_R_D_03, loc=<0 0 0>, rot=<0 0 0 1>}
 dyn_skirt_R_D_04 {parent=dyn_skirt_R_D_03, loc=<-11.598 0 0>, rot=<0 0 0 1>}
 dyn_skirt_R_D_05 {parent=dyn_skirt_R_D_04, loc=<-11.7083 0 0>, rot=<0 0 0 1>}
 base_LoinCloth_back {parent=pelvis, loc=<0.4655 -12.3109 0.7347>, rot=<0.0488 -0.9988 0 0.0006>}
 dyn_LoinCloth_back_01 {parent=base_LoinCloth_back, loc=<0 0 0>, rot=<0 0 0 1>}
 dyn_LoinCloth_back_02 {parent=dyn_LoinCloth_back_01, loc=<8.7384 0 0>, rot=<0 0 -0.0551 -0.9985>}
 dyn_LoinCloth_back_03 {parent=dyn_LoinCloth_back_02, loc=<10.4093 0 0>, rot=<0 0 0 1>}
 dyn_LoinCloth_back_04 {parent=dyn_LoinCloth_back_03, loc=<11.2453 -0.0086 -0.0005>, rot=<-0.0276 -0.0001 0.0019 -0.9996>}
 thigh_l_null {parent=pelvis, type=Null, loc=<-2.5693 -0.0109 -7.7501>, rot=<0.0012 0.0572 0.0155 -0.9982>}
 thigh_r_null {parent=pelvis, type=Null, loc=<-2.5693 -0.0109 7.7501>, rot=<0.0572 -0.0012 0.9982 0.0155>}
 deform_glute_base_l {parent=pelvis, loc=<-8.8064 -8.8222 -8.4287>, rot=<0.0275 0.0501 -0.4472 -0.8926>}
 deform_glute_l {parent=deform_glute_base_l, loc=<0 0 0>, rot=<0 0 0 1>}
 deform_glute_base_r {parent=pelvis, loc=<-8.8062 -8.8222 8.4287>, rot=<0.0501 -0.0275 0.8926 -0.4472>}
 deform_glute_r {parent=deform_glute_base_r, loc=<0 0 0>, rot=<0 0 0 1>}
]
variables [
 Anim dynamics On {type=bool, default=False}
 Character moving forward {type=double, default=0}
 Is 2D moving {type=bool, default=False}
 Is Crouching {type=bool, default=False}
 Is Hang Gliding {type=bool, default=False}
 Is Jumping {type=bool, default=False}
 Is Playing Emote {type=bool, default=False}
 Is Sprinting {type=bool, default=False}
 Is Walking and targeting {type=bool, default=False}
 Is Wearing a cape {type=bool, default=False}
 Thig_L_Z_negativeDelta {type=double, default=0}
 Thig_L_Z_positiveDelta {type=double, default=0}
 Thig_R_Z_negativeDelta {type=double, default=0}
 Thig_R_Z_positiveDelta {type=double, default=0}
 isDBNO {type=bool, default=False}
 l calfLocal rz float {type=double, default=0}
 l thighLocal ry float {type=double, default=0}
 l thighLocal rz float {type=double, default=0}
 r calfLocal rz float {type=double, default=0}
 r thighLocal ry float {type=double, default=0}
 r thighLocal rz float {type=double, default=0}
]
