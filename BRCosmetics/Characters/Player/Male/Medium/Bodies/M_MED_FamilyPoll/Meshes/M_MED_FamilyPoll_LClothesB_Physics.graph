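# Physics asset for the lower clothing (LClothesB) of the medium male FamilyPoll cosmetic body, defining bodies for a front loincloth and side skirts.
# Capsules on the pelvis, thighs, calves and feet form the body volumes.
# Four loincloth bodies chain from a tiny base capsule through narrow swing/twist limits that are generally looser toward the lower links.
# Left and right skirt bodies use capsules plus an end box each, with no constraints defined.
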PhysicsAsset /BRCosmetics/Characters/Player/Male/Medium/Bodies/M_MED_FamilyPoll/Meshes/M_MED_FamilyPoll_LClothesB_Physics
bodies {
 base_LoinCloth_front [{shapeType=Capsule, length=0.505, radius=0.505, shapeName=base_LoinCloth_front_capsule, rotation=<rot 0 -90 0>}]
 calf_l [{shapeType=Capsule, length=38.4375, radius=11.375, center=<-24.6324 -2.8287 0.0002>, rotation=<rot 88.5756 -12.1951 162.5744>}]
 calf_r [{shapeType=Capsule, length=38.4375, radius=11.375, center=<24.6324 2.8287 -0.0002>, rotation=<rot 88.5756 -12.1951 -17.4256>}]
 dyn_LoinCloth_front_01 [{shapeType=Capsule, length=10.0176, radius=3, shapeName=dyn_garment_01_capsule, center=<2 2.7 0>}]
 dyn_LoinCloth_front_02 [{shapeType=Capsule, length=12, radius=6, shapeName=dyn_garment_02_capsule, center=<3.3715 4.6 0>}]
 dyn_LoinCloth_front_03 [{shapeType=Capsule, length=9.9497, radius=7.6128, shapeName=dyn_garment_03_capsule, center=<2 4.6 0>}]
 dyn_LoinCloth_front_04 [{shapeType=Capsule, length=10.7481, radius=5.6128, shapeName=dyn_garment_04_capsule, center=<4.9878 2.4721 -0.3569>}]
 dyn_skirt_L_A_03 [{shapeType=Capsule, length=12, radius=4.7748, shapeName=dyn_garment_02_capsule, center=<-3.37 4.2 -4.2907>, rotation=<rot 0 0 90>}]
 dyn_skirt_L_A_04 [{shapeType=Capsule, length=12, radius=4.7748, shapeName=dyn_garment_02_capsule, center=<-3.37 4.2 -3.0325>, rotation=<rot 0 0 90>}]
 dyn_skirt_L_A_05 [{
   shapeType=Box
   extentX=20.4602
   extentY=22.0917
   extentZ=4.3167
   shapeName=dyn_skirt_L_A_05_box
   center=<2.8569 3.3713 0.4754>
   rotation=<rot -0.7018 3.6089 10.9921>
  }]
 dyn_skirt_R_A_03 [{shapeType=Capsule, length=12, radius=5.0467, shapeName=dyn_garment_02_capsule, center=<3.3715 -2.954 4.2029>, rotation=<rot 0 0 90>}]
 dyn_skirt_R_A_04 [{shapeType=Capsule, length=12, radius=5.0467, shapeName=dyn_garment_02_capsule, center=<3.3715 -2.954 4.2029>, rotation=<rot 0 0 90>}]
 dyn_skirt_R_A_05 [{
   shapeType=Box
   extentX=23.3944
   extentY=18.9948
   extentZ=4.3201
   shapeName=dyn_skirt_R_A_05_box
   center=<-2.561 -6.2445 -0.9057>
   rotation=<rot 0.0001 0 10.5556>
  }]
 foot_l [{
   shapeType=Capsule
   length=20.9371
   radius=6.3491
   shapeName=foot_l_capsule
   center=<-3.4883 6.6334 -0.1199>
   rotation=<rot 5.626 -158.2712 -88.574>
  }]
 foot_r [{
   shapeType=Capsule
   length=20.9371
   radius=6.3491
   shapeName=foot_r_capsule
   center=<3.4883 -6.6334 0.1199>
   rotation=<rot 5.626 -158.2712 91.426>
  }]
 pelvis [{
   shapeType=Capsule
   length=15.0027
   radius=14.4144
   shapeName=pelvis_capsule
   center=<1.4065 -1.4628 0.2287>
   rotation=<rot 3.7342 -90 -0.6972>
  }]
 thigh_l [{
   shapeType=Capsule
   length=27.75
   radius=12.5
   shapeName=thigh_l_capsule
   center=<-15.4135 -1.0948 -0.028>
   rotation=<rot 86.7774 178.3456 0.4171>
  }]
 thigh_r [{
   shapeType=Capsule
   length=27.75
   radius=12.5
   shapeName=thigh_r_capsule
   center=<15.4135 1.0948 0.028>
   rotation=<rot 86.7774 178.3456 -179.5829>
  }]
}
constraints [
 {bone1Name=dyn_LoinCloth_front_01, bone2Name=base_LoinCloth_front, swing1LimitDegrees=10, swing2LimitDegrees=2.5, twistLimitDegrees=0}
 {bone1Name=dyn_LoinCloth_front_02, bone2Name=dyn_LoinCloth_front_01, swing1LimitDegrees=10, swing2LimitDegrees=2.5, twistLimitDegrees=0}
 {bone1Name=dyn_LoinCloth_front_03, bone2Name=dyn_LoinCloth_front_02, swing1LimitDegrees=10, swing2LimitDegrees=5, twistLimitDegrees=10}
 {bone1Name=dyn_LoinCloth_front_04, bone2Name=dyn_LoinCloth_front_03, swing1LimitDegrees=10, swing2LimitDegrees=10, twistLimitDegrees=5}
]
