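# Animation Blueprint for the male medium FamilyPoll body cosmetic, adding secondary skirt and loincloth motion to the parent mesh's pose.
# The copied pose runs through a Control Rig and skirt AnimDynamics, then a bool blend between loincloth AnimDynamics and a RigidBody-plus-AnimDynamics alternative.
# Bool-gated additive bone offsets adjust the loincloth before output.
# The update event sets gravity and rigid-body settings from a parameter asset, then derives states such as front-end idle.
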AnimBlueprint /BRCosmetics/Characters/Player/Male/Medium/Bodies/M_MED_FamilyPoll/Meshes/M_MED_FamilyPoll_AnimBP
target_skeleton=&/M_MED_FamilyPoll_Skeleton
graphs [
 AnimGraph {nodes [
   Output Pose(Result=Use cached pose 'FPCache_postDyn'()) {layerGroup=None}
   n1 = Control Rig!(
    Source=Copy Pose From Mesh() {bUseAttachedParent=true}
    Alpha
    Is Jumping:bool=false
    Is Crouching:bool=false
    Is Sprinting:bool=false
    Is Hang Gliding:bool=false
    Anim dynamics On:bool=false
    Is 2D moving:bool=false
    isDBNO:bool=false
    Is Wearing a cape:bool=false
    Is Playing Emote:bool=false
    Is Walking and targeting:bool=false
   ) -> (Pose) {controlRigAssetReference {blueprintRigClass=&/M_MED_FamilyPoll_CtrlRig.M_MED_FamilyPoll_CtrlRig_C}}
   Save cached pose 'FPCache_InputPose'(Pose=@n1) {cacheName=FPCache_InputPose}
   n2 = Anim Dynamics - Bone: dyn_LoinCloth_front_01!(ComponentPose=Local To Component(LocalPose=Use cached pose 'FPCache_SkirtDyn'()), Alpha=1, GravityOverride:Vector="0,0,0") -> (Pose) {
    angularDampingOverride=0.99
    angularSpringConstant=6000
    bAngularSpring=true
    bOverrideAngularDamping=true
    bUseGravityOverride=true
    bUsePlanarLimit=false
    simulationSpace=RootRelative
    alphaBoolBlend {blendOption=Linear}
    boundBone {boneName=dyn_LoinCloth_front_01}
    componentAppliedLinearAccClamp=<6000 6000 6000>
    componentLinearAccScale=<0.5 0.5 0.5>
    componentLinearVelScale=<0.6 0.6 0.6>
    physicsBodyDefinitions [{
      boundBone {boneName=dyn_LoinCloth_front_01}
      boxExtents=<40 40 40>
      constraintSetup {angularLimitsMax=<0 0 50>, angularLimitsMin=<0 0 -5>, angularTarget=<1 0 0>}
      localJointOffset=<4 0 0>
     }]
    simSpaceSettings {maxAngularAcceleration=5, maxAngularVelocity=1.5, simSpaceAngularAlpha=1}
   }
   n3 = Anim Dynamics - Bone: dyn_LoinCloth_front_02!(ComponentPose=@n2, Alpha=0.85, GravityOverride:Vector="0,0,-980") -> (Pose) {
    angularDampingOverride=0.99
    angularSpringConstant=6000
    bAngularSpring=true
    bOverrideAngularDamping=true
    bUseGravityOverride=true
    bUsePlanarLimit=false
    simulationSpace=RootRelative
    alphaBoolBlend {blendOption=Linear}
    boundBone {boneName=dyn_LoinCloth_front_02}
    componentAppliedLinearAccClamp=<6000 6000 6000>
    componentLinearAccScale=<0.5 0.5 0.5>
    componentLinearVelScale=<0.6 0.6 0.6>
    gravityOverride {z=-980}
    physicsBodyDefinitions [{
      boundBone {boneName=dyn_LoinCloth_front_02}
      boxExtents=<20 20 20>
      constraintSetup {angularLimitsMax=<0 5 40>, angularLimitsMin=<0 -5 -0.5>, angularTarget=<1 0 0>}
      localJointOffset=<4 0 0>
     }]
    simSpaceSettings {maxAngularAcceleration=5, maxAngularVelocity=1.5, simSpaceAngularAlpha=1}
   }
   n4 = Anim Dynamics - Bone: dyn_LoinCloth_front_03!(ComponentPose=@n3, Alpha=0.85, GravityOverride:Vector="0,0,-980") -> (Pose) {
    angularDampingOverride=0.99
    angularSpringConstant=6000
    bAngularSpring=true
    bOverrideAngularDamping=true
    bUseGravityOverride=true
    bUsePlanarLimit=false
    simulationSpace=RootRelative
    alphaBoolBlend {blendOption=Linear}
    boundBone {boneName=dyn_LoinCloth_front_03}
    componentAppliedLinearAccClamp=<6000 6000 6000>
    componentLinearAccScale=<0.5 0.5 0.5>
    componentLinearVelScale=<0.6 0.6 0.6>
    gravityOverride {z=-980}
    physicsBodyDefinitions [{
      boundBone {boneName=dyn_LoinCloth_front_03}
      boxExtents=<20 20 20>
      constraintSetup {angularLimitsMax=<0 2.5 30>, angularLimitsMin=<0 -2.5 -0.5>, angularTarget=<1 0 0>}
      localJointOffset=<4 0 0>
     }]
    simSpaceSettings {maxAngularAcceleration=5, maxAngularVelocity=1.5, simSpaceAngularAlpha=1}
   }
   n5 = Anim Dynamics - Bone: dyn_LoinCloth_front_04!(ComponentPose=@n4, Alpha=0.85, GravityOverride:Vector="0,0,-980") -> (Pose) {
    angularDampingOverride=0.99
    angularSpringConstant=6000
    bAngularSpring=true
    bOverrideAngularDamping=true
    bUseGravityOverride=true
    bUsePlanarLimit=false
    simulationSpace=RootRelative
    alphaBoolBlend {blendOption=Linear}
    boundBone {boneName=dyn_LoinCloth_front_04}
    componentAppliedLinearAccClamp=<6000 6000 6000>
    componentLinearAccScale=<0.5 0.5 0.5>
    componentLinearVelScale=<0.6 0.6 0.6>
    gravityOverride {z=-980}
    physicsBodyDefinitions [{
      collisionType=CustomSphere
      sphereCollisionRadius=1.2
      boundBone {boneName=dyn_LoinCloth_front_04}
      boxExtents=<40 40 40>
      constraintSetup {angularLimitsMax=<0 1 20>, angularLimitsMin=<0 -1 -0.5>, angularTarget=<1 0 0>}
      localJointOffset=<4 0 0>
     }]
    simSpaceSettings {maxAngularAcceleration=5, maxAngularVelocity=1.5, simSpaceAngularAlpha=1}
   }
   n6 = Anim Dynamics - Bone: dyn_LoinCloth_back_01!(ComponentPose=@n5, Alpha=1, GravityOverride:Vector="0,0,0") -> (Pose) {
    angularDampingOverride=0.99
    angularSpringConstant=6000
    bAngularSpring=true
    bOverrideAngularDamping=true
    bUseGravityOverride=true
    bUsePlanarLimit=false
    simulationSpace=RootRelative
    alphaBoolBlend {blendOption=Linear}
    boundBone {boneName=dyn_LoinCloth_back_01}
    componentAppliedLinearAccClamp=<6000 6000 6000>
    componentLinearAccScale=<0.5 0.5 0.5>
    componentLinearVelScale=<0.6 0.6 0.6>
    physicsBodyDefinitions [{
      boundBone {boneName=dyn_LoinCloth_back_01}
      boxExtents=<40 40 40>
      constraintSetup {angularLimitsMax=<0 0 0.1>, angularLimitsMin=<0 0 -50>, angularTarget=<1 0 0>}
      localJointOffset=<4 0 0>
     }]
    simSpaceSettings {maxAngularAcceleration=5, maxAngularVelocity=1.5, simSpaceAngularAlpha=1}
   }
   n7 = Anim Dynamics - Bone: dyn_LoinCloth_back_02!(ComponentPose=@n6, Alpha=0.85, GravityOverride:Vector="0,0,-980") -> (Pose) {
    angularDampingOverride=0.99
    angularSpringConstant=6000
    bAngularSpring=true
    bOverrideAngularDamping=true
    bUseGravityOverride=true
    bUsePlanarLimit=false
    simulationSpace=RootRelative
    alphaBoolBlend {blendOption=Linear}
    boundBone {boneName=dyn_LoinCloth_back_02}
    componentAppliedLinearAccClamp=<6000 6000 6000>
    componentLinearAccScale=<0.5 0.5 0.5>
    componentLinearVelScale=<0.6 0.6 0.6>
    gravityOverride {z=-980}
    physicsBodyDefinitions [{
      boundBone {boneName=dyn_LoinCloth_back_02}
      boxExtents=<20 20 20>
      constraintSetup {angularLimitsMax=<0 5 0.1>, angularLimitsMin=<0 -5 -40>, angularTarget=<1 0 0>}
      localJointOffset=<4 0 0>
     }]
    simSpaceSettings {maxAngularAcceleration=5, maxAngularVelocity=1.5, simSpaceAngularAlpha=1}
   }
   n8 = Anim Dynamics - Bone: dyn_LoinCloth_back_03!(ComponentPose=@n7, Alpha=0.85, GravityOverride:Vector="0,0,-980") -> (Pose) {
    angularDampingOverride=0.99
    angularSpringConstant=6000
    bAngularSpring=true
    bOverrideAngularDamping=true
    bUseGravityOverride=true
    bUsePlanarLimit=false
    simulationSpace=RootRelative
    alphaBoolBlend {blendOption=Linear}
    boundBone {boneName=dyn_LoinCloth_back_03}
    componentAppliedLinearAccClamp=<6000 6000 6000>
    componentLinearAccScale=<0.5 0.5 0.5>
    componentLinearVelScale=<0.6 0.6 0.6>
    gravityOverride {z=-980}
    physicsBodyDefinitions [{
      boundBone {boneName=dyn_LoinCloth_back_03}
      boxExtents=<20 20 20>
      constraintSetup {angularLimitsMax=<0 2.5 0.1>, angularLimitsMin=<0 -2.5 -30>, angularTarget=<1 0 0>}
      localJointOffset=<4 0 0>
     }]
    simSpaceSettings {maxAngularAcceleration=5, maxAngularVelocity=1.5, simSpaceAngularAlpha=1}
   }
   n9 = Anim Dynamics - Bone: dyn_LoinCloth_back_04!(ComponentPose=@n8, Alpha=0.85, GravityOverride:Vector="0,0,-980") -> (Pose) {
    angularDampingOverride=0.99
    angularSpringConstant=6000
    bAngularSpring=true
    bOverrideAngularDamping=true
    bUseGravityOverride=true
    bUsePlanarLimit=false
    simulationSpace=RootRelative
    alphaBoolBlend {blendOption=Linear}
    boundBone {boneName=dyn_LoinCloth_back_04}
    componentAppliedLinearAccClamp=<6000 6000 6000>
    componentLinearAccScale=<0.5 0.5 0.5>
    componentLinearVelScale=<0.6 0.6 0.6>
    gravityOverride {z=-980}
    physicsBodyDefinitions [{
      collisionType=CustomSphere
      sphereCollisionRadius=1.2
      boundBone {boneName=dyn_LoinCloth_back_04}
      boxExtents=<40 40 40>
      constraintSetup {angularLimitsMax=<0 1 0.1>, angularLimitsMin=<0 -1 -20>, angularTarget=<1 0 0>}
      localJointOffset=<4 0 0>
     }]
    simSpaceSettings {maxAngularAcceleration=5, maxAngularVelocity=1.5, simSpaceAngularAlpha=1}
   }
   n10 = RigidBody!(ComponentPose=Local To Component(LocalPose=Use cached pose 'FPCache_SkirtDyn'()), Alpha, OverrideWorldGravity:Vector="0,0,0", ExternalForce) -> (Pose) {
    bOverrideWorldGravity=true
    overridePhysicsAsset=&/M_MED_FamilyPoll_LClothesB_Physics
    simulationSpace=BaseBoneSpace
    baseBoneRef {boneName=pelvis}
    componentAppliedLinearAccClamp=<3000 3000 3000>
    simSpaceSettings {
     dampingAlpha=0
     maxAngularAcceleration=16
     maxAngularVelocity=4
     maxLinearAcceleration=800
     maxLinearVelocity=600
     worldAlpha=1
     externalLinearDragV=<0.5 0.5 0.25>
    }
   }
   n11 = Anim Dynamics - Bone: dyn_LoinCloth_back_01(ComponentPose=@n10, Alpha=1, GravityOverride="0,0,0") {
    angularDampingOverride=0.99
    angularSpringConstant=6000
    bAngularSpring=true
    bOverrideAngularDamping=true
    bUseGravityOverride=true
    bUsePlanarLimit=false
    simulationSpace=RootRelative
    alphaBoolBlend {blendOption=Linear}
    alphaScaleBiasClamp {bInterpResult=true}
    boundBone {boneName=dyn_LoinCloth_back_01}
    componentAppliedLinearAccClamp=<6000 6000 6000>
    componentLinearAccScale=<0.5 0.5 0.5>
    componentLinearVelScale=<0.6 0.6 0.6>
    physicsBodyDefinitions [{
      boundBone {boneName=dyn_LoinCloth_back_01}
      boxExtents=<40 40 40>
      constraintSetup {angularLimitsMax=<0 0 0.5>, angularLimitsMin=<0 0 -50>, angularTarget=<1 0 0>}
      localJointOffset=<4 0 0>
     }]
    simSpaceSettings {maxAngularAcceleration=5, maxAngularVelocity=1.5, simSpaceAngularAlpha=1}
   }
   n12 = Anim Dynamics - Bone: dyn_LoinCloth_back_02(ComponentPose=@n11, Alpha=1, GravityOverride="0,0,-980") {
    angularDampingOverride=0.99
    angularSpringConstant=6000
    bAngularSpring=true
    bOverrideAngularDamping=true
    bUseGravityOverride=true
    bUsePlanarLimit=false
    simulationSpace=RootRelative
    alphaBoolBlend {blendOption=Linear}
    alphaScaleBiasClamp {bInterpResult=true}
    boundBone {boneName=dyn_LoinCloth_back_02}
    componentAppliedLinearAccClamp=<6000 6000 6000>
    componentLinearAccScale=<0.5 0.5 0.5>
    componentLinearVelScale=<0.6 0.6 0.6>
    gravityOverride {z=-980}
    physicsBodyDefinitions [{
      boundBone {boneName=dyn_LoinCloth_back_02}
      boxExtents=<20 20 20>
      constraintSetup {angularLimitsMax=<0 0 0.5>, angularLimitsMin=<0 0 -40>, angularTarget=<1 0 0>}
      localJointOffset=<4 0 0>
     }]
    simSpaceSettings {maxAngularAcceleration=5, maxAngularVelocity=1.5, simSpaceAngularAlpha=1}
   }
   n13 = Anim Dynamics - Bone: dyn_LoinCloth_back_03(ComponentPose=@n12, Alpha=1, GravityOverride="0,0,-980") {
    angularDampingOverride=0.99
    angularSpringConstant=6000
    bAngularSpring=true
    bOverrideAngularDamping=true
    bUseGravityOverride=true
    bUsePlanarLimit=false
    simulationSpace=RootRelative
    alphaBoolBlend {blendOption=Linear}
    alphaScaleBiasClamp {bInterpResult=true}
    boundBone {boneName=dyn_LoinCloth_back_03}
    componentAppliedLinearAccClamp=<6000 6000 6000>
    componentLinearAccScale=<0.5 0.5 0.5>
    componentLinearVelScale=<0.6 0.6 0.6>
    gravityOverride {z=-980}
    physicsBodyDefinitions [{
      boundBone {boneName=dyn_LoinCloth_back_03}
      boxExtents=<20 20 20>
      constraintSetup {angularLimitsMax=<0 0 0.5>, angularLimitsMin=<0 0 -30>, angularTarget=<1 0 0>}
      localJointOffset=<4 0 0>
     }]
    simSpaceSettings {maxAngularAcceleration=5, maxAngularVelocity=1.5, simSpaceAngularAlpha=1}
   }
   n14 = Anim Dynamics - Bone: dyn_LoinCloth_back_04(ComponentPose=@n13, Alpha=1, GravityOverride="0,0,-980") {
    angularDampingOverride=0.99
    angularSpringConstant=6000
    bAngularSpring=true
    bOverrideAngularDamping=true
    bUseGravityOverride=true
    bUsePlanarLimit=false
    simulationSpace=RootRelative
    alphaBoolBlend {blendOption=Linear}
    alphaScaleBiasClamp {bInterpResult=true}
    boundBone {boneName=dyn_LoinCloth_back_04}
    componentAppliedLinearAccClamp=<6000 6000 6000>
    componentLinearAccScale=<0.5 0.5 0.5>
    componentLinearVelScale=<0.6 0.6 0.6>
    gravityOverride {z=-980}
    physicsBodyDefinitions [{
      collisionType=CustomSphere
      sphereCollisionRadius=1.2
      boundBone {boneName=dyn_LoinCloth_back_04}
      boxExtents=<40 40 40>
      constraintSetup {angularLimitsMax=<0 0 0.5>, angularLimitsMin=<0 0 -20>, angularTarget=<1 0 0>}
      localJointOffset=<4 0 0>
     }]
    simSpaceSettings {maxAngularAcceleration=5, maxAngularVelocity=1.5, simSpaceAngularAlpha=1}
   }
   Save cached pose 'FPCache_dynamics'(Pose=Blend Poses by bool(BlendPose_0=Component To Local(ComponentPose=@n9), BlendPose_1=Component To Local(ComponentPose=@n14), bActiveValue=true)) {cacheName=FPCache_dynamics}
   n15 = "Transform (Modify) Bone - Bone: base_LoinCloth_front"!(
    ComponentPose=Local To Component(LocalPose=Use cached pose 'FPCache_dynamics'())
    bAlphaBoolEnabled:bool=True
    Translation="0,0,-2"
    Rotation="-0.9,-10,10"
   ) -> (Pose) {
    alphaInputType=Bool
    rotationMode=BMM_Additive
    rotationSpace=BCS_BoneSpace
    translationMode=BMM_Additive
    translationSpace=BCS_BoneSpace
    alphaBoolBlend {blendInTime=1, blendOutTime=0.5}
    boneToModify {boneName=base_LoinCloth_front}
    rotation=<rot -0.9 -10 10>
    translation {z=-2}
   }
   n16 = "Transform (Modify) Bone - Bone: dyn_LoinCloth_front_01"!(ComponentPose=@n15, bAlphaBoolEnabled:bool=True, Translation="0,0,0", Rotation="0,5,5") -> (Pose) {
    alphaInputType=Bool
    rotationMode=BMM_Additive
    rotationSpace=BCS_BoneSpace
    translationMode=BMM_Additive
    translationSpace=BCS_BoneSpace
    alphaBoolBlend {blendInTime=1, blendOutTime=0.5}
    boneToModify {boneName=dyn_LoinCloth_front_01}
    rotation {roll=5, yaw=5}
   }
   n17 = "Transform (Modify) Bone - Bone: dyn_LoinCloth_front_02"!(ComponentPose=@n16, bAlphaBoolEnabled:bool=True, Translation="0,-1,0", Rotation="0,0,5") -> (Pose) {
    alphaInputType=Bool
    rotationMode=BMM_Additive
    rotationSpace=BCS_BoneSpace
    translationMode=BMM_Additive
    translationSpace=BCS_BoneSpace
    alphaBoolBlend {blendInTime=1, blendOutTime=0.5}
    boneToModify {boneName=dyn_LoinCloth_front_02}
    rotation {roll=5}
    translation {y=-1}
   }
   n18 = "Transform (Modify) Bone - Bone: dyn_LoinCloth_front_03"!(ComponentPose=@n17, bAlphaBoolEnabled:bool=True, Translation="0,1,0", Rotation="0,0,0") -> (Pose) {
    alphaInputType=Bool
    rotationMode=BMM_Additive
    rotationSpace=BCS_BoneSpace
    translationMode=BMM_Additive
    translationSpace=BCS_BoneSpace
    alphaBoolBlend {blendInTime=1, blendOutTime=0.5}
    boneToModify {boneName=dyn_LoinCloth_front_03}
    translation {y=1}
   }
   n19 = "Transform (Modify) Bone - Bone: base_LoinCloth_front"(ComponentPose=@n18, bAlphaBoolEnabled=True, Translation="0,0,0", Rotation="0,-4,13") {
    alphaInputType=Bool
    rotationMode=BMM_Additive
    rotationSpace=BCS_BoneSpace
    scaleMode=BMM_Additive
    scaleSpace=BCS_BoneSpace
    translationMode=BMM_Additive
    translationSpace=BCS_BoneSpace
    alphaBoolBlend {blendInTime=0.85, blendOutTime=0.5}
    boneToModify {boneName=base_LoinCloth_front}
    rotation {roll=13, yaw=-4}
   }
   n20 = "Transform (Modify) Bone - Bone: dyn_LoinCloth_front_01"(ComponentPose=@n19, bAlphaBoolEnabled=True, Translation="0,0,0", Rotation="0,0,5") {
    alphaInputType=Bool
    rotationMode=BMM_Additive
    rotationSpace=BCS_BoneSpace
    scaleMode=BMM_Additive
    scaleSpace=BCS_BoneSpace
    translationMode=BMM_Additive
    translationSpace=BCS_BoneSpace
    alphaBoolBlend {blendInTime=1.5, blendOutTime=0.5}
    boneToModify {boneName=dyn_LoinCloth_front_01}
    rotation {roll=5}
   }
   n21 = "Transform (Modify) Bone - Bone: dyn_LoinCloth_front_02"(ComponentPose=@n20, bAlphaBoolEnabled=True, Translation="0,0,0", Rotation="0,-10,0") {
    alphaInputType=Bool
    rotationMode=BMM_Additive
    rotationSpace=BCS_BoneSpace
    scaleMode=BMM_Additive
    scaleSpace=BCS_BoneSpace
    translationMode=BMM_Additive
    translationSpace=BCS_BoneSpace
    alphaBoolBlend {blendInTime=0.5, blendOutTime=0.5}
    boneToModify {boneName=dyn_LoinCloth_front_02}
    rotation {yaw=-10}
   }
   n22 = "Transform (Modify) Bone - Bone: dyn_LoinCloth_front_02"(ComponentPose=@n21, bAlphaBoolEnabled=True, Translation="0,-1,0", Rotation="0,0,0") {
    alphaInputType=Bool
    rotationMode=BMM_Additive
    rotationSpace=BCS_BoneSpace
    translationMode=BMM_Additive
    translationSpace=BCS_BoneSpace
    alphaBoolBlend {blendInTime=0.5, blendOutTime=0.5}
    boneToModify {boneName=dyn_LoinCloth_front_02}
    translation {y=-1}
   }
   n23 = "Transform (Modify) Bone - Bone: dyn_LoinCloth_front_03"(ComponentPose=@n22, bAlphaBoolEnabled=True, Translation="0,0,0", Rotation="0,-5,0.004") {
    alphaInputType=Bool
    rotationMode=BMM_Additive
    rotationSpace=BCS_BoneSpace
    translationMode=BMM_Additive
    translationSpace=BCS_BoneSpace
    alphaBoolBlend {blendInTime=0.5, blendOutTime=0.5}
    boneToModify {boneName=dyn_LoinCloth_front_03}
    rotation {roll=0.004, yaw=-5}
   }
   n24 = "Transform (Modify) Bone - Bone: dyn_LoinCloth_front_04"!(ComponentPose=@n23, bAlphaBoolEnabled:bool=True, Translation="0,0,0", Rotation="0,-5,0") -> (Pose) {
    alphaInputType=Bool
    rotationMode=BMM_Additive
    rotationSpace=BCS_BoneSpace
    translationMode=BMM_Additive
    translationSpace=BCS_BoneSpace
    alphaBoolBlend {blendInTime=0.5, blendOutTime=0.5}
    boneToModify {boneName=dyn_LoinCloth_front_04}
    rotation {yaw=-5}
   }
   n25 = "Transform (Modify) Bone - Bone: dyn_LoinCloth_back_02"!(ComponentPose=@n24, bAlphaBoolEnabled:bool=True, Rotation="0,-15,0") -> (Pose) {
    alphaInputType=Bool
    rotationMode=BMM_Additive
    rotationSpace=BCS_BoneSpace
    translationMode=BMM_Additive
    translationSpace=BCS_BoneSpace
    boneToModify {boneName=dyn_LoinCloth_back_02}
    rotation {yaw=-15}
   }
   n26 = "Transform (Modify) Bone - Bone: dyn_LoinCloth_back_03"!(ComponentPose=@n25, bAlphaBoolEnabled:bool=True, Rotation="0.009,-7,0.033") -> (Pose) {
    alphaInputType=Bool
    rotationMode=BMM_Additive
    rotationSpace=BCS_BoneSpace
    translationMode=BMM_Additive
    translationSpace=BCS_BoneSpace
    boneToModify {boneName=dyn_LoinCloth_back_03}
    rotation=<rot 0.009 -7 0.033>
   }
   Save cached pose 'FPCache_postDyn'(Pose=Component To Local(ComponentPose=@n26)) {cacheName=FPCache_postDyn}
   n27 = Anim Dynamics - Bone: dyn_skirt_L_A_03!(ComponentPose=Local To Component(LocalPose=Use cached pose 'FPCache_InputPose'()), Alpha=0, GravityOverride:Vector="0,0,-980") -> (Pose) {
    angularDampingOverride=0.99
    angularSpringConstant=5500
    bAngularSpring=true
    bOverrideAngularDamping=true
    bUseGravityOverride=true
    bUsePlanarLimit=false
    simulationSpace=RootRelative
    alphaBoolBlend {blendOption=Linear}
    boundBone {boneName=dyn_skirt_L_A_03}
    componentAppliedLinearAccClamp=<6000 6000 6000>
    gravityOverride {z=-980}
    physicsBodyDefinitions [{
      boundBone {boneName=dyn_skirt_L_A_03}
      boxExtents=<20 20 20>
      constraintSetup {angularLimitsMax=<0 0.1 0>, angularLimitsMin=<0 -10 0>, angularTarget=<1 0 0>}
      localJointOffset=<3 0 0>
     }]
    simSpaceSettings {maxAngularAcceleration=15, maxAngularVelocity=7, simSpaceAngularAlpha=0.75}
   }
   n28 = Anim Dynamics - Bone: dyn_skirt_L_A_04!(ComponentPose=@n27, Alpha=0, GravityOverride:Vector="0,0,-980") -> (Pose) {
    angularDampingOverride=0.99
    angularSpringConstant=5500
    bAngularSpring=true
    bOverrideAngularDamping=true
    bUseGravityOverride=true
    bUsePlanarLimit=false
    simulationSpace=RootRelative
    alphaBoolBlend {blendOption=Linear}
    boundBone {boneName=dyn_skirt_L_A_04}
    componentAppliedLinearAccClamp=<6000 6000 6000>
    gravityOverride {z=-980}
    physicsBodyDefinitions [{
      boundBone {boneName=dyn_skirt_L_A_04}
      boxExtents=<20 20 20>
      constraintSetup {angularLimitsMax=<0 0.1 0>, angularLimitsMin=<0 -5 0>, angularTarget=<1 0 0>}
      localJointOffset=<3 0 0>
     }]
    simSpaceSettings {maxAngularAcceleration=15, maxAngularVelocity=7, simSpaceAngularAlpha=0.75}
   }
   n29 = Anim Dynamics - Bone: dyn_skirt_L_A_05!(ComponentPose=@n28, Alpha=1, GravityOverride:Vector="0,0,-980") -> (Pose) {
    angularDampingOverride=0.99
    angularSpringConstant=5500
    bAngularSpring=true
    bOverrideAngularDamping=true
    bUseGravityOverride=true
    bUsePlanarLimit=false
    simulationSpace=RootRelative
    alphaBoolBlend {blendOption=Linear}
    boundBone {boneName=dyn_skirt_L_A_05}
    componentAppliedLinearAccClamp=<6000 6000 6000>
    gravityOverride {z=-980}
    physicsBodyDefinitions [{
      collisionType=CustomSphere
      sphereCollisionRadius=1.2
      boundBone {boneName=dyn_skirt_L_A_05}
      boxExtents=<40 40 40>
      constraintSetup {angularLimitsMax=<0 0.1 0>, angularLimitsMin=<0 -2.5 0>, angularTarget=<1 0 0>}
      localJointOffset=<3 0 0>
     }]
    simSpaceSettings {maxAngularAcceleration=15, maxAngularVelocity=7, simSpaceAngularAlpha=0.75}
   }
   n30 = Anim Dynamics - Bone: dyn_skirt_R_A_03(ComponentPose=@n29, Alpha=0.3, ExternalForce="0,0,0") {
    angularDampingOverride=1
    angularSpringConstant=500
    bAngularSpring=true
    bOverrideAngularDamping=true
    bUseGravityOverride=true
    gravityScale=0.35
    simulationSpace=RootRelative
    boundBone {boneName=dyn_skirt_R_A_03}
    componentAppliedLinearAccClamp=<6000 6000 6000>
    gravityOverride {z=-980}
    physicsBodyDefinitions [{
      boundBone {boneName=dyn_skirt_R_A_03}
      boxExtents=<20 30 30>
      constraintSetup {angularLimitsMax=<0 0.1 0>, angularLimitsMin=<0 -10 0>, angularTarget=<1 0 0>}
      localJointOffset=<-2 0 0>
     }]
    simSpaceSettings {maxAngularAcceleration=15, maxAngularVelocity=7, simSpaceAngularAlpha=0.75}
   }
   n31 = Anim Dynamics - Bone: dyn_skirt_R_A_04(ComponentPose=@n30, Alpha=0.3, ExternalForce="0,0,0") {
    angularDampingOverride=1
    angularSpringConstant=500
    bAngularSpring=true
    bOverrideAngularDamping=true
    bUseGravityOverride=true
    gravityScale=0.35
    simulationSpace=RootRelative
    boundBone {boneName=dyn_skirt_R_A_04}
    componentAppliedLinearAccClamp=<6000 6000 6000>
    gravityOverride {z=-980}
    physicsBodyDefinitions [{
      boundBone {boneName=dyn_skirt_R_A_04}
      boxExtents=<20 30 30>
      constraintSetup {angularLimitsMax=<0 0.1 0>, angularLimitsMin=<0 -5 0>, angularTarget=<1 0 0>}
      localJointOffset=<-2 0 0>
     }]
    simSpaceSettings {maxAngularAcceleration=15, maxAngularVelocity=7, simSpaceAngularAlpha=0.75}
   }
   n32 = Anim Dynamics - Bone: dyn_skirt_R_A_05(ComponentPose=@n31, Alpha=0.3, ExternalForce="0,0,0") {
    angularDampingOverride=1
    angularSpringConstant=500
    bAngularSpring=true
    bOverrideAngularDamping=true
    bUseGravityOverride=true
    gravityScale=0.35
    simulationSpace=RootRelative
    boundBone {boneName=dyn_skirt_R_A_05}
    componentAppliedLinearAccClamp=<6000 6000 6000>
    gravityOverride {z=-980}
    physicsBodyDefinitions [{
      boundBone {boneName=dyn_skirt_R_A_05}
      boxExtents=<20 30 30>
      constraintSetup {angularLimitsMax=<0 0.5 0>, angularLimitsMin=<0 -2.5 0>, angularTarget=<1 0 0>}
      localJointOffset=<-2 0 0>
     }]
    simSpaceSettings {maxAngularAcceleration=15, maxAngularVelocity=7, simSpaceAngularAlpha=0.75}
   }
   n33 = Anim Dynamics - Bone: dyn_skirt_L_B_03!(ComponentPose=@n32, Alpha=0, GravityOverride:Vector="0,0,-980") -> (Pose) {
    angularDampingOverride=0.99
    angularSpringConstant=5500
    bAngularSpring=true
    bOverrideAngularDamping=true
    bUseGravityOverride=true
    bUsePlanarLimit=false
    simulationSpace=RootRelative
    alphaBoolBlend {blendOption=Linear}
    boundBone {boneName=dyn_skirt_L_B_03}
    componentAppliedLinearAccClamp=<6000 6000 6000>
    gravityOverride {z=-980}
    physicsBodyDefinitions [{
      boundBone {boneName=dyn_skirt_L_B_03}
      boxExtents=<20 20 20>
      constraintSetup {angularLimitsMax=<0 0.1 0>, angularLimitsMin=<0 -10 0>, angularTarget=<1 0 0>}
      localJointOffset=<3 0 0>
     }]
    simSpaceSettings {maxAngularAcceleration=15, maxAngularVelocity=7, simSpaceAngularAlpha=0.75}
   }
   n34 = Anim Dynamics - Bone: dyn_skirt_L_B_04!(ComponentPose=@n33, Alpha=0, GravityOverride:Vector="0,0,-980") -> (Pose) {
    angularDampingOverride=0.99
    angularSpringConstant=5500
    bAngularSpring=true
    bOverrideAngularDamping=true
    bUseGravityOverride=true
    bUsePlanarLimit=false
    simulationSpace=RootRelative
    alphaBoolBlend {blendOption=Linear}
    boundBone {boneName=dyn_skirt_L_B_04}
    componentAppliedLinearAccClamp=<6000 6000 6000>
    gravityOverride {z=-980}
    physicsBodyDefinitions [{
      boundBone {boneName=dyn_skirt_L_B_04}
      boxExtents=<20 20 20>
      constraintSetup {angularLimitsMax=<0 0.1 0>, angularLimitsMin=<0 -5 0>, angularTarget=<1 0 0>}
      localJointOffset=<3 0 0>
     }]
    simSpaceSettings {maxAngularAcceleration=15, maxAngularVelocity=7, simSpaceAngularAlpha=0.75}
   }
   n35 = Anim Dynamics - Bone: dyn_skirt_L_B_05!(ComponentPose=@n34, Alpha=1, GravityOverride:Vector="0,0,-980") -> (Pose) {
    angularDampingOverride=0.99
    angularSpringConstant=5500
    bAngularSpring=true
    bOverrideAngularDamping=true
    bUseGravityOverride=true
    bUsePlanarLimit=false
    simulationSpace=RootRelative
    alphaBoolBlend {blendOption=Linear}
    boundBone {boneName=dyn_skirt_L_B_05}
    componentAppliedLinearAccClamp=<6000 6000 6000>
    gravityOverride {z=-980}
    physicsBodyDefinitions [{
      collisionType=CustomSphere
      sphereCollisionRadius=1.2
      boundBone {boneName=dyn_skirt_L_B_05}
      boxExtents=<40 40 40>
      constraintSetup {angularLimitsMax=<0 0.1 0>, angularLimitsMin=<0 -2.5 0>, angularTarget=<1 0 0>}
      localJointOffset=<3 0 0>
     }]
    simSpaceSettings {maxAngularAcceleration=15, maxAngularVelocity=7, simSpaceAngularAlpha=0.75}
   }
   n36 = Anim Dynamics - Bone: dyn_skirt_R_B_03(ComponentPose=@n35, Alpha=0, ExternalForce="0,0,0") {
    angularDampingOverride=1
    angularSpringConstant=500
    bAngularSpring=true
    bOverrideAngularDamping=true
    bUseGravityOverride=true
    gravityScale=0.35
    simulationSpace=RootRelative
    boundBone {boneName=dyn_skirt_R_B_03}
    componentAppliedLinearAccClamp=<6000 6000 6000>
    gravityOverride {z=-980}
    physicsBodyDefinitions [{
      boundBone {boneName=dyn_skirt_R_B_03}
      boxExtents=<20 30 30>
      constraintSetup {angularLimitsMax=<0 0.1 0>, angularLimitsMin=<0 -10 0>, angularTarget=<1 0 0>}
      localJointOffset=<-2 0 0>
     }]
    simSpaceSettings {maxAngularAcceleration=15, maxAngularVelocity=7, simSpaceAngularAlpha=0.75}
   }
   n37 = Anim Dynamics - Bone: dyn_skirt_R_B_04(ComponentPose=@n36, Alpha=0, ExternalForce="0,0,0") {
    angularDampingOverride=1
    angularSpringConstant=500
    bAngularSpring=true
    bOverrideAngularDamping=true
    bUseGravityOverride=true
    gravityScale=0.35
    simulationSpace=RootRelative
    boundBone {boneName=dyn_skirt_R_B_04}
    componentAppliedLinearAccClamp=<6000 6000 6000>
    gravityOverride {z=-980}
    physicsBodyDefinitions [{
      boundBone {boneName=dyn_skirt_R_B_04}
      boxExtents=<20 30 30>
      constraintSetup {angularLimitsMax=<0 0.1 0>, angularLimitsMin=<0 -5 0>, angularTarget=<1 0 0>}
      localJointOffset=<-2 0 0>
     }]
    simSpaceSettings {maxAngularAcceleration=15, maxAngularVelocity=7, simSpaceAngularAlpha=0.75}
   }
   n38 = Anim Dynamics - Bone: dyn_skirt_R_B_05(ComponentPose=@n37, Alpha=0, ExternalForce="0,0,0") {
    angularDampingOverride=1
    angularSpringConstant=500
    bAngularSpring=true
    bOverrideAngularDamping=true
    bUseGravityOverride=true
    gravityScale=0.35
    simulationSpace=RootRelative
    boundBone {boneName=dyn_skirt_R_B_05}
    componentAppliedLinearAccClamp=<6000 6000 6000>
    gravityOverride {z=-980}
    physicsBodyDefinitions [{
      boundBone {boneName=dyn_skirt_R_B_05}
      boxExtents=<20 30 30>
      constraintSetup {angularLimitsMax=<0 0.1 0>, angularLimitsMin=<0 -2.5 0>, angularTarget=<1 0 0>}
      localJointOffset=<-2 0 0>
     }]
    simSpaceSettings {maxAngularAcceleration=15, maxAngularVelocity=7, simSpaceAngularAlpha=0.75}
   }
   n39 = Anim Dynamics - Bone: dyn_skirt_L_C_03!(ComponentPose=@n38, Alpha=1, GravityOverride:Vector="0,0,-980") -> (Pose) {
    angularDampingOverride=0.99
    angularSpringConstant=5500
    bAngularSpring=true
    bOverrideAngularDamping=true
    bUseGravityOverride=true
    bUsePlanarLimit=false
    simulationSpace=RootRelative
    alphaBoolBlend {blendOption=Linear}
    boundBone {boneName=dyn_skirt_L_C_03}
    componentAppliedLinearAccClamp=<6000 6000 6000>
    gravityOverride {z=-980}
    physicsBodyDefinitions [{
      boundBone {boneName=dyn_skirt_L_C_03}
      boxExtents=<20 20 20>
      constraintSetup {angularLimitsMax=<0 0.1 0>, angularLimitsMin=<0 -10 0>, angularTarget=<1 0 0>}
      localJointOffset=<3 0 0>
     }]
    simSpaceSettings {maxAngularAcceleration=15, maxAngularVelocity=7, simSpaceAngularAlpha=0.75}
   }
   n40 = Anim Dynamics - Bone: dyn_skirt_L_C_04!(ComponentPose=@n39, Alpha=1, GravityOverride:Vector="0,0,-980") -> (Pose) {
    angularDampingOverride=0.99
    angularSpringConstant=5500
    bAngularSpring=true
    bOverrideAngularDamping=true
    bUseGravityOverride=true
    bUsePlanarLimit=false
    simulationSpace=RootRelative
    alphaBoolBlend {blendOption=Linear}
    boundBone {boneName=dyn_skirt_L_C_04}
    componentAppliedLinearAccClamp=<6000 6000 6000>
    gravityOverride {z=-980}
    physicsBodyDefinitions [{
      boundBone {boneName=dyn_skirt_L_C_04}
      boxExtents=<20 20 20>
      constraintSetup {angularLimitsMax=<0 0.1 0>, angularLimitsMin=<0 -5 0>, angularTarget=<1 0 0>}
      localJointOffset=<3 0 0>
     }]
    simSpaceSettings {maxAngularAcceleration=15, maxAngularVelocity=7, simSpaceAngularAlpha=0.75}
   }
   n41 = Anim Dynamics - Bone: dyn_skirt_L_C_05!(ComponentPose=@n40, Alpha=1, GravityOverride:Vector="0,0,-980") -> (Pose) {
    angularDampingOverride=0.99
    angularSpringConstant=5500
    bAngularSpring=true
    bOverrideAngularDamping=true
    bUseGravityOverride=true
    bUsePlanarLimit=false
    simulationSpace=RootRelative
    alphaBoolBlend {blendOption=Linear}
    boundBone {boneName=dyn_skirt_L_C_05}
    componentAppliedLinearAccClamp=<6000 6000 6000>
    gravityOverride {z=-980}
    physicsBodyDefinitions [{
      collisionType=CustomSphere
      sphereCollisionRadius=1.2
      boundBone {boneName=dyn_skirt_L_C_05}
      boxExtents=<40 40 40>
      constraintSetup {angularLimitsMax=<0 0.1 0>, angularLimitsMin=<0 -2.5 0>, angularTarget=<1 0 0>}
      localJointOffset=<3 0 0>
     }]
    simSpaceSettings {maxAngularAcceleration=15, maxAngularVelocity=7, simSpaceAngularAlpha=0.75}
   }
   n42 = Anim Dynamics - Bone: dyn_skirt_R_C_03!(ComponentPose=@n41, Alpha=1, GravityOverride:Vector="0,0,-980") -> (Pose) {
    angularDampingOverride=0.99
    angularSpringConstant=5500
    bAngularSpring=true
    bOverrideAngularDamping=true
    bUseGravityOverride=true
    bUsePlanarLimit=false
    simulationSpace=RootRelative
    alphaBoolBlend {blendOption=Linear}
    boundBone {boneName=dyn_skirt_R_C_03}
    componentAppliedLinearAccClamp=<6000 6000 6000>
    gravityOverride {z=-980}
    physicsBodyDefinitions [{
      boundBone {boneName=dyn_skirt_R_C_03}
      boxExtents=<20 20 20>
      constraintSetup {angularLimitsMax=<0 0.2 0>, angularLimitsMin=<0 -10 0>, angularTarget=<1 0 0>}
      localJointOffset=<-3 0 0>
     }]
    simSpaceSettings {maxAngularAcceleration=15, maxAngularVelocity=7, simSpaceAngularAlpha=0.75}
   }
   n43 = Anim Dynamics - Bone: dyn_skirt_R_B_04!(ComponentPose=@n42, Alpha=1, GravityOverride:Vector="0,0,-980") -> (Pose) {
    angularDampingOverride=0.99
    angularSpringConstant=5500
    bAngularSpring=true
    bOverrideAngularDamping=true
    bUseGravityOverride=true
    bUsePlanarLimit=false
    simulationSpace=RootRelative
    alphaBoolBlend {blendOption=Linear}
    boundBone {boneName=dyn_skirt_R_B_04}
    componentAppliedLinearAccClamp=<6000 6000 6000>
    gravityOverride {z=-980}
    physicsBodyDefinitions [{
      boundBone {boneName=dyn_skirt_R_B_04}
      boxExtents=<20 20 20>
      constraintSetup {angularLimitsMax=<0 0.2 0>, angularLimitsMin=<0 -5 0>, angularTarget=<1 0 0>}
      localJointOffset=<-3 0 0>
     }]
    simSpaceSettings {maxAngularAcceleration=15, maxAngularVelocity=7, simSpaceAngularAlpha=0.75}
   }
   n44 = Anim Dynamics - Bone: dyn_skirt_R_B_05!(ComponentPose=@n43, Alpha=1, GravityOverride:Vector="0,0,-980") -> (Pose) {
    angularDampingOverride=0.99
    angularSpringConstant=5500
    bAngularSpring=true
    bOverrideAngularDamping=true
    bUseGravityOverride=true
    bUsePlanarLimit=false
    simulationSpace=RootRelative
    alphaBoolBlend {blendOption=Linear}
    boundBone {boneName=dyn_skirt_R_B_05}
    componentAppliedLinearAccClamp=<6000 6000 6000>
    gravityOverride {z=-980}
    physicsBodyDefinitions [{
      collisionType=CustomSphere
      sphereCollisionRadius=1.2
      boundBone {boneName=dyn_skirt_R_B_05}
      boxExtents=<40 40 40>
      constraintSetup {angularLimitsMax=<0 0.2 0>, angularLimitsMin=<0 -2.5 0>, angularTarget=<1 0 0>}
      localJointOffset=<-3 0 0>
     }]
    simSpaceSettings {maxAngularAcceleration=15, maxAngularVelocity=7, simSpaceAngularAlpha=0.75}
   }
   n45 = Anim Dynamics - Bone: dyn_skirt_L_D_03!(ComponentPose=@n44, Alpha=1, GravityOverride:Vector="0,0,-980") -> (Pose) {
    angularDampingOverride=0.99
    angularSpringConstant=5500
    bAngularSpring=true
    bOverrideAngularDamping=true
    bUseGravityOverride=true
    bUsePlanarLimit=false
    simulationSpace=RootRelative
    alphaBoolBlend {blendOption=Linear}
    boundBone {boneName=dyn_skirt_L_D_03}
    componentAppliedLinearAccClamp=<6000 6000 6000>
    gravityOverride {z=-980}
    physicsBodyDefinitions [{
      boundBone {boneName=dyn_skirt_L_D_03}
      boxExtents=<20 20 20>
      constraintSetup {angularLimitsMax=<0 0.1 0>, angularLimitsMin=<0 -10 0>, angularTarget=<1 0 0>}
      localJointOffset=<3 0 0>
     }]
    simSpaceSettings {maxAngularAcceleration=15, maxAngularVelocity=7, simSpaceAngularAlpha=0.75}
   }
   n46 = Anim Dynamics - Bone: dyn_skirt_L_D_04!(ComponentPose=@n45, Alpha=1, GravityOverride:Vector="0,0,-980") -> (Pose) {
    angularDampingOverride=0.99
    angularSpringConstant=5500
    bAngularSpring=true
    bOverrideAngularDamping=true
    bUseGravityOverride=true
    bUsePlanarLimit=false
    simulationSpace=RootRelative
    alphaBoolBlend {blendOption=Linear}
    boundBone {boneName=dyn_skirt_L_D_04}
    componentAppliedLinearAccClamp=<6000 6000 6000>
    gravityOverride {z=-980}
    physicsBodyDefinitions [{
      boundBone {boneName=dyn_skirt_L_D_04}
      boxExtents=<20 20 20>
      constraintSetup {angularLimitsMax=<0 0.1 0>, angularLimitsMin=<0 -5 0>, angularTarget=<1 0 0>}
      localJointOffset=<3 0 0>
     }]
    simSpaceSettings {maxAngularAcceleration=15, maxAngularVelocity=7, simSpaceAngularAlpha=0.75}
   }
   n47 = Anim Dynamics - Bone: dyn_skirt_L_D_05!(ComponentPose=@n46, Alpha=1, GravityOverride:Vector="0,0,-980") -> (Pose) {
    angularDampingOverride=0.99
    angularSpringConstant=5500
    bAngularSpring=true
    bOverrideAngularDamping=true
    bUseGravityOverride=true
    bUsePlanarLimit=false
    simulationSpace=RootRelative
    alphaBoolBlend {blendOption=Linear}
    boundBone {boneName=dyn_skirt_L_D_05}
    componentAppliedLinearAccClamp=<6000 6000 6000>
    gravityOverride {z=-980}
    physicsBodyDefinitions [{
      collisionType=CustomSphere
      sphereCollisionRadius=1.2
      boundBone {boneName=dyn_skirt_L_D_05}
      boxExtents=<40 40 40>
      constraintSetup {angularLimitsMax=<0 0.1 0>, angularLimitsMin=<0 -2.5 0>, angularTarget=<1 0 0>}
      localJointOffset=<3 0 0>
     }]
    simSpaceSettings {maxAngularAcceleration=15, maxAngularVelocity=7, simSpaceAngularAlpha=0.75}
   }
   n48 = Anim Dynamics - Bone: dyn_skirt_R_D_03!(ComponentPose=@n47, Alpha=1, GravityOverride:Vector="0,0,-980") -> (Pose) {
    angularDampingOverride=0.99
    angularSpringConstant=5500
    bAngularSpring=true
    bOverrideAngularDamping=true
    bUseGravityOverride=true
    bUsePlanarLimit=false
    simulationSpace=RootRelative
    alphaBoolBlend {blendOption=Linear}
    boundBone {boneName=dyn_skirt_R_D_03}
    componentAppliedLinearAccClamp=<6000 6000 6000>
    gravityOverride {z=-980}
    physicsBodyDefinitions [{
      boundBone {boneName=dyn_skirt_R_D_03}
      boxExtents=<20 20 20>
      constraintSetup {angularLimitsMax=<0 0.2 0>, angularLimitsMin=<0 -10 0>, angularTarget=<1 0 0>}
      localJointOffset=<-3 0 0>
     }]
    simSpaceSettings {maxAngularAcceleration=15, maxAngularVelocity=7, simSpaceAngularAlpha=0.75}
   }
   n49 = Anim Dynamics - Bone: dyn_skirt_R_D_04!(ComponentPose=@n48, Alpha=1, GravityOverride:Vector="0,0,-980") -> (Pose) {
    angularDampingOverride=0.99
    angularSpringConstant=5500
    bAngularSpring=true
    bOverrideAngularDamping=true
    bUseGravityOverride=true
    bUsePlanarLimit=false
    simulationSpace=RootRelative
    alphaBoolBlend {blendOption=Linear}
    boundBone {boneName=dyn_skirt_R_D_04}
    componentAppliedLinearAccClamp=<6000 6000 6000>
    gravityOverride {z=-980}
    physicsBodyDefinitions [{
      boundBone {boneName=dyn_skirt_R_D_04}
      boxExtents=<20 20 20>
      constraintSetup {angularLimitsMax=<0 0.2 0>, angularLimitsMin=<0 -5 0>, angularTarget=<1 0 0>}
      localJointOffset=<-3 0 0>
     }]
    simSpaceSettings {maxAngularAcceleration=15, maxAngularVelocity=7, simSpaceAngularAlpha=0.75}
   }
   n50 = Anim Dynamics - Bone: dyn_skirt_R_D_05!(ComponentPose=@n49, Alpha=1, GravityOverride:Vector="0,0,-980") -> (Pose) {
    angularDampingOverride=0.99
    angularSpringConstant=5500
    bAngularSpring=true
    bOverrideAngularDamping=true
    bUseGravityOverride=true
    bUsePlanarLimit=false
    simulationSpace=RootRelative
    alphaBoolBlend {blendOption=Linear}
    boundBone {boneName=dyn_skirt_R_D_05}
    componentAppliedLinearAccClamp=<6000 6000 6000>
    gravityOverride {z=-980}
    physicsBodyDefinitions [{
      collisionType=CustomSphere
      sphereCollisionRadius=1.2
      boundBone {boneName=dyn_skirt_R_D_05}
      boxExtents=<40 40 40>
      constraintSetup {angularLimitsMax=<0 0.2 0>, angularLimitsMin=<0 -2.5 0>, angularTarget=<1 0 0>}
      localJointOffset=<-3 0 0>
     }]
    simSpaceSettings {maxAngularAcceleration=15, maxAngularVelocity=7, simSpaceAngularAlpha=0.75}
   }
   Save cached pose 'FPCache_SkirtDyn'(Pose=Component To Local(ComponentPose=@n50)) {cacheName=FPCache_SkirtDyn}
  ]}
 EventGraph {nodes [
   Event BlueprintUpdateAnimation()
   > n1 = "Gravity Override State Function (Parameter Asset) (Nativized)"(
    CharacterPartAnimInstance=Self-Reference()
    GameTimeInSeconds=GetGameTimeInSeconds()
    Parameters=Get Parameters() -> (Parameters:Fort Character Dynamics Parameters Object Reference)
   )
   > Set GravityOverride(GravityOverride=@n1)
   > Set RBAN Settings Out(RBAN Settings Out=@n1.OutRigidBodyAnimNodeParameters)
   > K2Node_Composite_0 = StateLogic_Graph!(execute) -> (then:Exec) {is=title K2Node_Composite}
   n2 = Get bIsPlayingEmote()
   > n3 = Branch(Condition=AND Boolean(A=Get bIsInFrontEnd(), B=NOT Boolean(A=@n2)))
   > Set FE_Idle(FE_Idle:bool=true) -> (Output_Get:bool)
   Set FE_Idle(execute=@n3.else, FE_Idle=false)
   Event OnParachuteOpened()
   > Switch on EGliderType(Selection=Get PlayerGliderType() -> (PlayerGliderType=HangGlider))
   > Set isHangGliding(isHangGliding:bool=true) -> (Output_Get:bool)
   Event OnParachuteClosed()
   > Set isHangGliding(isHangGliding=false)
   Event OnBeginDBNO() {is=title K2Node_Event}
   > Set BkClothDynAlpha(BkClothDynAlpha:double=0.25) -> (Output_Get:double)
   Event OnEndDBNO() {is=title K2Node_Event}
   > Set BkClothDynAlpha(BkClothDynAlpha=1)
   Event OnCharacterCustomizationLoaded()
   > n4 = Sequence()
   n5 = CheckForBackbling()
   > Set isWearingCape(isWearingCape=OR Boolean(A=@n5.bIsBackblingACape, B=@n5.bIsBackblingLong))
   > n6 = Branch(Condition=Get isWearingCape() -> (isWearingCape:bool))
   > Set BkClothDynAlpha(BkClothDynAlpha=0)
   Set BkClothDynAlpha(execute=@n6.else, BkClothDynAlpha=1)
   n7 = Branch(execute=@n4.then_1, Condition=Get bIsInFrontEnd())
   > Set BkClothDynAlpha(BkClothDynAlpha=0.5)
   Set BkClothDynAlpha(execute=@n7.else, BkClothDynAlpha=1)
   Event BlueprintBeginPlay() {is=title K2Node_Event}
   > Set BeginPlay(BeginPlay:bool=true) -> (Output_Get:bool)
   > Set FrClothDynAlpha(FrClothDynAlpha:double=1) -> (Output_Get:double)
  ]}
 StateLogic_Graph {
  path=EventGraph.K2Node_Composite_0.StateLogic_Graph
  nodes [
   Inputs()
   n1 = Get bIsMoving2D()
   > n2 = Branch(Condition=AND Boolean!(
     A=Get bIsCrouching()
     B=Get bRigidBodyEnabled()
     C:bool="Equal (Boolean)"(A:bool=@n1, B:bool) -> (ReturnValue:bool=~) {is=title K2Node_PromotableOperator}
    ) -> (ReturnValue))
   > n3 = Set Is Crouching and Not Moving(Is Crouching and Not Moving:bool=true) -> (Output_Get:bool)
   n4 = Set Is rorating on place(
    execute=[@n10.then, @n3.then]
    Is rorating on place:bool="Equal (Rotator)"!(A:Rotator=Get PawnDeltaRotation() -> (PawnDeltaRotation:Rotator), B:Rotator, ErrorTolerance:float) -> (ReturnValue:bool=~) {is=title K2Node_PromotableOperator}
   ) -> (Output_Get:bool)
   n5 = Get bIsInFrontEnd()
   n6 = AND Boolean!(A=@n4, B="Equal (Boolean)"(A=Get bIsMoving2D()) -> (ReturnValue=~)) -> (ReturnValue)
   n7 = "Equal (Boolean)"(A=Get bIsCrouching()) -> (ReturnValue=~)
   n8 = Get bIsPlayingEmote()
   n9 = Get bIsDBNO()
   > Set In Idle pose(In Idle pose:bool=AND Boolean!(
     A=AND Boolean!(A=@n6, B=@n7) -> (ReturnValue)
     B=Get BeginPlay() -> (BeginPlay:bool)
     C:bool=NOT Boolean(A=@n5)
     D:bool="Equal (Boolean)"(A=@n8) -> (ReturnValue=~)
     E:bool="Equal (Boolean)"(A=@n9) -> (ReturnValue=~)
    ) -> (ReturnValue)) -> (Output_Get:bool)
   > Set TargetMove(TargetMove:bool=AND Boolean(A=Get bIsMoving2D(), B=Get bIsTargeting())) -> (Output_Get:bool)
   > Outputs!(then:Exec)
   n10 = Set Is Crouching and Not Moving(execute=@n2.else, Is Crouching and Not Moving=false)
  ]
 }
]
variables [
 BeginPlay {type=boolean}
 BkClothDynAlpha {type=number, default=1}
 FE_Idle {type=boolean, default=true}
 FrClothDynAlpha {type=number}
 GravityOverride {type=Vector}
 In Idle pose {type=boolean}
 Is Crouching and Not Moving {type=boolean}
 Is rorating on place {type=boolean}
 isHangGliding {type=boolean}
 isWearingCape {type=boolean}
 Parameters {
  type=FortCharacterDynamicsParameters
  default=/BRCosmetics/Characters/Player/Male/Medium/Bodies/M_MED_VouchTrig/Meshes/M_MED_VouchTrig_Parameters
 }
 RBAN Settings Out {
  type=FortRigidBodyAnimNodeParameters
  default {simSpaceSettings {
    maxAngularAcceleration=1000
    maxAngularVelocity=3
    maxLinearAcceleration=300
    maxLinearVelocity=700
    velocityScaleZ=0
    worldAlpha=1
    externalLinearDragV=<0 0.5 0>
   }}
 }
 TargetMove {type=boolean}
]
