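# Physics asset of collision volumes for the FamilyPoll medium male player body, named as the colliders for its cape.
# Capsules wrap the pelvis, spine, neck and clavicles, with a sphere on the head and extra capsules on spine_02 and spine_05.
# More capsules follow the dynamic back loincloth chain and the lower bones of the left and right skirt C and D chains.
# Arm, leg and foot bodies carry no shapes, and the asset has no constraints.
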PhysicsAsset /BRCosmetics/Characters/Player/Male/Medium/Bodies/M_MED_FamilyPoll/Meshes/M_MED_FamilyPoll_CapeColliders
bodies {
 calf_l []
 calf_r []
 clavicle_l [{shapeType=Capsule, length=9.0192, radius=9.8871, center=<11.0395 -2.0525 2.1368>, rotation=<rot 84.5715 -7.1105 168.5342>}]
 clavicle_r [{shapeType=Capsule, length=9.0192, radius=9.8871, center=<-11.0395 2.0525 -2.1368>, rotation=<rot 84.5715 -7.1105 -11.4658>}]
 dyn_LoinCloth_back_01 [{
   shapeType=Capsule
   length=13.5936
   radius=4.5351
   shapeName=dyn_LoinCloth_back_01_capsule
   center=<4.2308 1.3039 0.4341>
   rotation=<rot -0.9041 -90 -0.6627>
  }]
 dyn_LoinCloth_back_02 [{
   shapeType=Capsule
   length=15.8818
   radius=4.4885
   shapeName=dyn_LoinCloth_back_02_capsule
   center=<6.6669 -0.271 0.1723>
   rotation=<rot -0.296 -90 -0.2683>
  }]
 dyn_LoinCloth_back_03 [{
   shapeType=Capsule
   length=16.8224
   radius=4.588
   shapeName=dyn_LoinCloth_back_03_capsule
   center=<7.3479 -0.0805 -0.2121>
   rotation=<rot 1.412 -90 -0.3775>
  }]
 dyn_LoinCloth_back_04 [{
   shapeType=Capsule
   length=17.1474
   radius=4.434
   shapeName=dyn_LoinCloth_back_04_capsule
   center=<5.274 -0.2496 -0.4277>
   rotation=<rot 4.9162 -90 0.6461>
  }]
 dyn_skirt_L_C_03 [{shapeType=Capsule, length=5.4886, radius=5.7202, shapeName=dyn_skirt_L_C_03_capsule, rotation=<rot 90 0 0>}]
 dyn_skirt_L_C_04 [{shapeType=Capsule, length=5.4886, radius=5.7202, shapeName=dyn_skirt_L_C_04_capsule, rotation=<rot 90 0 0>}]
 dyn_skirt_L_C_05 [{shapeType=Capsule, length=9.2246, radius=6.9186, shapeName=dyn_skirt_L_C_05_capsule, center=<3.6425 0 -1.3157>, rotation=<rot 90 0 0>}]
 dyn_skirt_L_D_03 [{shapeType=Capsule, length=5.4886, radius=5.7202, shapeName=dyn_skirt_L_D_03_capsule, rotation=<rot 90 0 0>}]
 dyn_skirt_L_D_04 [{shapeType=Capsule, length=5.4886, radius=5.7202, shapeName=dyn_skirt_L_D_04_capsule, rotation=<rot 90 0 0>}]
 dyn_skirt_L_D_05 [{shapeType=Capsule, length=9.3928, radius=6.7503, shapeName=dyn_skirt_L_D_05_capsule, center=<3.6425 0 -1.3157>, rotation=<rot 90 0 0>}]
 dyn_skirt_R_C_03 [{shapeType=Capsule, length=5.4886, radius=5.7202, shapeName=dyn_skirt_R_C_03_capsule, rotation=<rot 90 0 0>}]
 dyn_skirt_R_C_04 [{shapeType=Capsule, length=5.4886, radius=5.7202, shapeName=dyn_skirt_R_C_04_capsule, rotation=<rot 90 0 0>}]
 dyn_skirt_R_C_05 [{shapeType=Capsule, length=8.482, radius=7.6612, shapeName=dyn_skirt_R_C_05_capsule, center=<-4.126 0 2.2271>, rotation=<rot 90 180 0>}]
 dyn_skirt_R_D_03 [{shapeType=Capsule, length=5.4886, radius=5.7202, shapeName=dyn_skirt_R_D_03_capsule, rotation=<rot 90 0 0>}]
 dyn_skirt_R_D_04 [{shapeType=Capsule, length=5.4886, radius=5.7202, shapeName=dyn_skirt_R_D_04_capsule, rotation=<rot 90 0 0>}]
 dyn_skirt_R_D_05 [{shapeType=Capsule, length=8.1551, radius=7.988, shapeName=dyn_skirt_R_D_05_capsule, center=<-4.126 0 2.2271>, rotation=<rot 90 180 0>}]
 foot_l []
 foot_r []
 head [{shapeType=Sphere, radius=17.3125, center=<11.028 -0.174 0>}]
 lowerarm_l []
 lowerarm_r []
 neck_02 [{shapeType=Capsule, length=13.4375, radius=7.3175, center=<-0.2642 -0.3408 0>, rotation=<rot 90 -171.1193 -175.3978>}]
 pelvis [{shapeType=Capsule, length=8.6456, radius=14.5388, center=<-2.4887 0.1881 0>}]
 spine_01 [{shapeType=Capsule, length=7.9438, radius=13.5586, center=<-1.7965 0.1884 0>}]
 spine_02 [
  {shapeType=Capsule, length=7.3264, radius=13.6322, center=<-1.711 -0.1851 0.0001>}
  {shapeType=Capsule, length=29.3956, radius=11.5807, center=<2.5024 0.4992 3.9825>, rotation=<rot 90 0.0049 0.0049>}
  {shapeType=Capsule, length=29.875, radius=11.1013, center=<2.4704 0.4991 -3.9825>, rotation=<rot 90 -0.0396 -0.0396>}
 ]
 spine_03 [{shapeType=Capsule, length=8.2107, radius=13.7031, center=<0.1445 0.2396 0.0001>}]
 spine_04 [{shapeType=Capsule, length=9.4404, radius=14.1325, center=<2.766 0.7765 0.0002>}]
 spine_05 [
  {shapeType=Capsule, length=8.0226, radius=15.695, center=<-2.2795 1.8109 0>}
  {shapeType=Capsule, length=13.9518, radius=13.9425, center=<-0.2916 0.4452 0>}
 ]
 thigh_l []
 thigh_r []
 upperarm_l []
 upperarm_r []
}
constraints []
